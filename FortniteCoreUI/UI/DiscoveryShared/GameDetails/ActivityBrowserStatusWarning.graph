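# Warning tag widget for Discovery game details, showing an Attention status indicator beside uppercase outlined text, by default "Photosensitivity Warning".
# A custom event, Play Start Animation!, takes the warning text, sets it on the label and makes the widget visible without being hit-testable itself.
# It then stops running animations and plays OnStarted, which fades the label in while sliding it from the left; the indicator is not animated.
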WidgetBlueprint /FortniteCoreUI/UI/DiscoveryShared/GameDetails/ActivityBrowserStatusWarning
animations [OnStarted {
  binding_count=1
  length_seconds=0.32
  object=&self:OnStarted
  sequence {
   bindings [Text_WarningTagName {
     id=#1
     possessed_class=CommonTextBlock
     tracks [
      {
       type=MovieSceneFloatTrack
       display_name=Render Opacity
       sections [{type=MovieSceneFloatSection, channels [None {type=MovieSceneScriptingFloatChannel, keys [[48.0033,0],[80.0017,1]]}]}]
      }
      {
       type=MovieScene2DTransformTrack
       sections [{
         type=MovieScene2DTransformSection
         channels [
          Translation.X {type=MovieSceneScriptingFloatChannel, keys [[48.0017,-32],[80.0017,0]]}
          Translation.Y {type=MovieSceneScriptingFloatChannel}
          Angle {type=MovieSceneScriptingFloatChannel}
          Scale.X {type=MovieSceneScriptingFloatChannel}
          Scale.Y {type=MovieSceneScriptingFloatChannel}
          Shear.X {type=MovieSceneScriptingFloatChannel}
          Shear.Y {type=MovieSceneScriptingFloatChannel}
         ]
        }]
      }
     ]
    }]
   display_rate {numerator=100}
  }
 }]
graphs [EventGraph {nodes [
   n1 = Play Start Animation!() -> (OutputDelegate, then, In Warning Text:Text) {is=title K2Node_CustomEvent}
   > "SetText (Text)"(self=Get Text_WarningTagName() -> (Text_WarningTagName:Common Text Object Reference), InText=@n1.In Warning Text)
   > SetVisibility(InVisibility=SelfHitTestInvisible)
   > StopAllAnimations(self:User Widget Object Reference)
   > PlayAnimationForward(InAnimation=Get OnStarted() -> (OnStarted:Widget Animation Object Reference))
  ]}]
variables []
widgets {
 info {inheritedWidgetCount=0, parentClass=&UserWidget, rootWidgetClass=&HorizontalBox, widgetCount=3}
 widgets [
  TagHB {widgetClassPath=&HorizontalBox, widget=TagHB}
  WBP_UIKit_StatusIndicator {
   widgetClassPath=/FortUILibrary/UIKit/StatusIndicator/WBP_UIKit_StatusIndicator
   bIsVariable=true
   parent=TagHB
   widget=WBP_UIKit_StatusIndicator
   properties {type=Attention}
   slot {class=HorizontalBoxSlot, properties {horizontalAlignment=HAlign_Left, verticalAlignment=VAlign_Center, padding=<ltrb 0 0 8 0>}}
  }
  Text_WarningTagName {
   widgetClassPath=&CommonTextBlock
   bIsVariable=true
   parent=TagHB
   widget=Text_WarningTagName
   properties {
    mobileFontSizeMultiplier=1.2
    text=Photosensitivity Warning
    textTransformPolicy=ToUpper
    font {
     bForceMonospaced=false
     bMaterialIsStencil=false
     fontMaterial=null
     fontObject=&HeadingNow
     letterSpacing=10
     monospacedWidth=1
     size=15.75
     skewAmount=0
     typefaceFontName=74 Regular
     outlineSettings {
      bApplyOutlineToDropShadows=false
      bMiteredCorners=false
      bSeparateFillAlpha=false
      outlineMaterial=/FortUILibrary/MaterialInstances/Text/MI_UI_Block_Text_Outline
      outlineSize=2
      outlineColor=<rgba 0 0 0 1>
     }
    }
    margin=<ltrb 0 -4 0 -4>
    shadowOffset=<0 0>
   }
   slot {class=HorizontalBoxSlot, properties {verticalAlignment=VAlign_Center}}
  }
 ]
}
