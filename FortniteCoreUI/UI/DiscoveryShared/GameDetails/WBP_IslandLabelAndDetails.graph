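# Discovery game-details widget for an island: a label stacked above a row holding an icon and a value.
# PreConstruct applies icon visibility, the Icon material brush and Label_Text from the widget's variables.
# SetValueText fills the value from a passed-in text and is not called within this widget.
# Both options of the HasIcon visibility select are Collapsed, so the icon stays hidden either way.
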WidgetBlueprint /FortniteCoreUI/UI/DiscoveryShared/GameDetails/WBP_IslandLabelAndDetails
graphs [
 EventGraph {nodes [
   Event PreConstruct()
   > SetIconVisibility()
   > Set Icon Image() {is=title K2Node_CallFunction}
   > SetLabelText()
   Event Construct()
   Event Tick()
  ]}
 Set Icon Image {nodes [
   Set Icon Image() {is=title K2Node_FunctionEntry}
   > SetBrush!(
    execute
    self=Get Image_Icon() -> (Image_Icon:Fort Mobile Image Object Reference)
    InBrush_TintColor:Slate Color Structure=(SpecifiedColor=(R=1,G=1,B=1,A=1),ColorUseRule=UseColor_Specified)
    InBrush_DrawAs:ESlateBrushDrawType Enum=Image
    InBrush_Tiling:ESlateBrushTileType Enum=NoTile
    InBrush_Mirroring:ESlateBrushMirrorType Enum=NoMirror
    InBrush_ImageSize:Vector 2D Structure=(X=32,Y=32)
    InBrush_Margin:Margin Structure=(Left=0,Top=0,Right=0,Bottom=0)
    InBrush_ResourceObject:Object Reference=Get Icon() -> (Icon:Material Instance Object Reference)
    InBrush_OutlineSettings:Slate Brush Outline Settings Structure=(CornerRadii=(X=0,Y=0,Z=0,W=0),Color=(SpecifiedColor=(R=0,G=0,B=0,A=0),ColorUseRule=UseColor_Specified),Width=0,RoundingType=HalfHeightRadius,bUseBrushTransparency=False)
   ) -> (then)
  ]}
 SetIconVisibility {nodes [
   SetIconVisibility() {is=title K2Node_FunctionEntry}
   > SetVisibility(self=Get Image_Icon(), InVisibility=Select(Option 0=Collapsed, Option 1=Collapsed, Index=Get HasIcon() -> (HasIcon:bool)))
  ]}
 SetLabelText {nodes [
   SetLabelText() {is=title K2Node_FunctionEntry}
   > Set Text(Text=Get Label_Text() -> (Label_Text:Text), self=Get Text_Label() -> (Text_Label:Common Text Object Reference))
  ]}
 SetValueText {nodes [
   n1 = SetValueText!() -> (then, NewParam:Text) {is=title K2Node_FunctionEntry}
   > Set Text(Text=@n1, self=Get Text_Value() -> (Text_Value:Common Text Object Reference))
  ]}
]
variables [
 HasIcon {type=boolean, default=true}
 Icon {type=MaterialInstance, default=null}
 Label_Text {type=string, default="{Label Text}"}
]
widgets {
 info {inheritedWidgetCount=0, parentClass=&UserWidget, rootWidgetClass=&VerticalBox, widgetCount=5}
 widgets [
  VB_Content {widgetClassPath=&VerticalBox, widget=VB_Content}
  Text_Label {
   widgetClassPath=&CommonTextBlock
   bIsVariable=true
   parent=VB_Content
   widget=Text_Label
   properties {
    text="{Lable Text}"
    colorAndOpacity {colorUseRule=UseColor_Specified, specifiedColor=<rgba 0.5 0.5 0.5 1>}
    font {
     bForceMonospaced=false
     bMaterialIsStencil=false
     fontMaterial=null
     fontObject=&HeadingNow
     letterSpacing=0
     monospacedWidth=1
     size=20.25
     skewAmount=0
     typefaceFontName=75 Medium
     outlineSettings {
      bApplyOutlineToDropShadows=false
      bMiteredCorners=false
      bSeparateFillAlpha=false
      outlineMaterial=/FortUILibrary/MaterialInstances/Text/MI_UI_Block_Text_Outline
      outlineSize=2
      outlineColor=<rgba 0 0 0 1>
     }
    }
    shadowOffset=<0 0>
   }
   slot {class=VerticalBoxSlot}
  }
  HB_Details {
   widgetClassPath=&HorizontalBox
   bIsVariable=true
   parent=VB_Content
   widget=HB_Details
   properties {visibility=HitTestInvisible}
   slot {class=VerticalBoxSlot}
  }
  Image_Icon {
   widgetClassPath=/Script/FortniteUI.FortMobileImage
   bIsVariable=true
   parent=HB_Details
   widget=Image_Icon
   properties {
    mobileSizeMultiplier=1.2
    visibility=Collapsed
    brush {
     bIsDynamicallyLoaded=false
     drawAs=Image
     imageType=NoImage
     mirroring=NoMirror
     resourceName=None
     resourceObject=/FortniteCoreUI/UI/DiscoveryShared/MI_UI_CCU_UserCount
     tiling=NoTile
     imageSize=<18 18>
     margin=<ltrb 0 0 0 0>
     outlineSettings {
      bUseBrushTransparency=false
      roundingType=HalfHeightRadius
      width=0
      color {colorUseRule=UseColor_Specified, specifiedColor=<rgba 0 0 0 0>}
      cornerRadii=<0 0 0 0>
     }
     tintColor {colorUseRule=UseColor_Specified, specifiedColor=<rgba 1 1 1 1>}
     uVRegion {bIsValid=false, max=<0 0>, min=<0 0>}
    }
   }
   slot {class=HorizontalBoxSlot, properties {verticalAlignment=VAlign_Center, padding=<ltrb 0 0 8 0>}}
  }
  Text_Value {
   widgetClassPath=&CommonTextBlock
   bIsVariable=true
   parent=HB_Details
   widget=Text_Value
   properties {
    mobileFontSizeMultiplier=1.45
    scrollStyle=&TextScrollStyle-Base_C
    text="{Value Text}"
    font {
     bForceMonospaced=false
     bMaterialIsStencil=false
     fontMaterial=null
     fontObject=&HeadingNow
     letterSpacing=0
     monospacedWidth=1
     size=20.25
     skewAmount=0
     typefaceFontName=75 Medium
     outlineSettings {
      bApplyOutlineToDropShadows=false
      bMiteredCorners=false
      bSeparateFillAlpha=false
      outlineMaterial=&MI_UI_Block_Text_Outline
      outlineSize=2
      outlineColor=<rgba 0 0 0 1>
     }
    }
    shadowOffset=<0 0>
   }
   slot {class=HorizontalBoxSlot, properties {verticalAlignment=VAlign_Center, size {sizeRule=Fill, value=1}}}
  }
 ]
}
