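# Discovery UI widget showing the selected game activity's details: name, creator, description, tags, link code, CCU, favorite state, content warnings and action buttons.
# Parent-class events trigger refresh functions that set visibility from show flags, category and lock state, and rebroadcast through dispatchers.
# Invalid-activity reasons swap the accept button for a pulsing warning.
# An animated override slot can replace the details panel, with interaction updated when the animation finishes.
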WidgetBlueprint /FortniteCoreUI/UI/DiscoveryShared/ActivityDetailsDisplay
animations [
 DetailsIntro {
  binding_count=12
  length_seconds=0.48
  object=&self:DetailsIntro
  sequence {bindings [
    Overlay_110 {
     id=#1
     possessed_class=Overlay
     tracks [{
       type=MovieScene2DTransformTrack
       sections [{type=MovieScene2DTransformSection, channels [Translation.Y {type=MovieSceneScriptingFloatChannel, keys [[0,4],[8,0]]}]}]
      }]
    }
    EntryBox_ActivityTags {
     id=#2
     possessed_class=DynamicEntryBox
     tracks [{
       type=MovieSceneFloatTrack
       display_name=Render Opacity
       sections [{type=MovieSceneFloatSection, channels [None {type=MovieSceneScriptingFloatChannel, keys [[3.2,0.125],[7.9997,1]]}]}]
      }]
    }
    StatusEntryBox {
     id=#3
     possessed_class=DynamicEntryBox
     tracks [{
       type=MovieSceneFloatTrack
       display_name=Render Opacity
       sections [{type=MovieSceneFloatSection, channels [None {type=MovieSceneScriptingFloatChannel, keys [[3.2,0],[7.9997,1]]}]}]
      }]
    }
    Text_ActivityDescription {
     id=#4
     possessed_class=CommonTextBlock
     tracks [{
       type=MovieSceneWidgetMaterialTrack
       display_name=Font.FontMaterial
       sections [{type=MovieSceneParameterSection, channels [AlphaOveride {type=MovieSceneScriptingFloatChannel, keys [[3.2,0.125],[7.9997,1]]}]}]
      }]
    }
    NamedSlot_AdditionalDetail {
     id=#5
     possessed_class=NamedSlot
     tracks [
      {
       type=MovieSceneFloatTrack
       display_name=Render Opacity
       sections [{type=MovieSceneFloatSection, channels [None {type=MovieSceneScriptingFloatChannel, keys [[4.7997,0],[9.5993,1]]}]}]
      }
      {
       type=MovieScene2DTransformTrack
       sections [{type=MovieScene2DTransformSection, channels [Translation.Y {type=MovieSceneScriptingFloatChannel, keys [[4.8,8],[9.5997,0]]}]}]
      }
     ]
    }
    DisabledWarningHB {
     id=#6
     possessed_class=HorizontalBox
     tracks [{
       type=MovieSceneFloatTrack
       display_name=Render Opacity
       sections [{type=MovieSceneFloatSection, channels [None {type=MovieSceneScriptingFloatChannel, keys [[3.2,0.125],[7.9997,1]]}]}]
      }]
    }
    Button_Favorite {
     id=#7
     possessed_class=WBP_GameDetails_Toggle_C
     tracks [
      {
       type=MovieSceneFloatTrack
       display_name=Render Opacity
       sections [{type=MovieSceneFloatSection, channels [None {type=MovieSceneScriptingFloatChannel, keys [[6.4,0],[9.5993,1]]}]}]
      }
      {
       type=MovieScene2DTransformTrack
       sections [{
         type=MovieScene2DTransformSection
         channels [
          Scale.X {type=MovieSceneScriptingFloatChannel, keys [[6.4,0.875],[9.5997,1]]}
          Scale.Y {type=MovieSceneScriptingFloatChannel, keys [[6.4,0.875],[9.5997,1]]}
         ]
        }]
      }
     ]
    }
    StatusEntryBox_Redesign {
     id=#8
     possessed_class=DynamicEntryBox
     tracks [{
       type=MovieSceneFloatTrack
       display_name=Render Opacity
       sections [{type=MovieSceneFloatSection, channels [None {type=MovieSceneScriptingFloatChannel, keys [[3.2,0.125],[7.9997,1]]}]}]
      }]
    }
    HorizontalBox_0 {
     id=#9
     possessed_class=HorizontalBox
     tracks [{
       type=MovieSceneFloatTrack
       display_name=Render Opacity
       sections [{type=MovieSceneFloatSection, channels [None {type=MovieSceneScriptingFloatChannel, keys [[3.2,0.125],[7.9993,1]]}]}]
      }]
    }
    Text_ActivityName {
     id=#10
     possessed_class=FortActivityScalingTextBlock
     tracks [
      {
       type=MovieSceneWidgetMaterialTrack
       display_name=Font.FontMaterial
       sections [{type=MovieSceneParameterSection, channels [AlphaOveride {type=MovieSceneScriptingFloatChannel, keys [[0,0.125],[8,1]]}]}]
      }
      {
       type=MovieSceneFloatTrack
       display_name=Line Height Percentage
       sections [{type=MovieSceneFloatSection, end_frame=9, start_frame=0, channels [None {type=MovieSceneScriptingFloatChannel, const [0.85,2]}]}]
      }
     ]
    }
    Button_Accept {
     id=#11
     possessed_class=WBP_UIKit_Button_Loud_C
     tracks [
      {
       type=MovieSceneFloatTrack
       display_name=Render Opacity
       sections [{type=MovieSceneFloatSection, channels [None {type=MovieSceneScriptingFloatChannel, keys [[4.7997,0],[9.5993,1]]}]}]
      }
      {
       type=MovieScene2DTransformTrack
       sections [{type=MovieScene2DTransformSection, channels [Translation.Y {type=MovieSceneScriptingFloatChannel, keys [[4.8,8],[9.5997,0]]}]}]
      }
     ]
    }
    Button_Details {
     id=#12
     possessed_class=WBP_Discovery_DetailsButton_C
     tracks [
      {
       type=MovieSceneFloatTrack
       display_name=Render Opacity
       sections [{type=MovieSceneFloatSection, channels [None {type=MovieSceneScriptingFloatChannel, keys [[6.4,0],[9.5993,1]]}]}]
      }
      {
       type=MovieScene2DTransformTrack
       sections [{
         type=MovieScene2DTransformSection
         channels [
          Scale.X {type=MovieSceneScriptingFloatChannel, keys [[6.4,0.875],[9.5997,1]]}
          Scale.Y {type=MovieSceneScriptingFloatChannel, keys [[6.4,0.875],[9.5997,1]]}
         ]
        }]
      }
     ]
    }
   ]}
 }
 DetailsOutro {
  binding_count=11
  length_seconds=0.32
  object=&self:DetailsOutro
  sequence {
   bindings [
    StatusEntryBox {
     id=#13
     possessed_class=DynamicEntryBox
     tracks [{
       type=MovieSceneFloatTrack
       display_name=Render Opacity
       sections [{type=MovieSceneFloatSection, channels [None {type=MovieSceneScriptingFloatChannel, keys [[0,1],[20,0]]}]}]
      }]
    }
    Text_ActivityName {
     id=#14
     possessed_class=FortActivityScalingTextBlock
     tracks [
      {
       type=MovieSceneWidgetMaterialTrack
       display_name=Font.FontMaterial
       sections [{type=MovieSceneParameterSection, channels [AlphaOveride {type=MovieSceneScriptingFloatChannel, keys [[0,1],[32,0.125]]}]}]
      }
      {
       type=MovieSceneFloatTrack
       display_name=Line Height Percentage
       sections [{type=MovieSceneFloatSection, end_frame=32, start_frame=0, channels [None {type=MovieSceneScriptingFloatChannel, const [0.85,2]}]}]
      }
     ]
    }
    StatusEntryBox_Redesign {
     id=#15
     possessed_class=DynamicEntryBox
     tracks [{
       type=MovieSceneFloatTrack
       display_name=Render Opacity
       sections [{type=MovieSceneFloatSection, channels [None {type=MovieSceneScriptingFloatChannel, keys [[0,1],[20,0.125]]}]}]
      }]
    }
    Text_ActivityOrigin {
     id=#16
     display_name=HorizontalBox_0
     possessed_class=CommonRichTextBlock
     tracks [{
       type=MovieSceneFloatTrack
       display_name=Render Opacity
       sections [{type=MovieSceneFloatSection, channels [None {type=MovieSceneScriptingFloatChannel, keys [[0,1],[20,0.125]]}]}]
      }]
    }
    Text_ActivityDescription {
     id=#17
     possessed_class=CommonTextBlock
     tracks [{
       type=MovieSceneWidgetMaterialTrack
       display_name=Font.FontMaterial
       sections [{type=MovieSceneParameterSection, channels [AlphaOveride {type=MovieSceneScriptingFloatChannel, keys [[0,1],[20,0.125]]}]}]
      }]
    }
    Button_Favorite {
     id=#18
     possessed_class=WBP_GameDetails_Toggle_C
     tracks [
      {
       type=MovieSceneFloatTrack
       display_name=Render Opacity
       sections [{type=MovieSceneFloatSection, channels [None {type=MovieSceneScriptingFloatChannel, keys [[0,1],[12,0]]}]}]
      }
      {
       type=MovieScene2DTransformTrack
       sections [{
         type=MovieScene2DTransformSection
         channels [
          Scale.X {type=MovieSceneScriptingFloatChannel, keys [[0.0017,1],[12,0.875]]}
          Scale.Y {type=MovieSceneScriptingFloatChannel, keys [[0.0017,1],[12,0.875]]}
         ]
        }]
      }
     ]
    }
    DisabledWarningHB {
     id=#19
     possessed_class=HorizontalBox
     tracks [{
       type=MovieSceneFloatTrack
       display_name=Render Opacity
       sections [{type=MovieSceneFloatSection, channels [None {type=MovieSceneScriptingFloatChannel, keys [[0,1],[20,0.125]]}]}]
      }]
    }
    Button_Accept {
     id=#20
     possessed_class=ActivityBrowserSoloButton_C
     tracks [{
       type=MovieSceneFloatTrack
       display_name=Render Opacity
       sections [{type=MovieSceneFloatSection, channels [None {type=MovieSceneScriptingFloatChannel, keys [[0,1],[20,0]]}]}]
      }]
    }
    Button_Details {
     id=#21
     possessed_class=WBP_Discovery_DetailsButton_C
     tracks [
      {
       type=MovieSceneFloatTrack
       display_name=Render Opacity
       sections [{type=MovieSceneFloatSection, channels [None {type=MovieSceneScriptingFloatChannel, keys [[0,1],[12,0]]}]}]
      }
      {
       type=MovieScene2DTransformTrack
       sections [{
         type=MovieScene2DTransformSection
         channels [
          Scale.X {type=MovieSceneScriptingFloatChannel, keys [[0.0017,1],[12,0.875]]}
          Scale.Y {type=MovieSceneScriptingFloatChannel, keys [[0.0017,1],[12,0.875]]}
         ]
        }]
      }
     ]
    }
    Overlay_110 {
     id=#22
     possessed_class=Overlay
     tracks [{
       type=MovieScene2DTransformTrack
       sections [{type=MovieScene2DTransformSection, channels [Translation.Y {type=MovieSceneScriptingFloatChannel, keys [[0,0],[32,4]]}]}]
      }]
    }
    EntryBox_ActivityTags {
     id=#23
     possessed_class=DynamicEntryBox
     tracks [{
       type=MovieSceneFloatTrack
       display_name=Render Opacity
       sections [{type=MovieSceneFloatSection, channels [None {type=MovieSceneScriptingFloatChannel, keys [[0,1],[20,0.125]]}]}]
      }]
    }
   ]
   display_rate {numerator=100}
  }
 }
 DescriptionScrollRestarted {
  binding_count=1
  length_seconds=0.2
  object=&self:DescriptionScrollRestarted
  sequence {bindings [Text_ActivityDescription {
     id=#24
     possessed_class=CommonTextBlock
     tracks [
      {
       type=MovieSceneFloatTrack
       display_name=Render Opacity
       sections [{type=MovieSceneFloatSection, end_frame=4, start_frame=0, channels [None {type=MovieSceneScriptingFloatChannel, keys [[0,0],[4,1]]}]}]
      }
      {
       type=MovieSceneFloatTrack
       display_name=Line Height Percentage
       sections [{type=MovieSceneFloatSection, end_frame=4, start_frame=0, channels [None {type=MovieSceneScriptingFloatChannel, const [0.85,2]}]}]
      }
     ]
    }]}
 }
 DisabledIconPulse {
  binding_count=3
  length_seconds=0.96
  object=&self:DisabledIconPulse
  sequence {bindings [
    Text_ActivityName {
     id=#25
     possessed_class=FortActivityScalingTextBlock
     tracks [{
       type=MovieSceneFloatTrack
       display_name=Line Height Percentage
       sections [{type=MovieSceneFloatSection, end_frame=19, start_frame=0, channels [None {type=MovieSceneScriptingFloatChannel, const [0.85,2]}]}]
      }]
    }
    WarningTriangle_Pulse {
     id=#26
     possessed_class=Image
     tracks [
      {
       type=MovieSceneFloatTrack
       display_name=Render Opacity
       sections [{type=MovieSceneFloatSection, channels [None {type=MovieSceneScriptingFloatChannel, keys [[0,0],[9.6,1],[19.2,0]]}]}]
      }
      {
       type=MovieScene2DTransformTrack
       sections [{
         type=MovieScene2DTransformSection
         channels [
          Translation.X {type=MovieSceneScriptingFloatChannel, keys [[0,0],[9.6,1]]}
          Translation.Y {type=MovieSceneScriptingFloatChannel, keys [[0,0],[9.6,-7]]}
          Scale.X {type=MovieSceneScriptingFloatChannel, keys [[0,1],[9.6,2.25]]}
          Scale.Y {type=MovieSceneScriptingFloatChannel, keys [[0,1],[9.6,2.25]]}
         ]
        }]
      }
      {
       type=MovieSceneWidgetMaterialTrack
       sections [{
         type=MovieSceneParameterSection
         channels [Border_Thickness {type=MovieSceneScriptingFloatChannel, keys [[0,0.125],[11.2,0.05],[19.2,0.001]]}]
        }]
      }
     ]
    }
    IconOvr {
     id=#27
     possessed_class=Overlay
     tracks [{
       type=MovieScene2DTransformTrack
       sections [{
         type=MovieScene2DTransformSection
         channels [
          Scale.X {type=MovieSceneScriptingFloatChannel, keys [[4,1],[5.6,1.25],[8.8,1]]}
          Scale.Y {type=MovieSceneScriptingFloatChannel, keys [[4,1],[5.6,1.25],[8.8,1]]}
         ]
        }]
      }]
    }
   ]}
 }
 ShowDetailsOverrideAnim {
  binding_count=3
  length_seconds=0.3
  object=&self:ShowDetailsOverrideAnim
  sequence {bindings [
    GridPanel_DynamicDetails {
     id=#28
     possessed_class=VerticalBox
     tracks [{
       type=MovieSceneFloatTrack
       display_name=Render Opacity
       sections [{type=MovieSceneFloatSection, channels [None {type=MovieSceneScriptingFloatChannel, keys [[0,1],[3,0],[6,0]]}]}]
      }]
    }
    NamedSlot_DetailsOverride {
     id=#29
     possessed_class=NamedSlot
     tracks [{
       type=MovieSceneFloatTrack
       display_name=Render Opacity
       sections [{type=MovieSceneFloatSection, channels [None {type=MovieSceneScriptingFloatChannel, keys [[0,0],[3,0],[6,1]]}]}]
      }]
    }
    Text_ActivityName {
     id=#30
     possessed_class=FortActivityScalingTextBlock
     tracks [{
       type=MovieSceneFloatTrack
       display_name=Line Height Percentage
       sections [{type=MovieSceneFloatSection, end_frame=6, start_frame=0, channels [None {type=MovieSceneScriptingFloatChannel, const [0.85,2]}]}]
      }]
    }
   ]}
 }
]
graphs [
 Create Favorite Status Widget {nodes [
   Create Favorite Status Widget() {is=title K2Node_FunctionEntry}
   > n1 = Branch(Condition=AND Boolean(A=Get UseRedesign() -> (UseRedesign:bool), B=IsSubModeSelectVisualsEnabled!() -> (ReturnValue:bool)))
   > n2 = Create Entry of Class!(
    execute
    self:Dynamic Entry Box Object Reference=Get StatusEntryBox_Redesign() -> (StatusEntryBox_Redesign:Dynamic Entry Box Object Reference)
    EntryClass:User Widget Class Reference=&/ActivityBrowserStatusFavoriteButton.ActivityBrowserStatusFavoriteButton_C
   ) -> (then, ReturnValue:Activity Browser Status Favorite Button Object Reference)
   > n3 = Set FavoriteWidgetButton(FavoriteWidgetButton:Activity Browser Status Favorite Button Object Reference=@n2) -> (Output_Get:Activity Browser Status Favorite Button Object Reference)
   > Bind Event to On Clicked(self:Common Button Base Object Reference=@n3, Delegate=Create Event()) {is=title K2Node_AddDelegate}
   n4 = Create Entry of Class(
    execute=@n1.else
    self=Get StatusEntryBox() -> (StatusEntryBox:Dynamic Entry Box Object Reference)
    EntryClass=&/ActivityBrowserStatusFavorite.ActivityBrowserStatusFavorite_C
   ) -> (ReturnValue:Activity Browser Status Favorite Object Reference)
   > Set FavoriteWidget(FavoriteWidget:Activity Browser Status Favorite Object Reference=@n4) -> (Output_Get:Activity Browser Status Favorite Object Reference)
  ]}
 EventGraph {nodes [
   "On Clicked (Button_Accept)"!() -> (OutputDelegate, then, Button:Common Button Base Object Reference) {is=title K2Node_ComponentBoundEvent}
   > AcceptActivity(self:Fort Activity Details Display Object Reference)
   FavoriteButtonClicked!() -> (OutputDelegate, then, Button:Common Button Base Object Reference) {is=title K2Node_CustomEvent}
   > ToggleFavorite(self:Fort Activity Details Display Object Reference)
   n1 = Event OnDetailsUpdated!() -> (OutputDelegate, then, bIsDetailsValid:bool, bIsFavoritable:bool) {is=title K2Node_Event}
   > Call On Details Updated Event Dispatcher!(execute:Exec, self:Activity Details Display Object Reference, IsDetailsValid:bool=@n1.bIsDetailsValid, IsFavoritable:bool=@n1.bIsFavoritable) -> (then:Exec) {is=title K2Node_CallDelegate}
   > SetButtonsInteractable!(execute, self, IsInteractionEnabled:bool=@n1.bIsDetailsValid) -> (then)
   > UpdateDetailVisibility!(execute, self, IsFavoritable:bool=@n1.bIsFavoritable) -> (then)
   > RefreshContentWarnings()
   > ResetDetailsOverrideState()
   n2 = GetSelectedGameActivity!(self:Fort Activity Details Display Object Reference) -> (ReturnValue:Fort Game Activity Object Reference)
   > n3 = ShowActivityData!(
    execute
    self:Fort Squeegee Widget Injection Slot Object Reference=[
     Get SqueegeeSlot_LockInfo() -> (SqueegeeSlot_LockInfo:Squeegee Injection Slot Object Reference)
     Get SqueegeeSlot_Card() -> (SqueegeeSlot_Card:Squeegee Injection Slot Object Reference)
    ]
    GameActivity:Fort Game Activity Object Reference=@n2
   ) -> (then)
   n4 = IsLocked!(self:Fort Game Activity Object Reference=@n2) -> (ReturnValue:bool)
   SetActiveWidgetIndex(
    execute=[@n16.then, @n3.then]
    self=Get WidgetSwitcher_ButtonRedesign() -> (WidgetSwitcher_ButtonRedesign:Widget Switcher Object Reference)
    Index=Select(Option 0:int=0, Option 1:int=1, Index=@n4) -> (ReturnValue:int)
   )
   > SetVisibility(
    self=Get NamedSlot_AdditionalDetail() -> (NamedSlot_AdditionalDetail:Named Slot Object Reference)
    InVisibility=Select(Option 0:ESlateVisibility Enum=Visible, Option 1:ESlateVisibility Enum=Collapsed, Index=@n4) -> (ReturnValue:ESlateVisibility Enum)
   )
   > RefreshDisabledState()
   n5 = Event OnIsFavoriteChanged!() -> (OutputDelegate, then, bIsFavorite:bool, bPlayAnimation:bool) {is=title K2Node_Event}
   > Call On Activity Favorite Changed Event Dispatcher!(execute:Exec, self:Activity Details Display Object Reference, Is Favorite:bool=@n5.bIsFavorite, Play Animation:bool=@n5.bPlayAnimation) -> (then:Exec) {is=title K2Node_CallDelegate}
   > UpdateFavoriteStatus!(execute, self, IsFavorite:bool=@n5.bIsFavorite, PlayAnimation:bool=@n5.bPlayAnimation) -> (then)
   n6 = Event OnLocalPlayerPromotedToLeader() {is=title K2Node_Event}
   Call On Refresh Disabled State Event Dispatcher!(execute:Exec=[@n6.then, @n7.then, @n8.then, @n9.then], self:Activity Details Display Object Reference) -> (then:Exec) {is=title K2Node_CallDelegate}
   > RefreshDisabledState()
   n7 = Event OnLocalPlayerDemoted() {is=title K2Node_Event}
   n8 = Event OnPartySizeChanged!() -> (OutputDelegate, then, PartySize:int=0) {is=title K2Node_Event}
   n9 = Event OnParentActivated() {is=title K2Node_Event}
   Event ResetWidgetState() {is=title K2Node_Event}
   > Call On Reset Widget State Event Dispatcher!(execute:Exec, self:Activity Details Display Object Reference) -> (then:Exec) {is=title K2Node_CallDelegate}
   n10 = Get DetailsIntro() -> (DetailsIntro:Widget Animation Object Reference)
   > PlayAnimation(InAnimation=@n10, StartAtTime=GetEndTime(self=@n10))
   Event OnDescriptionScrollRestarted() {is=title K2Node_Event}
   > Call On Description Scroll Restarted Event Dispatcher!(execute:Exec, self:Activity Details Display Object Reference) -> (then:Exec) {is=title K2Node_CallDelegate}
   > PlayAnimation(InAnimation=Get DescriptionScrollRestarted() -> (DescriptionScrollRestarted:Widget Animation Object Reference))
   n11 = Event PreConstruct()
   > Branch(Condition=@n11.IsDesignTime)
   > Is Valid(InputObject=GetParent!(self:Widget Object Reference) -> (ReturnValue:Panel Widget Object Reference))
   > n12 = SetRenderOpacity(
    self=Get GridPanel_DynamicDetails() -> (GridPanel_DynamicDetails:Grid Panel Object Reference)
    InOpacity=Select(Option 0:double=0, Option 1:double=1, Index=Get DesignTimeShowDetails() -> (DesignTimeShowDetails:bool)) -> (ReturnValue:double)
   )
   UpdateDetailVisibility(execute=[@n15.then, @n12.then], IsFavoritable=false)
   > UpdateFavoriteStatus(IsFavorite=false, PlayAnimation=false)
   > UpdateVisualForArabic()
   Event Construct()
   > Parent: Construct()
   n13 = Create Event()
   n14 = Get ShowDetailsOverrideAnim() -> (ShowDetailsOverrideAnim:Widget Animation Object Reference)
   > BindToAnimationFinished(Animation=@n14, Delegate=@n13)
   > Add(
    TargetArray:Array of Common Button Base Object References=Get Buttons() -> (Buttons:Array of Common Button Base Object References)
    NewItem:"Common Button Base Object Reference (by ref)"=Get Button_Accept() -> (Button_Accept:Common Button Base Object Reference)
   ) {is=title K2Node_CallArrayFunction}
   > Add(TargetArray=Get Buttons(), NewItem=Get Button_Details() -> (Button_Details:Common Button Base Object Reference))
   > Add(TargetArray=Get Buttons(), NewItem=Get Button_Favorite() -> (Button_Favorite:Common Button Base Object Reference))
   > n15 = SetButtonsInteractable(IsInteractionEnabled=false)
   Event Destruct()
   > Parent: Destruct()
   > UnbindFromAnimationFinished!(execute, self:User Widget Object Reference, Animation:Widget Animation Object Reference=@n14, Delegate:Delegate=@n13) -> (then)
   Event OnLockStateChanged() {is=title K2Node_Event}
   > n16 = Call On Lock State Changed Event Dispatcher!(execute:Exec, self:Activity Details Display Object Reference) -> (then:Exec) {is=title K2Node_CallDelegate}
   n17 = Event UpdateCCU!() -> (OutputDelegate, then, CCUCount:int=0) {is=title K2Node_Event}
   > Call On Update CCUEvent Dispatcher!(execute:Exec, self:Activity Details Display Object Reference, CCUCount:int=@n17.CCUCount) -> (then:Exec) {is=title K2Node_CallDelegate}
   > SetCCUCountValueText!(execute, self, CCUCount:int=@n17.CCUCount) -> (then)
  ]}
 GetCreatorTextFormat {nodes [
   n1 = GetCreatorTextFormat!() -> (then, CreatorName:"Text (by ref)") {is=title K2Node_FunctionEntry}
   > Return Node!(
    execute
    ReturnValue:Text=Format Text!(Format="LOCTABLE(\"/Game/UI/DiscoveryShared/StringTable_ActivityBrowser\", \"Details_CreatorFormat\")", Creator:"Text (by ref)"=@n1) -> (Result)
   )
  ]}
 OnActivityFavoriteChangedEventDispatcher {nodes [OnActivityFavoriteChangedEventDispatcher!() -> (then, Is Favorite:bool=~, Play Animation:bool=~) {is=title K2Node_FunctionEntry}]}
 OnDescriptionScrollRestartedEventDispatcher {nodes [OnDescriptionScrollRestartedEventDispatcher() {is=title K2Node_FunctionEntry}]}
 OnDetailsOverrideUpdated {nodes [OnDetailsOverrideUpdated!() -> (then, IsVisible:bool=~) {is=title K2Node_FunctionEntry}]}
 OnDetailsUpdatedEventDispatcher {nodes [OnDetailsUpdatedEventDispatcher!() -> (then, IsDetailsValid:bool=~, IsFavoritable:bool=~) {is=title K2Node_FunctionEntry}]}
 OnLockStateChangedEventDispatcher {nodes [OnLockStateChangedEventDispatcher() {is=title K2Node_FunctionEntry}]}
 OnRefreshDisabledStateEventDispatcher {nodes [OnRefreshDisabledStateEventDispatcher() {is=title K2Node_FunctionEntry}]}
 OnResetWidgetStateEventDispatcher {nodes [OnResetWidgetStateEventDispatcher() {is=title K2Node_FunctionEntry}]}
 OnShowDetailsOverrideAnimationFinished {nodes [
   OnShowDetailsOverrideAnimationFinished() {is=title K2Node_FunctionEntry}
   > UpdateDetailsOverrideInteraction()
   > Call On Details Override Updated!(
    execute:Exec
    self:Activity Details Display Object Reference
    IsVisible:bool=Get IsShowingDetailsOverride() -> (IsShowingDetailsOverride:bool)
   ) -> (then:Exec) {is=title K2Node_CallDelegate}
  ]}
 OnUpdateCCUEventDispatcher {nodes [OnUpdateCCUEventDispatcher!() -> (then, CCUCount:int) {is=title K2Node_FunctionEntry}]}
 OnUpdateLinkCodeEventDispatcher {nodes [OnUpdateLinkCodeEventDispatcher!() -> (then, LinkCode:Text, IsCreativeActivity:bool=~) {is=title K2Node_FunctionEntry}]}
 RefreshContentWarnings {nodes [
   RefreshContentWarnings() {is=title K2Node_FunctionEntry}
   > Reset!(
    execute
    self:Dynamic Entry Box Object Reference=Get WarningEntryBox() -> (WarningEntryBox:Dynamic Entry Box Object Reference)
    bDeleteWidgets:bool=false
   ) -> (then)
   n1 = GetContentWarningStrings!(self:Fort Activity Details Display Object Reference) -> (ReturnValue:Array of Strings)
   > Branch(Condition=Is Array Non Empty!(InputArray:Array of Strings=@n1) -> (Non-Empty:bool=~) {is=title K2Node_MacroInstance})
   > n2 = For Each Loop(Array=@n1)
   > Branch(Condition=NOT Boolean(A=IsEmpty(InString=@n2)))
   n3 = Find(TargetMap:Map of Strings to Texts=Get Content Warning Map() -> (Content Warning Map:Map of Strings to Texts), Key=@n2) -> (Value:Text)
   > Branch(Condition=@n3.ReturnValue)
   > n4 = Create Entry of Class(
    self=Get WarningEntryBox()
    EntryClass=/FortniteCoreUI/UI/DiscoveryShared/GameDetails/ActivityBrowserStatusWarning.ActivityBrowserStatusWarning_C
   ) -> (ReturnValue:Activity Browser Status Warning Object Reference)
   > Play Start Animation!(execute, self:Activity Browser Status Warning Object Reference=@n4, In Warning Text:Text=@n3) -> (then)
  ]}
 RefreshDisabledState {nodes [
   RefreshDisabledState() {is=title K2Node_FunctionEntry}
   > n1 = Set Local_InvalidActivityReason(Local_InvalidActivityReason:EFortInvalidActivityReason Enum=GetInvalidActivityReason!(self:Fort Activity Details Display Object Reference) -> (ReturnValue:EFortInvalidActivityReason Enum=None)) -> (Output_Get:EFortInvalidActivityReason Enum=None)
   n2 = IsPreviewTile!(self:Fort Activity Details Display Object Reference) -> (ReturnValue:bool)
   > n3 = Branch(Condition=AND Boolean!(
     A=Get ShowAcceptButton() -> (ShowAcceptButton:bool)
     B="Not Equal (Enum)"!(A:EFortInvalidActivityReason Enum=@n1, B:EFortInvalidActivityReason Enum=InvalidData) -> (ReturnValue:bool)
     C:bool=NOT Boolean(A=@n2)
    ) -> (ReturnValue))
   > "SetText (Text)"(
    self=Get DisabledWarningMessage() -> (DisabledWarningMessage:Common Text Object Reference)
    InText=Select!(
     None:Text
     PartyTooBig:Text="LOCTABLE(\"&StringTable_ActivityBrowser\", \"Warning_PartyTooLarge_Long\")"
     NotPartyLeader:Text="LOCTABLE(\"&StringTable_ActivityBrowser\", \"Warning_NotPartyLeader\")"
     MatchmakingAlready:Text="LOCTABLE(\"&StringTable_ActivityBrowser\", \"Warning_CurrentlyMatchmaking\")"
     PlaylistDisabled:Text="LOCTABLE(\"&StringTable_ActivityBrowser\", \"Warning_ModeDisabled\")"
     PlayerIsLfg:Text="LOCTABLE(\"&StringTable_ActivityBrowser\", \"Warning_PlayerIsLfg\")"
     InvalidData:Text="LOCTABLE(\"&StringTable_ActivityBrowser\", \"Warning_InvalidMode\")"
     ExpiredTournamentLockCache:Text
     Index:EFortInvalidActivityReason Enum=Get Local_InvalidActivityReason() -> (Local_InvalidActivityReason:EFortInvalidActivityReason Enum=None)
    ) -> (ReturnValue:Text)
   )
   n4 = GetSelectedGameActivity()
   n5 = "Equal (Enum)"(
    A:EFortInvalidActivityReason Enum=Get Local_InvalidActivityReason() -> (Local_InvalidActivityReason=None)
    B:EFortInvalidActivityReason Enum=None
   )
   n6 = Get ShowAcceptButton()
   > UpdateVisibility!(
    execute
    self
    Target:Widget Object Reference=Get WidgetSwitcher_ButtonRedesign()
    ShowWidget:bool=OR Boolean(A=AND Boolean!(A=@n6, B=@n5) -> (ReturnValue), B=IsLocked(self=@n4))
   ) -> (then)
   n7 = Get Local_InvalidActivityReason() -> (Local_InvalidActivityReason=None)
   > n8 = Branch(Condition=AND Boolean!(A=Get ShowAcceptButton(), B=NOT Boolean(A=false), C:bool="Not Equal (Enum)"(A=@n7, B=None)) -> (ReturnValue))
   n9 = Get DisabledWarningHB() -> (DisabledWarningHB:Horizontal Box Object Reference)
   > UpdateButtonVisibility!(execute, self, Target:Widget Object Reference=@n9, ShowWidget:bool=true) -> (then)
   > PlayAnimation(InAnimation=Get DisabledIconPulse() -> (DisabledIconPulse:Widget Animation Object Reference), NumLoopsToPlay=3)
   UpdateButtonVisibility(execute=@n8.else, Target=@n9, ShowWidget=false)
   SetVisibility(execute=@n3.else, self=[Get DisabledWarningHB(), Get WidgetSwitcher_ButtonRedesign()], InVisibility=Collapsed)
  ]}
 ResetDetailsOverrideState {nodes [
   ResetDetailsOverrideState() {is=title K2Node_FunctionEntry}
   > SetRenderOpacity(self=Get NamedSlot_DetailsOverride() -> (NamedSlot_DetailsOverride:Named Slot Object Reference), InOpacity=0)
   > SetRenderOpacity(self=Get GridPanel_DynamicDetails(), InOpacity=1)
   > Set IsShowingDetailsOverride(IsShowingDetailsOverride:bool=false) -> (Output_Get:bool)
   > UpdateDetailsOverrideInteraction()
  ]}
 SetButtonsInteractable {nodes [
   n1 = SetButtonsInteractable!() -> (then, IsInteractionEnabled:bool=~) {is=title K2Node_FunctionEntry}
   > n2 = For Each Loop(Array:Array of Common Button Base Object References=Get Buttons()) -> (Array Element:Common Button Base Object Reference)
   > SetIsInteractionEnabled!(execute, self:Common Button Base Object Reference=@n2, bInIsInteractionEnabled:bool=@n1) -> (then)
  ]}
 SetCCUCountValueText {nodes [
   n1 = SetCCUCountValueText!() -> (then, CCUCount:int) {is=title K2Node_FunctionEntry}
   n2 = Get WBP_CCUCount() -> (WBP_CCUCount:WBP Island Label And Details Object Reference)
   > SetVisibility(
    self=@n2
    InVisibility=Select!(Option 0:ESlateVisibility Enum=HitTestInvisible, Option 1:ESlateVisibility Enum=Collapsed, Index:bool="integer < integer"(A=@n1)) -> (ReturnValue:ESlateVisibility Enum)
   )
   > SetValueText!(
    execute
    self:WBP Island Label And Details Object Reference=@n2
    NewParam:Text=Select(
     Option 0:Text="To Text (Integer)"(Value=@n1)
     Option 1:Text=GetShorthandQuantity!(Quantity:int=@n1, MinimumFractionalDigits:int=0, MaximumFractionalDigits:int=1, bShouldRoundDown:bool=false) -> (ReturnValue:Text)
     Index="integer > integer"(A=@n1, B=999)
    ) -> (ReturnValue:Text)
   ) -> (then)
  ]}
 ShowDetailsOverride {nodes [
   n1 = ShowDetailsOverride!() -> (then, IsVisible:bool=~) {is=title K2Node_FunctionEntry}
   > Branch(Condition="Not Equal (Boolean)"(A=@n1, B=Get IsShowingDetailsOverride()) -> (ReturnValue=~))
   > n2 = Set IsShowingDetailsOverride(IsShowingDetailsOverride=Get IsVisible() -> (IsVisible:bool))
   > n3 = Branch(Condition=@n2)
   n4 = Get ShowDetailsOverrideAnim()
   > PlayAnimationForward(InAnimation=@n4)
   PlayAnimationReverse(execute=@n3.else, InAnimation=@n4)
  ]}
 Update Accept Text {nodes [
   n1 = Update Accept Text!() -> (then, ButtonWidget:Object Reference, IsCategory:bool=~, IsDownloadRequired:bool=~) {is=title K2Node_FunctionEntry}
   > n2 = Cast To FortCTAButton!(execute, Object=@n1) -> (then, CastFailed, AsFort CTAButton:Fort CTAButton Object Reference) {is=title K2Node_DynamicCast}
   > SetText!(
    execute
    self:Fort CTAButton Object Reference=@n2
    ButtonText:"Text (by ref)"=Select(
     Option 0="LOCTABLE(\"&StringTable_ActivityBrowser\", \"Activity_Select_Button\")"
     Option 1="LOCTABLE(\"&StringTable_ActivityBrowser\", \"Category_Browse_Button\")"
     Index=@n1.IsCategory
    )
   ) -> (then)
   > SetSecondaryText!(
    execute
    self:Fort CTAButton Object Reference=@n2
    ButtonSecondaryText:"Text (by ref)"=Select(Option 1="LOCTABLE(\"&StringTable_ActivityBrowser\", \"Warning_DownloadRequired\")", Index=@n1.IsDownloadRequired)
   ) -> (then)
  ]}
 UpdateButtonVisibility {nodes [
   n1 = UpdateButtonVisibility!() -> (then, Target:Widget Object Reference, ShowWidget:bool=~) {is=title K2Node_FunctionEntry}
   > Is Valid(InputObject=@n1)
   > SetVisibility(
    self=@n1
    InVisibility=Select(Option 0:ESlateVisibility Enum=Collapsed, Option 1:ESlateVisibility Enum=Visible, Index=@n1.ShowWidget) -> (ReturnValue:ESlateVisibility Enum)
   )
  ]}
 UpdateDetailVisibility {nodes [
   UpdateDetailVisibility!() -> (then, IsFavoritable:bool=~) {is=title K2Node_FunctionEntry}
   > Set IsDownloadRequired(IsDownloadRequired:bool=IsDownloadRequired!(self:Fort Activity Details Display Object Reference) -> (ReturnValue:bool)) -> (Output_Get:bool)
   > n1 = Set IsCategory(IsCategory:bool=IsDisplayingCategory!(self:Fort Activity Details Display Object Reference) -> (ReturnValue:bool)) -> (Output_Get:bool)
   > UpdateVisibility(
    Target=Get Text_ActivityOrigin() -> (Text_ActivityOrigin:Common Rich Text Block Object Reference)
    ShowWidget=AND Boolean!(A=Get ShowCreator() -> (ShowCreator:bool), B=NOT Boolean(A=@n1)) -> (ReturnValue)
   )
   n2 = Get IsCategory() -> (IsCategory:bool)
   > UpdateVisibility(
    Target=Get EntryBox_ActivityTags() -> (EntryBox_ActivityTags:Dynamic Entry Box Object Reference)
    ShowWidget=AND Boolean(A=Get ShowTags() -> (ShowTags:bool), B=NOT Boolean(A=@n2))
   )
   > UpdateVisibility(Target=Get WarningEntryBox(), ShowWidget=Get ShowContentWarning() -> (ShowContentWarning:bool))
   > UpdateVisibility(Target=Get WidgetSwitcher_ButtonRedesign(), ShowWidget=Get ShowAcceptButton())
   > SetActiveWidgetIndex(self=Get WidgetSwitcher_ButtonRedesign(), Index=0)
   > Update Accept Text!(
    execute
    self
    ButtonWidget:Object Reference=Get Button_Accept()
    IsCategory:bool=Get IsCategory()
    IsDownloadRequired:bool=Get IsDownloadRequired() -> (IsDownloadRequired:bool)
   ) -> (then) {is=title K2Node_CallFunction}
   n3 = Get IsCategory()
   > UpdateButtonVisibility(Target=Get DisabledWarningHB(), ShowWidget=AND Boolean(A=Get ShowAcceptButton(), B=NOT Boolean(A=@n3))) {is=title K2Node_CallFunction}
   n4 = IsSubModeSelectVisualsEnabled()
   n5 = Get UseRedesign()
   > n6 = Cast To ActivityBrowserStatusFavoriteButton!(
    execute
    Object=SelectObject!(
     A:Object Reference=Get FavoriteWidgetButton() -> (FavoriteWidgetButton:Activity Browser Status Favorite Button Object Reference)
     B:Object Reference=Get FavoriteWidget() -> (FavoriteWidget:Activity Browser Status Favorite Object Reference)
     bSelectA:bool=AND Boolean(A=@n5, B=@n4)
    ) -> (ReturnValue:Object Reference)
   ) -> (then, CastFailed, AsActivity Browser Status Favorite Button:Activity Browser Status Favorite Button Object Reference) {is=title K2Node_DynamicCast}
   n7 = Get IsCategory()
   > UpdateButtonVisibility(
    Target=@n6
    ShowWidget=AND Boolean!(A=Get ShowFavoriteButton() -> (ShowFavoriteButton:bool), B=NOT Boolean(A=@n7), C:bool=Get IsFavoritable() -> (IsFavoritable:bool)) -> (ReturnValue)
   )
   n8 = IsMobileGame()
   n9 = Get IsCategory()
   > UpdateButtonVisibility(
    Target=Get Button_Details()
    ShowWidget=AND Boolean(A=Get ShowDetailsButton() -> (ShowDetailsButton:bool), B=NOT Boolean(A=@n8), C=NOT Boolean(A=@n9))
   )
  ]}
 UpdateDetailsOverrideInteraction {nodes [
   UpdateDetailsOverrideInteraction() {is=title K2Node_FunctionEntry}
   > n1 = Branch(Condition=Get IsShowingDetailsOverride())
   n2 = Get GridPanel_DynamicDetails()
   > SetVisibility(self=@n2, InVisibility=Collapsed)
   n3 = Get NamedSlot_DetailsOverride()
   > SetVisibility(self=@n3, InVisibility=SelfHitTestInvisible)
   SetVisibility(execute=@n1.else, self=@n2, InVisibility=SelfHitTestInvisible)
   > SetVisibility(self=@n3, InVisibility=Collapsed)
  ]}
 UpdateFavoriteStatus {nodes [
   n1 = UpdateFavoriteStatus!() -> (then, IsFavorite:bool=~, PlayAnimation:bool=~) {is=title K2Node_FunctionEntry}
   > Set IsFavorited(IsFavorited:bool=@n1) -> (Output_Get:bool)
   > n2 = Sequence()
   n3 = AND Boolean!(A=Get UseRedesign(), B=IsSubModeSelectVisualsEnabled()) -> (ReturnValue)
   n4 = Get ShowFavoriteButton()
   > Branch(Condition=OR Boolean(A=Get ShowFavoriteStatus() -> (ShowFavoriteStatus:bool), B=AND Boolean(A=@n4, B=@n3)))
   > n5 = Is Valid(InputObject=SelectObject(A=Get FavoriteWidgetButton(), B=Get FavoriteWidget(), bSelectA=@n3))
   n6 = IsSubModeSelectVisualsEnabled()
   n7 = Get UseRedesign()
   Branch(execute=@n5.Is Not Valid, Condition=OR Boolean(A=AND Boolean(A=@n7, B=@n6), B=Get IsFavorite() -> (IsFavorite:bool)))
   > n8 = Create Favorite Status Widget() {is=title K2Node_CallFunction}
   UpdateFavoriteWidget!(execute=[@n8.then, @n5.Is Valid], self, PlayAnimation:bool=Get PlayAnimation() -> (PlayAnimation:bool), IsFavorited:bool=Get IsFavorite()) -> (then)
   n9 = Cast To WBP_GameDetails_Toggle!(execute=@n2.then_1, Object=Get Button_Favorite()) -> (then, CastFailed, AsWBP Game Details Toggle:WBP Game Details Toggle Object Reference) {is=title K2Node_DynamicCast}
   > Set Toggle!(execute, self:WBP Game Details Toggle Object Reference=@n9, IsInit:bool=false, IsToggled:bool=Get IsFavorite()) -> (then) {is=title K2Node_CallFunction}
  ]}
 UpdateFavoriteWidget {nodes [
   UpdateFavoriteWidget!() -> (then, PlayAnimation:bool=~, IsFavorited:bool=~) {is=title K2Node_FunctionEntry}
   > n1 = Sequence()
   > Is Valid(InputObject=Get FavoriteWidget())
   n2 = Get PlayAnimation()
   > n3 = Branch(Condition=@n2)
   > n4 = Branch(Condition=Get IsFavorited() -> (IsFavorited:bool))
   n5 = Get FavoriteWidget()
   > PlayFavorite(self:Activity Browser Status Favorite Object Reference=@n5)
   PlayUnfavorite(execute=@n4.else, self:Activity Browser Status Favorite Object Reference=@n5)
   SetFavoriteNoAnimation!(execute=@n3.else, self:Activity Browser Status Favorite Object Reference=Get FavoriteWidget(), IsFavorited:bool=Get IsFavorited()) -> (then)
   Is Valid(exec=@n1.then_1, InputObject=Get FavoriteWidgetButton())
   > n6 = Branch(Condition=@n2)
   > OnIsFavoriteChanged!(execute, self:Activity Browser Status Favorite Button Object Reference=Get FavoriteWidgetButton(), bIsFavorite:bool=Get IsFavorited()) -> (then)
   SetFavoriteNoAnim!(
    execute=@n6.else
    self:Activity Browser Status Favorite Button Object Reference=Get FavoriteWidgetButton()
    IsFavorite:bool=Get IsFavorited()
   ) -> (then)
  ]}
 UpdateLinkCode {nodes [
   n1 = UpdateLinkCode!() -> (then, LinkCode:"Text (by ref)", bIsCreativeActivity:bool) {is=title K2Node_FunctionEntry}
   > Call On Update Link Code Event Dispatcher!(execute:Exec, self:Activity Details Display Object Reference, LinkCode:Text=@n1, IsCreativeActivity:bool=@n1.bIsCreativeActivity) -> (then:Exec) {is=title K2Node_CallDelegate}
   n2 = NOT Boolean(A=TextIsEmpty(InText=Get LinkCode() -> (LinkCode:"Text (by ref)")))
   n3 = IsSubModeSelectVisualsEnabled()
   n4 = Get ShowLinkCode() -> (ShowLinkCode:bool)
   > SetVisibility(
    self=Get Text_LinkCode() -> (Text_LinkCode:Common Rich Text Block Object Reference)
    InVisibility=Select(
     Option 0=Collapsed
     Option 1=SelfHitTestInvisible
     Index=AND Boolean!(A=@n1.bIsCreativeActivity, B=@n3, C:bool=@n4, D:bool=@n2) -> (ReturnValue)
    )
   )
   > SetText!(
    execute
    self:Rich Text Block Object Reference=Get Text_LinkCode()
    InText:"Text (by ref)"=Format Text!(
     Format="LOCTABLE(\"&StringTable_ActivityBrowser\", \"Details_LinkCodeFormat\")"
     XXXX-XXXX-XXXX-XXXX:"Text (by ref)"=Get LinkCode()
    ) -> (Result)
   ) -> (then)
  ]}
 UpdateVisibility {nodes [
   n1 = UpdateVisibility!() -> (then, Target:Widget Object Reference, ShowWidget:bool=~) {is=title K2Node_FunctionEntry}
   > Is Valid(InputObject=@n1)
   > SetVisibility(self=@n1, InVisibility=Select(Option 0=Collapsed, Option 1=SelfHitTestInvisible, Index=@n1.ShowWidget))
  ]}
 UpdateVisualForArabic {nodes [
   UpdateVisualForArabic() {is=title K2Node_FunctionEntry}
   n1 = "Equal Exactly (String)"(A=GetCurrentCulture!() -> (ReturnValue:String), B=ar)
   > Set HorizontalAlignment!(
    execute
    HorizontalAlignment:EHorizontalAlignment Enum=Select(Option 0:EHorizontalAlignment Enum=HAlign_Left, Option 1:EHorizontalAlignment Enum=HAlign_Right, Index=@n1) -> (ReturnValue:EHorizontalAlignment Enum)
    self:Overlay Slot Object Reference=SlotAsOverlaySlot(Widget=Get CommonVisualAttachment() -> (CommonVisualAttachment:Common Visual Attachment Object Reference))
   ) -> (then, Output_Get:EHorizontalAlignment Enum=HAlign_Fill)
   > Set ContentAnchor!(
    execute
    ContentAnchor:Vector 2D Structure=Select(Option 0:Vector 2D Structure=(X=0,Y=1), Option 1:Vector 2D Structure=(X=1,Y=1), Index=@n1) -> (ReturnValue:Vector 2D Structure)
    self:Common Visual Attachment Object Reference=Get CommonVisualAttachment()
   ) -> (then, Output_Get:Vector 2D Structure)
  ]}
]
variables [
 Buttons {type="array<CommonButtonBase>", default []}
 Content Warning Map {
  type="map<string>"
  default {
   PartyLargerThanTeam=Your party is larger than Maximum team size. Playing this island will cause some party members to be on a different team.
   Photosensitivity=Photosensitivity Warning
  }
 }
 DesignTimeShowDetails {type=boolean, default=true}
 FavoriteWidget {type=ActivityBrowserStatusFavorite_C, default=null}
 FavoriteWidgetButton {type=ActivityBrowserStatusFavoriteButton_C, default=null}
 IsFavorited {type=boolean}
 IsShowingDetailsOverride {type=boolean}
 OnActivityFavoriteChangedEventDispatcher {}
 OnDescriptionScrollRestartedEventDispatcher {}
 OnDetailsOverrideUpdated {}
 OnDetailsUpdatedEventDispatcher {}
 OnLockStateChangedEventDispatcher {}
 OnRefreshDisabledStateEventDispatcher {}
 OnResetWidgetStateEventDispatcher {}
 OnUpdateCCUEventDispatcher {}
 OnUpdateLinkCodeEventDispatcher {}
 ShowAcceptButton {type=boolean}
 ShowContentWarning {type=boolean, default=true}
 ShowCreator {type=boolean, default=true}
 ShowDetailsButton {type=boolean}
 ShowFavoriteButton {type=boolean}
 ShowFavoriteStatus {type=boolean}
 ShowLinkCode {type=boolean}
 ShowTags {type=boolean, default=true}
 UseRedesign {type=boolean}
]
widgets {
 info {inheritedWidgetCount=8, parentClass=/Script/FortniteUI.FortActivityDetailsDisplay, rootWidgetClass=&Overlay, widgetCount=64}
 widgets [
  Overlay_110 {widgetClassPath=&Overlay, widget=Overlay_110}
  NamedSlot_DetailsOverride {
   widgetClassPath=&NamedSlot
   bIsVariable=true
   parent=Overlay_110
   widget=NamedSlot_DetailsOverride
   properties {bExposeOnInstanceOnly=true}
   slot {class=OverlaySlot, properties {horizontalAlignment=HAlign_Fill, verticalAlignment=VAlign_Fill}}
  }
  VBox_SubModeList {widgetClassPath=&VerticalBox, bIsVariable=true, parent=NamedSlot_DetailsOverride, widget=VBox_SubModeList, slot {class=PanelSlot}}
  HorizontalBox_ModeSelectionAvailable_SubMode {
   widgetClassPath=&HorizontalBox
   bIsVariable=true
   parent=VBox_SubModeList
   widget=HorizontalBox_ModeSelectionAvailable_SubMode
   slot {class=VerticalBoxSlot, properties {padding=<ltrb 0 0 0 8>}}
  }
  Image {
   widgetClassPath=&Image
   parent=HorizontalBox_ModeSelectionAvailable_SubMode
   widget=Image
   properties {
    visibility=SelfHitTestInvisible
    brush {
     bIsDynamicallyLoaded=false
     drawAs=Image
     imageType=NoImage
     mirroring=NoMirror
     resourceName=None
     resourceObject=/FortniteCoreUI/UI/DiscoveryShared/Art/T_UI_Discovery_SubModeIcon
     tiling=NoTile
     imageSize=<32 32>
     margin=<ltrb 0 0 0 0>
     outlineSettings {
      bUseBrushTransparency=false
      roundingType=HalfHeightRadius
      width=0
      color {colorUseRule=UseColor_Specified, specifiedColor=<rgba 0 0 0 0>}
      cornerRadii=<0 0 0 0>
     }
     tintColor {colorUseRule=UseColor_Specified, specifiedColor=<rgba 1 1 1 1>}
     uVRegion {bIsValid=false, max=<0 0>, min=<0 0>}
    }
    colorAndOpacity=<rgba 0.0331 0.402 1 1>
   }
   slot {class=HorizontalBoxSlot, properties {horizontalAlignment=HAlign_Left, verticalAlignment=VAlign_Center, padding=<ltrb 0 0 10 0>}}
  }
  CommonTextBlock {
   widgetClassPath=&CommonTextBlock
   parent=HorizontalBox_ModeSelectionAvailable_SubMode
   widget=CommonTextBlock
   properties {
    mobileFontSizeMultiplier=1.2
    text=Mode Selection
    textTransformPolicy=ToUpper
    colorAndOpacity {colorUseRule=UseColor_Specified, specifiedColor=<rgba 0.0331 0.402 1 1>}
    font {
     bForceMonospaced=false
     bMaterialIsStencil=false
     fontMaterial=null
     fontObject=&BurbankSmall
     letterSpacing=0
     monospacedWidth=1
     size=18
     skewAmount=0
     typefaceFontName=Black
     outlineSettings {
      bApplyOutlineToDropShadows=false
      bMiteredCorners=false
      bSeparateFillAlpha=false
      outlineMaterial=null
      outlineSize=0
      outlineColor=<rgba 0 0 0 1>
     }
    }
    shadowOffset=<0 0>
   }
   slot {class=HorizontalBoxSlot}
  }
  SizeBox_Habanero {
   widgetClassPath=&SizeBox
   bIsVariable=true
   parent=VBox_SubModeList
   widget=SizeBox_Habanero
   properties {bOverride_HeightOverride=true, heightOverride=69, visibility=Collapsed}
   slot {class=VerticalBoxSlot}
  }
  Border_Habanero {
   widgetClassPath=&Border
   parent=SizeBox_Habanero
   widget=Border_Habanero
   properties {
    background {
     bIsDynamicallyLoaded=false
     drawAs=Image
     imageType=NoImage
     mirroring=NoMirror
     resourceName=None
     resourceObject=/FortniteCoreUI/UI/Subscriptions/Art/M_UI_ShopTimerModalBg
     tiling=NoTile
     imageSize=<256 1>
     margin=<ltrb 0 0 0 0>
     outlineSettings {
      bUseBrushTransparency=false
      roundingType=HalfHeightRadius
      width=0
      color {colorUseRule=UseColor_Specified, specifiedColor=<rgba 0 0 0 0>}
      cornerRadii=<0 0 0 0>
     }
     tintColor {colorUseRule=UseColor_Specified, specifiedColor=<rgba 1 1 1 0.75>}
     uVRegion {bIsValid=false, max=<0 0>, min=<0 0>}
    }
    brushColor=<rgba 0.1719 0.1719 0.1719 1>
    padding=<ltrb 20 5 20 5>
   }
   slot {class=SizeBoxSlot}
  }
  HorizontalBox {
   widgetClassPath=&HorizontalBox
   parent=Border_Habanero
   widget=HorizontalBox
   slot {class=BorderSlot, properties {padding=<ltrb 20 5 20 5>}}
  }
  Icon_Habanero {
   widgetClassPath=/Game/UI/Frontend/Lobby/Habanero/WBP_Habanero_Icon
   bIsVariable=true
   parent=HorizontalBox
   widget=Icon_Habanero
   properties {iconMinSize=<64 64>}
   slot {class=HorizontalBoxSlot}
  }
  Text_Habanero {
   widgetClassPath=&CommonTextBlock
   bIsVariable=true
   parent=HorizontalBox
   widget=Text_Habanero
   properties {
    mobileFontSizeMultiplier=1.2
    text="[PH] -"
    textTransformPolicy=ToUpper
    colorAndOpacity {colorUseRule=UseColor_Specified, specifiedColor=<rgba 0.5089 0.5583 0.7084 1>}
    font {
     bForceMonospaced=false
     bMaterialIsStencil=false
     fontMaterial=null
     fontObject=&BurbankSmall
     letterSpacing=0
     monospacedWidth=1
     size=18
     skewAmount=0
     typefaceFontName=Black
     outlineSettings {
      bApplyOutlineToDropShadows=false
      bMiteredCorners=false
      bSeparateFillAlpha=false
      outlineMaterial=null
      outlineSize=0
      outlineColor=<rgba 0 0 0 1>
     }
    }
    shadowOffset=<0 0>
   }
   slot {class=HorizontalBoxSlot, properties {verticalAlignment=VAlign_Center}}
  }
  Button_Activity_Habanero {
   widgetClassPath=/Game/UI/Frontend/Lobby/Habanero/ActivityHabaneroButton
   bIsVariable=true
   parent=VBox_SubModeList
   widget=Button_Activity_Habanero
   properties {visibility=Collapsed}
   slot {class=VerticalBoxSlot}
  }
  Button_PrivacyButton {
   widgetClassPath=/FortniteCoreUI/UI/DiscoveryShared/ModeSet/ActivityPrivacyButton
   bIsVariable=true
   parent=VBox_SubModeList
   widget=Button_PrivacyButton
   slot {class=VerticalBoxSlot}
  }
  Button_SquadFillButton {
   widgetClassPath=/FortniteCoreUI/UI/DiscoveryShared/ModeSet/ActivitySquadFillButton
   bIsVariable=true
   parent=VBox_SubModeList
   widget=Button_SquadFillButton
   slot {class=VerticalBoxSlot}
  }
  List_SubModeList {
   widgetClassPath=/Script/FortniteUI.FortActivityModeSetSelection
   bIsVariable=true
   parent=VBox_SubModeList
   widget=List_SubModeList
   properties {
    entryWidgetClass=/FortniteCoreUI/UI/DiscoveryShared/ModeSet/ActivityBrowserModeSetListEntry.ActivityBrowserModeSetListEntry_C
    horizontalEntrySpacing=5
    navigation=&self:WidgetTree.List_SubModeList.WidgetNavigation_0
    numDesignerPreviewEntries=4
    verticalEntrySpacing=5
   }
   slot {class=VerticalBoxSlot, properties {size {sizeRule=Fill, value=1}}}
  }
  GridPanel_DynamicDetails {
   widgetClassPath=&GridPanel
   bIsVariable=true
   parent=Overlay_110
   widget=GridPanel_DynamicDetails
   properties {columnFill [1], rowFill [0]}
   slot {class=OverlaySlot, properties {horizontalAlignment=HAlign_Fill, verticalAlignment=VAlign_Fill}}
  }
  VerticalBox_0 {widgetClassPath=&VerticalBox, parent=GridPanel_DynamicDetails, widget=VerticalBox_0, slot {class=GridSlot}}
  Overlay_1 {
   widgetClassPath=&Overlay
   parent=VerticalBox_0
   widget=Overlay_1
   slot {class=VerticalBoxSlot, properties {size {sizeRule=Fill, value=1}}}
  }
  CommonVisualAttachment {
   widgetClassPath=&CommonVisualAttachment
   bIsVariable=true
   parent=Overlay_1
   widget=CommonVisualAttachment
   properties {flowDirectionPreference=Culture, contentAnchor=<0 1>}
   slot {class=OverlaySlot}
  }
  VerticalBox_1 {
   widgetClassPath=&VerticalBox
   parent=CommonVisualAttachment
   widget=VerticalBox_1
   slot {class=SizeBoxSlot, properties {horizontalAlignment=HAlign_Right}}
  }
  Text_ActivityName {
   widgetClassPath=/Script/FortniteUI.FortActivityScalingTextBlock
   bInherited=true
   bIsVariable=true
   parent=VerticalBox_1
   widget=Text_ActivityName
   properties {
    autoWrapText=true
    scaling=/Game/UI/DiscoveryShared/ActivityNameFontScaleData.ActivityNameFontScaleData_C
    text="{Game Mode}{Game Mode}{Game Mode}{Game Mode}"
    textTransformPolicy=ToUpper
    wrapTextAt=730
    wrappingPolicy=AllowPerCharacterWrapping
    font {
     bForceMonospaced=false
     bMaterialIsStencil=false
     fontMaterial=null
     fontObject=&HeadingNow
     letterSpacing=-40
     monospacedWidth=1
     size=34.5
     skewAmount=0
     typefaceFontName=86 Bold Italic
     outlineSettings {
      bApplyOutlineToDropShadows=false
      bMiteredCorners=false
      bSeparateFillAlpha=false
      outlineMaterial=/FortUILibrary/MaterialInstances/Text/MI_UI_Block_Text_Outline
      outlineSize=4
      outlineColor=<rgba 0 0 0 1>
     }
    }
    shadowOffset=<0 0>
    shapedTextOptions {bOverride_TextFlowDirection=true, bOverride_TextShapingMethod=false, textFlowDirection=Culture, textShapingMethod=Auto}
    strikeBrush {
     bIsDynamicallyLoaded=false
     drawAs=NoDrawType
     imageType=NoImage
     mirroring=NoMirror
     resourceName=None
     resourceObject=null
     tiling=NoTile
     imageSize=<32 32>
     margin=<ltrb 0 0 0 0>
     outlineSettings {
      bUseBrushTransparency=false
      roundingType=HalfHeightRadius
      width=0
      color {colorUseRule=UseColor_Specified, specifiedColor=<rgba 0 0 0 0>}
      cornerRadii=<0 0 0 0>
     }
     tintColor {colorUseRule=UseColor_Specified, specifiedColor=<rgba 1 1 1 1>}
     uVRegion {bIsValid=false, max=<0 0>, min=<0 0>}
    }
   }
   slot {class=VerticalBoxSlot, properties {padding=<ltrb 0 0 0 16>}}
  }
  HorizontalBox_0 {
   widgetClassPath=&HorizontalBox
   parent=VerticalBox_1
   widget=HorizontalBox_0
   slot {class=VerticalBoxSlot, properties {horizontalAlignment=HAlign_Left, padding=<ltrb 0 0 0 24>}}
  }
  Text_ActivityOrigin {
   widgetClassPath=&CommonRichTextBlock
   bInherited=true
   bIsVariable=true
   parent=HorizontalBox_0
   widget=Text_ActivityOrigin
   properties {
    mobileTextBlockScale=1.3
    text="Created by <CreatorStyle>{Creator}</> "
    textStyleSet=/Game/UI/DiscoveryShared/DT_UI_RichDetails_Tags
    defaultTextStyleOverride {
     overflowPolicy=Clip
     transformPolicy=None
     colorAndOpacity {colorUseRule=UseColor_Specified, specifiedColor=<rgba 0.0331 0.402 1 1>}
     font {
      bForceMonospaced=false
      bMaterialIsStencil=false
      fontMaterial=null
      fontObject=&BurbankSmall
      letterSpacing=0
      monospacedWidth=1
      size=15
      skewAmount=0
      typefaceFontName=Bold
      outlineSettings {
       bApplyOutlineToDropShadows=false
       bMiteredCorners=false
       bSeparateFillAlpha=false
       outlineMaterial=null
       outlineSize=0
       outlineColor=<rgba 0 0 0 1>
      }
     }
     highlightColor {colorUseRule=UseColor_Specified, specifiedColor=<rgba 0 0 0 1>}
     highlightShape {
      bIsDynamicallyLoaded=false
      drawAs=Image
      imageType=NoImage
      mirroring=NoMirror
      resourceName=None
      resourceObject=null
      tiling=NoTile
      imageSize=<32 32>
      margin=<ltrb 0 0 0 0>
      outlineSettings {
       bUseBrushTransparency=false
       roundingType=HalfHeightRadius
       width=0
       color {colorUseRule=UseColor_Specified, specifiedColor=<rgba 0 0 0 0>}
       cornerRadii=<0 0 0 0>
      }
      tintColor {colorUseRule=UseColor_Specified, specifiedColor=<rgba 1 1 1 1>}
      uVRegion {bIsValid=false, max=<0 0>, min=<0 0>}
     }
     selectedBackgroundColor {colorUseRule=UseColor_Foreground, specifiedColor=<rgba 1 0 1 1>}
     shadowColorAndOpacity=<rgba 0 0 0 1>
     shadowOffset=<0 0>
     strikeBrush {
      bIsDynamicallyLoaded=false
      drawAs=Image
      imageType=NoImage
      mirroring=NoMirror
      resourceName=None
      resourceObject=null
      tiling=NoTile
      imageSize=<32 32>
      margin=<ltrb 0 0 0 0>
      outlineSettings {
       bUseBrushTransparency=false
       roundingType=HalfHeightRadius
       width=0
       color {colorUseRule=UseColor_Specified, specifiedColor=<rgba 0 0 0 0>}
       cornerRadii=<0 0 0 0>
      }
      tintColor {colorUseRule=UseColor_Specified, specifiedColor=<rgba 1 1 1 1>}
      uVRegion {bIsValid=false, max=<0 0>, min=<0 0>}
     }
     underlineBrush {
      bIsDynamicallyLoaded=false
      drawAs=Image
      imageType=NoImage
      mirroring=NoMirror
      resourceName=None
      resourceObject=null
      tiling=NoTile
      imageSize=<32 32>
      margin=<ltrb 0 0 0 0>
      outlineSettings {
       bUseBrushTransparency=false
       roundingType=HalfHeightRadius
       width=0
       color {colorUseRule=UseColor_Specified, specifiedColor=<rgba 0 0 0 0>}
       cornerRadii=<0 0 0 0>
      }
      tintColor {colorUseRule=UseColor_Specified, specifiedColor=<rgba 1 1 1 1>}
      uVRegion {bIsValid=false, max=<0 0>, min=<0 0>}
     }
    }
    margin=<ltrb 0 -4 0 -4>
   }
   slot {class=HorizontalBoxSlot, properties {padding=<ltrb 0 0 32 0>}}
  }
  Text_LinkCode {
   widgetClassPath=&CommonRichTextBlock
   bIsVariable=true
   parent=HorizontalBox_0
   widget=Text_LinkCode
   properties {
    mobileTextBlockScale=1.3
    text="Island Code <CreatorStyle>{XXXX-XXXX-XXXX-XXXX}</> "
    textStyleSet=&DT_UI_RichDetails_Tags
    defaultTextStyleOverride {
     overflowPolicy=Clip
     transformPolicy=None
     colorAndOpacity {colorUseRule=UseColor_Specified, specifiedColor=<rgba 0.0331 0.402 1 1>}
     font {
      bForceMonospaced=false
      bMaterialIsStencil=false
      fontMaterial=null
      fontObject=&BurbankSmall
      letterSpacing=0
      monospacedWidth=1
      size=15
      skewAmount=0
      typefaceFontName=Bold
      outlineSettings {
       bApplyOutlineToDropShadows=false
       bMiteredCorners=false
       bSeparateFillAlpha=false
       outlineMaterial=null
       outlineSize=0
       outlineColor=<rgba 0 0 0 1>
      }
     }
     highlightColor {colorUseRule=UseColor_Specified, specifiedColor=<rgba 0 0 0 1>}
     highlightShape {
      bIsDynamicallyLoaded=false
      drawAs=Image
      imageType=NoImage
      mirroring=NoMirror
      resourceName=None
      resourceObject=null
      tiling=NoTile
      imageSize=<32 32>
      margin=<ltrb 0 0 0 0>
      outlineSettings {
       bUseBrushTransparency=false
       roundingType=HalfHeightRadius
       width=0
       color {colorUseRule=UseColor_Specified, specifiedColor=<rgba 0 0 0 0>}
       cornerRadii=<0 0 0 0>
      }
      tintColor {colorUseRule=UseColor_Specified, specifiedColor=<rgba 1 1 1 1>}
      uVRegion {bIsValid=false, max=<0 0>, min=<0 0>}
     }
     selectedBackgroundColor {colorUseRule=UseColor_Foreground, specifiedColor=<rgba 1 0 1 1>}
     shadowColorAndOpacity=<rgba 0 0 0 1>
     shadowOffset=<0 0>
     strikeBrush {
      bIsDynamicallyLoaded=false
      drawAs=Image
      imageType=NoImage
      mirroring=NoMirror
      resourceName=None
      resourceObject=null
      tiling=NoTile
      imageSize=<32 32>
      margin=<ltrb 0 0 0 0>
      outlineSettings {
       bUseBrushTransparency=false
       roundingType=HalfHeightRadius
       width=0
       color {colorUseRule=UseColor_Specified, specifiedColor=<rgba 0 0 0 0>}
       cornerRadii=<0 0 0 0>
      }
      tintColor {colorUseRule=UseColor_Specified, specifiedColor=<rgba 1 1 1 1>}
      uVRegion {bIsValid=false, max=<0 0>, min=<0 0>}
     }
     underlineBrush {
      bIsDynamicallyLoaded=false
      drawAs=Image
      imageType=NoImage
      mirroring=NoMirror
      resourceName=None
      resourceObject=null
      tiling=NoTile
      imageSize=<32 32>
      margin=<ltrb 0 0 0 0>
      outlineSettings {
       bUseBrushTransparency=false
       roundingType=HalfHeightRadius
       width=0
       color {colorUseRule=UseColor_Specified, specifiedColor=<rgba 0 0 0 0>}
       cornerRadii=<0 0 0 0>
      }
      tintColor {colorUseRule=UseColor_Specified, specifiedColor=<rgba 1 1 1 1>}
      uVRegion {bIsValid=false, max=<0 0>, min=<0 0>}
     }
    }
    margin=<ltrb 0 -4 0 -4>
   }
   slot {class=HorizontalBoxSlot}
  }
  EntryBox_ActivityTags {
   widgetClassPath=&DynamicEntryBox
   bInherited=true
   parent=VerticalBox_1
   widget=EntryBox_ActivityTags
   properties {
    entryHorizontalAlignment=HAlign_Right
    entryWidgetClass=/Game/UI/DiscoveryShared/ActivityBrowserTag.ActivityBrowserTag_C
    flowDirectionPreference=Culture
    numDesignerPreviewEntries=4
    entrySpacing=<-2 8>
   }
   slot {class=VerticalBoxSlot, properties {horizontalAlignment=HAlign_Left, padding=<ltrb 0 0 0 16>}}
  }
  StatusEntryBox_Redesign {
   widgetClassPath=&DynamicEntryBox
   bIsVariable=true
   parent=VerticalBox_1
   widget=StatusEntryBox_Redesign
   properties {
    entryWidgetClass=&ActivityBrowserStatusFavoriteButton_C
    numDesignerPreviewEntries=2
    visibility=Collapsed
    entrySpacing=<25 0>
   }
   slot {class=VerticalBoxSlot, properties {horizontalAlignment=HAlign_Left, padding=<ltrb 0 0 0 8>}}
  }
  WarningEntryBox {
   widgetClassPath=&DynamicEntryBox
   bIsVariable=true
   parent=VerticalBox_1
   widget=WarningEntryBox
   properties {
    entryBoxType=Vertical
    entryWidgetClass=&ActivityBrowserStatusWarning_C
    numDesignerPreviewEntries=1
    entrySpacing=<25 0>
   }
   slot {class=VerticalBoxSlot, properties {horizontalAlignment=HAlign_Left, padding=<ltrb 0 0 0 16>}}
  }
  StatusEntryBox {
   widgetClassPath=&DynamicEntryBox
   bIsVariable=true
   parent=VerticalBox_1
   widget=StatusEntryBox
   properties {
    entryWidgetClass=&ActivityBrowserStatusFavorite_C
    numDesignerPreviewEntries=2
    visibility=Collapsed
    entrySpacing=<25 0>
   }
   slot {class=VerticalBoxSlot, properties {horizontalAlignment=HAlign_Left, padding=<ltrb 0 0 0 8>}}
  }
  DisabledWarningHB {
   widgetClassPath=&HorizontalBox
   bIsVariable=true
   parent=VerticalBox_1
   widget=DisabledWarningHB
   slot {class=VerticalBoxSlot, properties {horizontalAlignment=HAlign_Left, padding=<ltrb 0 0 0 16>}}
  }
  Overlay {
   widgetClassPath=&Overlay
   parent=DisabledWarningHB
   widget=Overlay
   slot {class=HorizontalBoxSlot, properties {horizontalAlignment=HAlign_Left, verticalAlignment=VAlign_Center}}
  }
  WarningTriangle_Pulse {
   widgetClassPath=&Image
   parent=Overlay
   widget=WarningTriangle_Pulse
   properties {
    renderOpacity=0
    visibility=SelfHitTestInvisible
    brush {
     bIsDynamicallyLoaded=false
     drawAs=Image
     imageType=NoImage
     mirroring=NoMirror
     resourceName=None
     resourceObject=/Game/UI/DiscoveryShared/Art/M_UI_Discovery_WarningTriangle_Pulse
     tiling=NoTile
     imageSize=<40 40>
     margin=<ltrb 0 0 0 0>
     outlineSettings {
      bUseBrushTransparency=false
      roundingType=HalfHeightRadius
      width=0
      color {colorUseRule=UseColor_Specified, specifiedColor=<rgba 0 0 0 0>}
      cornerRadii=<0 0 0 0>
     }
     tintColor {colorUseRule=UseColor_Specified, specifiedColor=<rgba 1 1 1 1>}
     uVRegion {bIsValid=false, max=<0 0>, min=<0 0>}
    }
   }
   slot {class=OverlaySlot, properties {horizontalAlignment=HAlign_Fill, verticalAlignment=VAlign_Fill}}
  }
  IconOvr {
   widgetClassPath=&Overlay
   parent=Overlay
   widget=IconOvr
   properties {visibility=HitTestInvisible}
   slot {class=OverlaySlot, properties {horizontalAlignment=HAlign_Fill, verticalAlignment=VAlign_Fill}}
  }
  WarningTriangle {
   widgetClassPath=&Image
   parent=IconOvr
   widget=WarningTriangle
   properties {
    visibility=SelfHitTestInvisible
    brush {
     bIsDynamicallyLoaded=false
     drawAs=Image
     imageType=NoImage
     mirroring=NoMirror
     resourceName=None
     resourceObject=/Game/UI/DiscoveryShared/Art/M_UI_Discovery_WarningTriangle
     tiling=NoTile
     imageSize=<32 32>
     margin=<ltrb 0 0 0 0>
     outlineSettings {
      bUseBrushTransparency=false
      roundingType=HalfHeightRadius
      width=0
      color {colorUseRule=UseColor_Specified, specifiedColor=<rgba 0 0 0 0>}
      cornerRadii=<0 0 0 0>
     }
     tintColor {colorUseRule=UseColor_Specified, specifiedColor=<rgba 1 1 1 1>}
     uVRegion {bIsValid=false, max=<0 0>, min=<0 0>}
    }
   }
   slot {class=OverlaySlot, properties {horizontalAlignment=HAlign_Fill, verticalAlignment=VAlign_Fill}}
  }
  Exclamation {
   widgetClassPath=&CommonTextBlock
   parent=IconOvr
   widget=Exclamation
   properties {
    text="!"
    colorAndOpacity {colorUseRule=UseColor_Specified, specifiedColor=<rgba 0 0 0 1>}
    font {
     bForceMonospaced=false
     bMaterialIsStencil=false
     fontMaterial=null
     fontObject=&Burbank
     letterSpacing=0
     monospacedWidth=1
     size=18
     skewAmount=0
     typefaceFontName=BurbankRegular-Black
     outlineSettings {
      bApplyOutlineToDropShadows=false
      bMiteredCorners=false
      bSeparateFillAlpha=false
      outlineMaterial=null
      outlineSize=0
      outlineColor=<rgba 0 0 0 1>
     }
    }
    margin=<ltrb 0 -4 0 -4>
    renderTransform {angle=0, scale=<1 1>, shear=<4 0>, translation=<-1 5>}
    shadowOffset=<0 0>
   }
   slot {class=OverlaySlot, properties {horizontalAlignment=HAlign_Center, verticalAlignment=VAlign_Center}}
  }
  DisabledWarningMessage {
   widgetClassPath=&CommonTextBlock
   bIsVariable=true
   parent=DisabledWarningHB
   widget=DisabledWarningMessage
   properties {
    text="{Disabled warning message. Lorem ipsum.}"
    textTransformPolicy=ToUpper
    wrapTextAt=544
    font {
     bForceMonospaced=false
     bMaterialIsStencil=false
     fontMaterial=null
     fontObject=&HeadingNow
     letterSpacing=10
     monospacedWidth=1
     size=15.75
     skewAmount=0
     typefaceFontName=74 Regular
     outlineSettings {
      bApplyOutlineToDropShadows=false
      bMiteredCorners=false
      bSeparateFillAlpha=false
      outlineMaterial=&MI_UI_Block_Text_Outline
      outlineSize=2
      outlineColor=<rgba 0 0 0 1>
     }
    }
    shadowOffset=<0 0>
    shapedTextOptions {bOverride_TextFlowDirection=true, bOverride_TextShapingMethod=false, textFlowDirection=Culture, textShapingMethod=Auto}
   }
   slot {class=HorizontalBoxSlot, properties {verticalAlignment=VAlign_Center, padding=<ltrb 8 0 0 0>, size {sizeRule=Fill, value=1}}}
  }
  NamedSlot_AdditionalDetail {
   widgetClassPath=&NamedSlot
   bIsVariable=true
   parent=VerticalBox_1
   widget=NamedSlot_AdditionalDetail
   properties {bExposeOnInstanceOnly=true, visibility=Collapsed}
   slot {class=VerticalBoxSlot}
  }
  VerticalBox_ModeSelectionAvailable {
   widgetClassPath=&VerticalBox
   bIsVariable=true
   parent=NamedSlot_AdditionalDetail
   widget=VerticalBox_ModeSelectionAvailable
   slot {class=PanelSlot}
  }
  HorizontalBox_ModeSelectionAvailable {
   widgetClassPath=&HorizontalBox
   bIsVariable=true
   parent=VerticalBox_ModeSelectionAvailable
   widget=HorizontalBox_ModeSelectionAvailable
   slot {class=VerticalBoxSlot, properties {horizontalAlignment=HAlign_Left, padding=<ltrb 0 0 0 10>}}
  }
  Image_77 {
   widgetClassPath=&Image
   parent=HorizontalBox_ModeSelectionAvailable
   widget=Image_77
   properties {
    visibility=SelfHitTestInvisible
    brush {
     bIsDynamicallyLoaded=false
     drawAs=Image
     imageType=NoImage
     mirroring=NoMirror
     resourceName=None
     resourceObject=&T_UI_Discovery_SubModeIcon
     tiling=NoTile
     imageSize=<32 32>
     margin=<ltrb 0 0 0 0>
     outlineSettings {
      bUseBrushTransparency=false
      roundingType=HalfHeightRadius
      width=0
      color {colorUseRule=UseColor_Specified, specifiedColor=<rgba 0 0 0 0>}
      cornerRadii=<0 0 0 0>
     }
     tintColor {colorUseRule=UseColor_Specified, specifiedColor=<rgba 1 1 1 1>}
     uVRegion {bIsValid=false, max=<0 0>, min=<0 0>}
    }
    colorAndOpacity=<rgba 0.2159 0.6584 1 1>
   }
   slot {class=HorizontalBoxSlot, properties {horizontalAlignment=HAlign_Left, verticalAlignment=VAlign_Center, padding=<ltrb 0 0 10 0>}}
  }
  CommonTextBlock_75 {
   widgetClassPath=&CommonTextBlock
   parent=HorizontalBox_ModeSelectionAvailable
   widget=CommonTextBlock_75
   properties {
    mobileFontSizeMultiplier=1.2
    text=Multiple Modes
    textTransformPolicy=ToUpper
    colorAndOpacity {colorUseRule=UseColor_Specified, specifiedColor=<rgba 0.2159 0.6584 1 1>}
    font {
     bForceMonospaced=false
     bMaterialIsStencil=false
     fontMaterial=null
     fontObject=&BurbankSmall
     letterSpacing=0
     monospacedWidth=1
     size=18
     skewAmount=0
     typefaceFontName=Black
     outlineSettings {
      bApplyOutlineToDropShadows=false
      bMiteredCorners=false
      bSeparateFillAlpha=false
      outlineMaterial=null
      outlineSize=0
      outlineColor=<rgba 0 0 0 1>
     }
    }
    shadowOffset=<0 0>
   }
   slot {class=HorizontalBoxSlot}
  }
  HorizontalBox_1 {
   widgetClassPath=&HorizontalBox
   parent=VerticalBox_ModeSelectionAvailable
   widget=HorizontalBox_1
   slot {class=VerticalBoxSlot, properties {horizontalAlignment=HAlign_Left}}
  }
  Button_ShowSubModeList {
   widgetClassPath=&WBP_UIKit_Button_Regular
   bIsVariable=true
   parent=HorizontalBox_1
   widget=Button_ShowSubModeList
   properties {
    bIsFocusable=false
    minWidth=400
    text="{Text}"
    triggeringInputAction {dataTable=/FortUILibrary/Foundation/ActivityBrowserInputActionTable, rowName=ShowSubModes_DetailsModal}
   }
   slot {class=HorizontalBoxSlot}
  }
  Spacer_70 {
   widgetClassPath=&Spacer
   bIsVariable=true
   parent=HorizontalBox_1
   widget=Spacer_70
   properties {size=<5 1>}
   slot {class=HorizontalBoxSlot}
  }
  DiscoveryTooltip {
   widgetClassPath=/Game/UI/Tooltips/WBP_PopupTooltip
   bIsVariable=true
   parent=HorizontalBox_1
   widget=DiscoveryTooltip
   properties {
    arrowHorizontalAlignment=HAlign_Left
    arrowVerticalAlignment=VAlign_Center
    firstLineText=Change mode here
    secondLineRichText=""
    arrowPadding=<ltrb 0 0 0 0>
    arrowTransform {angle=-90, scale=<1 1>, shear=<0 0>, translation=<-20 0>}
    renderTransformPivot=<0 0.5>
   }
   slot {class=HorizontalBoxSlot, properties {verticalAlignment=VAlign_Center, padding=<ltrb 16 0 0 0>}}
  }
  Spacer_63 {
   widgetClassPath=&Spacer
   bIsVariable=true
   parent=VerticalBox_ModeSelectionAvailable
   widget=Spacer_63
   properties {size=<1 16>}
   slot {class=VerticalBoxSlot}
  }
  HorizontalBox_2 {
   widgetClassPath=&HorizontalBox
   parent=Overlay_1
   widget=HorizontalBox_2
   properties {flowDirectionPreference=Culture}
   slot {class=OverlaySlot, properties {horizontalAlignment=HAlign_Fill}}
  }
  WidgetSwitcher_ButtonRedesign {
   widgetClassPath=&WidgetSwitcher
   bIsVariable=true
   parent=HorizontalBox_2
   widget=WidgetSwitcher_ButtonRedesign
   slot {class=HorizontalBoxSlot, properties {horizontalAlignment=HAlign_Left, padding=<ltrb 0 0 8 0>, size {sizeRule=Fill, value=1}}}
  }
  HorizontalBox_Buttons {
   widgetClassPath=&HorizontalBox
   parent=WidgetSwitcher_ButtonRedesign
   widget=HorizontalBox_Buttons
   slot {class=WidgetSwitcherSlot}
  }
  AcceptButtonOvr {
   widgetClassPath=&Overlay
   parent=HorizontalBox_Buttons
   widget=AcceptButtonOvr
   slot {class=HorizontalBoxSlot, properties {verticalAlignment=VAlign_Center, padding=<ltrb 0 0 16 0>}}
  }
  Button_Accept {
   widgetClassPath=/FortUILibrary/UIKit/Buttons/WBP_UIKit_Button_Loud
   bInherited=true
   bIsVariable=true
   parent=AcceptButtonOvr
   widget=Button_Accept
   properties {
    buttonSize=Large
    minHeight=88
    minWidth=336
    text=Play
    triggeringInputAction {dataTable=&InputActionDataTable, rowName=GenericForward}
   }
   slot {class=OverlaySlot, properties {horizontalAlignment=HAlign_Fill, verticalAlignment=VAlign_Center}}
  }
  HB_SideButtons {
   widgetClassPath=&HorizontalBox
   parent=HorizontalBox_Buttons
   widget=HB_SideButtons
   slot {class=HorizontalBoxSlot, properties {verticalAlignment=VAlign_Center, padding=<ltrb 0 0 16 0>}}
  }
  Button_Favorite {
   widgetClassPath=/Game/UI/DiscoveryShared/GameDetails/WBP_GameDetails_Toggle
   bInherited=true
   bIsVariable=true
   parent=HB_SideButtons
   widget=Button_Favorite
   properties {triggeringInputAction {dataTable=&ActivityBrowserInputActionTable, rowName=ToggleFavorite}}
   slot {class=HorizontalBoxSlot}
  }
  Button_Details {
   widgetClassPath=&/WBP_Discovery_DetailsButton
   bInherited=true
   bIsVariable=true
   parent=HB_SideButtons
   widget=Button_Details
   properties {bIsFocusable=true, imageSize=<72 72>}
   slot {class=HorizontalBoxSlot, properties {horizontalAlignment=HAlign_Right, verticalAlignment=VAlign_Center, padding=<ltrb 8 0 0 0>}}
  }
  WBP_CCUCount {
   widgetClassPath=/FortniteCoreUI/UI/DiscoveryShared/GameDetails/WBP_IslandLabelAndDetails
   bIsVariable=true
   parent=HorizontalBox_Buttons
   widget=WBP_CCUCount
   properties {icon=/FortniteCoreUI/UI/DiscoveryShared/Art/MaterialInstances/MI_UI_GameDetails_CCU, label_Text=Active Players}
   slot {class=HorizontalBoxSlot, properties {verticalAlignment=VAlign_Center, padding=<ltrb 16 0 0 0>}}
  }
  WBP_IslandCode {
   widgetClassPath=&WBP_IslandLabelAndDetails
   bIsVariable=true
   parent=HorizontalBox_Buttons
   widget=WBP_IslandCode
   properties {hasIcon=false, label_Text=Island Code, visibility=Collapsed}
   slot {class=HorizontalBoxSlot, properties {verticalAlignment=VAlign_Center, padding=<ltrb 16 0 0 0>}}
  }
  SqueegeeSlot_LockInfo {
   widgetClassPath=/Game/UI/Plugin/Squeegee/SqueegeeInjectionSlot
   bIsVariable=true
   parent=WidgetSwitcher_ButtonRedesign
   widget=SqueegeeSlot_LockInfo
   properties {pluginId=Squeegee, pluginWidgetId=Game Mode Info}
   slot {class=WidgetSwitcherSlot}
  }
  SqueegeeSlot_Card {
   widgetClassPath=&SqueegeeInjectionSlot
   bIsVariable=true
   parent=HorizontalBox_2
   widget=SqueegeeSlot_Card
   properties {pluginId=Squeegee, pluginWidgetId=RatingCard}
   slot {class=HorizontalBoxSlot, properties {horizontalAlignment=HAlign_Right, verticalAlignment=VAlign_Bottom}}
  }
  StackBox_1 {
   widgetClassPath=&StackBox
   parent=VerticalBox_0
   widget=StackBox_1
   properties {orientation=Orient_Vertical}
   slot {class=VerticalBoxSlot, properties {verticalAlignment=VAlign_Bottom, padding=<ltrb 0 48 0 0>}}
  }
  Text_Details {
   widgetClassPath=&CommonTextBlock
   bIsVariable=true
   parent=StackBox_1
   widget=Text_Details
   properties {
    autoWrapText=true
    lineHeightPercentage=1.28
    mobileFontSizeMultiplier=1.2
    text=DETAILS
    wrapTextAt=1000
    wrappingPolicy=AllowPerCharacterWrapping
    font {
     bForceMonospaced=false
     bMaterialIsStencil=false
     fontMaterial=null
     fontObject=&HeadingNow
     letterSpacing=-40
     monospacedWidth=1
     size=18
     skewAmount=0
     typefaceFontName=86 Bold Italic
     outlineSettings {
      bApplyOutlineToDropShadows=false
      bMiteredCorners=false
      bSeparateFillAlpha=false
      outlineMaterial=null
      outlineSize=0
      outlineColor=<rgba 0 0 0 1>
     }
    }
    shadowOffset=<0 0>
    shapedTextOptions {bOverride_TextFlowDirection=true, bOverride_TextShapingMethod=false, textFlowDirection=Culture, textShapingMethod=Auto}
    strikeBrush {
     bIsDynamicallyLoaded=false
     drawAs=NoDrawType
     imageType=NoImage
     mirroring=NoMirror
     resourceName=None
     resourceObject=null
     tiling=NoTile
     imageSize=<32 32>
     margin=<ltrb 0 0 0 0>
     outlineSettings {
      bUseBrushTransparency=false
      roundingType=HalfHeightRadius
      width=0
      color {colorUseRule=UseColor_Specified, specifiedColor=<rgba 0 0 0 0>}
      cornerRadii=<0 0 0 0>
     }
     tintColor {colorUseRule=UseColor_Specified, specifiedColor=<rgba 1 1 1 1>}
     uVRegion {bIsValid=false, max=<0 0>, min=<0 0>}
    }
   }
   slot {class=StackBoxSlot, properties {padding=<ltrb 0 0 0 8>}}
  }
  DescriptionClippingBorder {
   widgetClassPath=&Border
   parent=StackBox_1
   widget=DescriptionClippingBorder
   properties {
    clipping=ClipToBounds
    visibility=SelfHitTestInvisible
    background {
     bIsDynamicallyLoaded=false
     drawAs=NoDrawType
     imageType=NoImage
     mirroring=NoMirror
     resourceName=None
     resourceObject=null
     tiling=NoTile
     imageSize=<32 32>
     margin=<ltrb 0 0 0 0>
     outlineSettings {
      bUseBrushTransparency=false
      roundingType=HalfHeightRadius
      width=0
      color {colorUseRule=UseColor_Specified, specifiedColor=<rgba 0 0 0 0>}
      cornerRadii=<0 0 0 0>
     }
     tintColor {colorUseRule=UseColor_Specified, specifiedColor=<rgba 1 1 1 1>}
     uVRegion {bIsValid=false, max=<0 0>, min=<0 0>}
    }
    padding=<ltrb 0 0 0 0>
   }
   slot {class=StackBoxSlot}
  }
  Text_ActivityDescription {
   widgetClassPath=&CommonTextBlock
   bInherited=true
   bIsVariable=true
   parent=DescriptionClippingBorder
   widget=Text_ActivityDescription
   properties {
    autoWrapText=true
    lineHeightPercentage=1.28
    mobileFontSizeMultiplier=1.2
    text="[Game Mode Description. Wrapping is set to per character, text widget clips to bounds on itself. Lorem ipsum dolor sit amet, consectetur adipiscing elit."
    wrapTextAt=1000
    wrappingPolicy=AllowPerCharacterWrapping
    font {
     bForceMonospaced=false
     bMaterialIsStencil=false
     fontMaterial=null
     fontObject=&HeadingNow
     letterSpacing=0
     monospacedWidth=1
     size=15.75
     skewAmount=0
     typefaceFontName=74 Regular
     outlineSettings {
      bApplyOutlineToDropShadows=false
      bMiteredCorners=false
      bSeparateFillAlpha=false
      outlineMaterial=null
      outlineSize=0
      outlineColor=<rgba 0 0 0 1>
     }
    }
    shadowOffset=<0 0>
    shapedTextOptions {bOverride_TextFlowDirection=true, bOverride_TextShapingMethod=false, textFlowDirection=Culture, textShapingMethod=Auto}
    strikeBrush {
     bIsDynamicallyLoaded=false
     drawAs=NoDrawType
     imageType=NoImage
     mirroring=NoMirror
     resourceName=None
     resourceObject=null
     tiling=NoTile
     imageSize=<32 32>
     margin=<ltrb 0 0 0 0>
     outlineSettings {
      bUseBrushTransparency=false
      roundingType=HalfHeightRadius
      width=0
      color {colorUseRule=UseColor_Specified, specifiedColor=<rgba 0 0 0 0>}
      cornerRadii=<0 0 0 0>
     }
     tintColor {colorUseRule=UseColor_Specified, specifiedColor=<rgba 1 1 1 1>}
     uVRegion {bIsValid=false, max=<0 0>, min=<0 0>}
    }
   }
   slot {class=BorderSlot, properties {padding=<ltrb 0 0 0 0>}}
  }
  TODO_DeleteWhenUnbound {
   widgetClassPath=&WrapBox
   parent=GridPanel_DynamicDetails
   widget=TODO_DeleteWhenUnbound
   properties {visibility=Collapsed}
   slot {class=GridSlot}
  }
  ScrollBox_ActivityDescription {
   widgetClassPath=&ScrollBox
   bInherited=true
   parent=TODO_DeleteWhenUnbound
   widget=ScrollBox_ActivityDescription
   properties {
    allowOverscroll=false
    bAllowRightClickDragScrolling=false
    scrollBarVisibility=Collapsed
    widgetBarStyle {
     thickness=16
     draggedThumbImage {
      bIsDynamicallyLoaded=false
      drawAs=Image
      imageType=FullColor
      mirroring=NoMirror
      resourceName=../../../Engine/Content/Slate/Common/Scrollbar_Thumb.png
      resourceObject=/FortUILibrary/Foundation/Art_ToolKit/AngledBoxes/M_UI_ItemShop_ScrollBar_Drag
      tiling=NoTile
      imageSize=<0 0>
      margin=<ltrb 0.25 0.25 0.25 0.25>
      outlineSettings {
       bUseBrushTransparency=false
       roundingType=HalfHeightRadius
       width=0
       color {colorUseRule=UseColor_Specified, specifiedColor=<rgba 0 0 0 0>}
       cornerRadii=<0 0 0 0>
      }
      tintColor {colorUseRule=UseColor_Specified, specifiedColor=<rgba 1 1 1 1>}
      uVRegion {bIsValid=false, max=<0 0>, min=<0 0>}
     }
     horizontalBackgroundImage {
      bIsDynamicallyLoaded=false
      drawAs=Image
      imageType=NoImage
      mirroring=NoMirror
      resourceName=None
      resourceObject=/FortUILibrary/Foundation/Art_ToolKit/AngledBoxes/M_UI_ItemShop_ScrollBar_Backing
      tiling=NoTile
      imageSize=<0 0>
      margin=<ltrb 0 0 0 0>
      outlineSettings {
       bUseBrushTransparency=false
       roundingType=HalfHeightRadius
       width=0
       color {colorUseRule=UseColor_Specified, specifiedColor=<rgba 0 0 0 0>}
       cornerRadii=<0 0 0 0>
      }
      tintColor {colorUseRule=UseColor_Specified, specifiedColor=<rgba 1 1 1 1>}
      uVRegion {bIsValid=false, max=<0 0>, min=<0 0>}
     }
     horizontalBottomSlotImage {
      bIsDynamicallyLoaded=false
      drawAs=NoDrawType
      imageType=FullColor
      mirroring=NoMirror
      resourceName=../../../Engine/Content/Slate/Common/Scrollbar_Background_Horizontal.png
      resourceObject=null
      tiling=NoTile
      imageSize=<8 8>
      margin=<ltrb 0 0 0 0>
      outlineSettings {
       bUseBrushTransparency=false
       roundingType=HalfHeightRadius
       width=0
       color {colorUseRule=UseColor_Specified, specifiedColor=<rgba 0 0 0 0>}
       cornerRadii=<0 0 0 0>
      }
      tintColor {colorUseRule=UseColor_Specified, specifiedColor=<rgba 1 1 1 1>}
      uVRegion {bIsValid=false, max=<0 0>, min=<0 0>}
     }
     horizontalTopSlotImage {
      bIsDynamicallyLoaded=false
      drawAs=NoDrawType
      imageType=FullColor
      mirroring=NoMirror
      resourceName=../../../Engine/Content/Slate/Common/Scrollbar_Background_Horizontal.png
      resourceObject=null
      tiling=NoTile
      imageSize=<8 8>
      margin=<ltrb 0 0 0 0>
      outlineSettings {
       bUseBrushTransparency=false
       roundingType=HalfHeightRadius
       width=0
       color {colorUseRule=UseColor_Specified, specifiedColor=<rgba 0 0 0 0>}
       cornerRadii=<0 0 0 0>
      }
      tintColor {colorUseRule=UseColor_Specified, specifiedColor=<rgba 1 1 1 1>}
      uVRegion {bIsValid=false, max=<0 0>, min=<0 0>}
     }
     hoveredThumbImage {
      bIsDynamicallyLoaded=false
      drawAs=Image
      imageType=FullColor
      mirroring=NoMirror
      resourceName=../../../Engine/Content/Slate/Common/Scrollbar_Thumb.png
      resourceObject=/FortUILibrary/Foundation/Art_ToolKit/AngledBoxes/M_UI_ItemShop_ScrollBar_Hover
      tiling=NoTile
      imageSize=<0 0>
      margin=<ltrb 0.25 0.25 0.25 0.25>
      outlineSettings {
       bUseBrushTransparency=false
       roundingType=HalfHeightRadius
       width=0
       color {colorUseRule=UseColor_Specified, specifiedColor=<rgba 0 0 0 0>}
       cornerRadii=<0 0 0 0>
      }
      tintColor {colorUseRule=UseColor_Specified, specifiedColor=<rgba 1 1 1 1>}
      uVRegion {bIsValid=false, max=<0 0>, min=<0 0>}
     }
     normalThumbImage {
      bIsDynamicallyLoaded=false
      drawAs=Image
      imageType=FullColor
      mirroring=NoMirror
      resourceName=../../../Engine/Content/Slate/Common/Scrollbar_Thumb.png
      resourceObject=/FortUILibrary/Foundation/Art_ToolKit/AngledBoxes/M_UI_ItemShop_ScrollBar_Base
      tiling=NoTile
      imageSize=<0 0>
      margin=<ltrb 0.25 0.25 0.25 0.25>
      outlineSettings {
       bUseBrushTransparency=false
       roundingType=HalfHeightRadius
       width=0
       color {colorUseRule=UseColor_Specified, specifiedColor=<rgba 0 0 0 0>}
       cornerRadii=<0 0 0 0>
      }
      tintColor {colorUseRule=UseColor_Specified, specifiedColor=<rgba 1 1 1 1>}
      uVRegion {bIsValid=false, max=<0 0>, min=<0 0>}
     }
     verticalBackgroundImage {
      bIsDynamicallyLoaded=false
      drawAs=Image
      imageType=NoImage
      mirroring=NoMirror
      resourceName=None
      resourceObject=&M_UI_ItemShop_ScrollBar_Backing
      tiling=NoTile
      imageSize=<0 0>
      margin=<ltrb 0 0 0 0>
      outlineSettings {
       bUseBrushTransparency=false
       roundingType=HalfHeightRadius
       width=0
       color {colorUseRule=UseColor_Specified, specifiedColor=<rgba 0 0 0 0>}
       cornerRadii=<0 0 0 0>
      }
      tintColor {colorUseRule=UseColor_Specified, specifiedColor=<rgba 1 1 1 1>}
      uVRegion {bIsValid=false, max=<0 0>, min=<0 0>}
     }
     verticalBottomSlotImage {
      bIsDynamicallyLoaded=false
      drawAs=NoDrawType
      imageType=FullColor
      mirroring=NoMirror
      resourceName=../../../Engine/Content/Slate/Common/Scrollbar_Background_Vertical.png
      resourceObject=null
      tiling=NoTile
      imageSize=<8 8>
      margin=<ltrb 0 0 0 0>
      outlineSettings {
       bUseBrushTransparency=false
       roundingType=HalfHeightRadius
       width=0
       color {colorUseRule=UseColor_Specified, specifiedColor=<rgba 0 0 0 0>}
       cornerRadii=<0 0 0 0>
      }
      tintColor {colorUseRule=UseColor_Specified, specifiedColor=<rgba 1 1 1 1>}
      uVRegion {bIsValid=false, max=<0 0>, min=<0 0>}
     }
     verticalTopSlotImage {
      bIsDynamicallyLoaded=false
      drawAs=NoDrawType
      imageType=FullColor
      mirroring=NoMirror
      resourceName=../../../Engine/Content/Slate/Common/Scrollbar_Background_Vertical.png
      resourceObject=null
      tiling=NoTile
      imageSize=<8 8>
      margin=<ltrb 0 0 0 0>
      outlineSettings {
       bUseBrushTransparency=false
       roundingType=HalfHeightRadius
       width=0
       color {colorUseRule=UseColor_Specified, specifiedColor=<rgba 0 0 0 0>}
       cornerRadii=<0 0 0 0>
      }
      tintColor {colorUseRule=UseColor_Specified, specifiedColor=<rgba 1 1 1 1>}
      uVRegion {bIsValid=false, max=<0 0>, min=<0 0>}
     }
    }
    widgetStyle {
     barThickness=9
     bottomShadowBrush {
      bIsDynamicallyLoaded=false
      drawAs=Image
      imageType=FullColor
      mirroring=NoMirror
      resourceName=../../../Engine/Content/Slate/Common/ScrollBoxShadowBottom.png
      resourceObject=/FortniteCoreUI/UI/DiscoveryShared/Art/MI_UI_Discovery_ScrollingShadow_Bottom
      tiling=NoTile
      imageSize=<32 32>
      margin=<ltrb 0.5 0 0.5 1>
      outlineSettings {
       bUseBrushTransparency=false
       roundingType=HalfHeightRadius
       width=0
       color {colorUseRule=UseColor_Specified, specifiedColor=<rgba 0 0 0 0>}
       cornerRadii=<0 0 0 0>
      }
      tintColor {colorUseRule=UseColor_Specified, specifiedColor=<rgba 1 1 1 1>}
      uVRegion {bIsValid=false, max=<0 0>, min=<0 0>}
     }
     horizontalScrolledContentPadding=<ltrb 0 0 1 0>
     leftShadowBrush {
      bIsDynamicallyLoaded=false
      drawAs=NoDrawType
      imageType=FullColor
      mirroring=NoMirror
      resourceName=../../../Engine/Content/Slate/Common/ScrollBoxShadowLeft.png
      resourceObject=null
      tiling=NoTile
      imageSize=<8 16>
      margin=<ltrb 1 0.5 0 0.5>
      outlineSettings {
       bUseBrushTransparency=false
       roundingType=HalfHeightRadius
       width=0
       color {colorUseRule=UseColor_Specified, specifiedColor=<rgba 0 0 0 0>}
       cornerRadii=<0 0 0 0>
      }
      tintColor {colorUseRule=UseColor_Specified, specifiedColor=<rgba 1 1 1 1>}
      uVRegion {bIsValid=false, max=<0 0>, min=<0 0>}
     }
     rightShadowBrush {
      bIsDynamicallyLoaded=false
      drawAs=NoDrawType
      imageType=FullColor
      mirroring=NoMirror
      resourceName=../../../Engine/Content/Slate/Common/ScrollBoxShadowRight.png
      resourceObject=null
      tiling=NoTile
      imageSize=<8 16>
      margin=<ltrb 0 0.5 1 0.5>
      outlineSettings {
       bUseBrushTransparency=false
       roundingType=HalfHeightRadius
       width=0
       color {colorUseRule=UseColor_Specified, specifiedColor=<rgba 0 0 0 0>}
       cornerRadii=<0 0 0 0>
      }
      tintColor {colorUseRule=UseColor_Specified, specifiedColor=<rgba 1 1 1 1>}
      uVRegion {bIsValid=false, max=<0 0>, min=<0 0>}
     }
     topShadowBrush {
      bIsDynamicallyLoaded=false
      drawAs=Image
      imageType=FullColor
      mirroring=NoMirror
      resourceName=../../../Engine/Content/Slate/Common/ScrollBoxShadowTop.png
      resourceObject=/FortniteCoreUI/UI/DiscoveryShared/Art/MI_UI_Discovery_ScrollingShadow
      tiling=NoTile
      imageSize=<32 32>
      margin=<ltrb 0.5 1 0.5 0>
      outlineSettings {
       bUseBrushTransparency=false
       roundingType=HalfHeightRadius
       width=0
       color {colorUseRule=UseColor_Specified, specifiedColor=<rgba 0 0 0 0>}
       cornerRadii=<0 0 0 0>
      }
      tintColor {colorUseRule=UseColor_Specified, specifiedColor=<rgba 1 1 1 1>}
      uVRegion {bIsValid=false, max=<0 0>, min=<0 0>}
     }
     verticalScrolledContentPadding=<ltrb 0 0 0 1>
    }
   }
   slot {class=WrapBoxSlot}
  }
  Text_ActivityDescription_1 {
   widgetClassPath=&CommonTextBlock
   bIsVariable=true
   parent=ScrollBox_ActivityDescription
   widget=Text_ActivityDescription_1
   properties {
    autoWrapText=true
    lineHeightPercentage=1.28
    mobileFontSizeMultiplier=1.2
    style=/FortUILibrary/Styles/Typography/_TextStyle_Body2_LightBlue._TextStyle_Body2_LightBlue_C
    text="[Game Mode Description. Wrapping is set to per character, text widget clips to bounds on itself. Lorem ipsum dolor sit amet, consectetur adipiscing elit, sed do eiusmod tempor incididunt ut labore et dolore magna aliqua. Ut enim ad minim veniam, quis nostrud exercitation ullamco laboris nisi ut aliquip ex ea commodo consequat. Duis aute irure dolor in reprehenderit in voluptate velit esse cillum dolore eu fugiat nulla pariatur. Excepteur sint occaecat cupidatat non proident, sunt in culpa qui officia deserunt mollit anim id est laborum.}"
    wrapTextAt=1000
    wrappingPolicy=AllowPerCharacterWrapping
    colorAndOpacity {colorUseRule=UseColor_Specified, specifiedColor=<rgba 0.6038 1 1 1>}
    font {
     bForceMonospaced=false
     bMaterialIsStencil=false
     fontMaterial=null
     fontObject=&BurbankSmall
     letterSpacing=0
     monospacedWidth=1
     size=15
     skewAmount=0
     typefaceFontName=Bold
     outlineSettings {
      bApplyOutlineToDropShadows=false
      bMiteredCorners=false
      bSeparateFillAlpha=false
      outlineMaterial=null
      outlineSize=0
      outlineColor=<rgba 0 0 0 1>
     }
    }
    shadowOffset=<0 0>
    strikeBrush {
     bIsDynamicallyLoaded=false
     drawAs=NoDrawType
     imageType=NoImage
     mirroring=NoMirror
     resourceName=None
     resourceObject=null
     tiling=NoTile
     imageSize=<32 32>
     margin=<ltrb 0 0 0 0>
     outlineSettings {
      bUseBrushTransparency=false
      roundingType=HalfHeightRadius
      width=0
      color {colorUseRule=UseColor_Specified, specifiedColor=<rgba 0 0 0 0>}
      cornerRadii=<0 0 0 0>
     }
     tintColor {colorUseRule=UseColor_Specified, specifiedColor=<rgba 1 1 1 1>}
     uVRegion {bIsValid=false, max=<0 0>, min=<0 0>}
    }
   }
   slot {class=ScrollBoxSlot}
  }
 ]
}
