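# Sidebar onboarding tooltip for Discovery game details that introduces developer pages and following creators, built on the shared UIKit coach mark.
# On Construct, it picks "follow" or "favorites" description wording using the creator-follow rebranding runtime option.
# Showing the tooltip activates the coach mark; its single "Got It" action deactivates it, then notifies the parent that the action was handled and closing finished.
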WidgetBlueprint /FortniteCoreUI/UI/DiscoveryShared/GameDetails/WBP_ActivityDetails_FollowCreator_CoachMark
graphs [EventGraph {nodes [
   "Button Other Action Clicked (WBP_UIKit_CoachMark)"() {is=title K2Node_ComponentBoundEvent}
   n1 = Get WBP_UIKit_CoachMark() -> (WBP_UIKit_CoachMark:WBP UIKit Coach Mark Object Reference)
   > DeactivateWidget(self=@n1)
   > HandleOtherActionClicked(self:Fort Sidebar Onboard Tooltip Widget Object Reference)
   > NotifyCloseAnimationFinished(self:Fort Sidebar Onboard Tooltip Widget Object Reference)
   Event Construct()
   n2 = GetRuntimeOptions()
   > Set TextDescription!(
    execute
    TextDescription:Text=SelectText(
     A="LOCTEXT(\"Discover their featured islands, browse hand-selected picks, and follow developers to have new content delivered straight to you!\")"
     B="LOCTEXT(\"Discover their featured islands, browse hand-selected picks, and add creators to your favorites to have new content delivered straight to your Library!\")"
     bPickA=GetIsCreatorFollowRebrandingEnabled!(self:Fort Runtime Options Object Reference=@n2) -> (ReturnValue:bool)
    )
    self:WBP UIKit Coach Mark Base Object Reference=Get WBP_UIKit_CoachMark()
   ) -> (then, Output_Get:Text)
   Event OnShowTooltip() {is=title K2Node_Event}
   > ActivateWidget(self=@n1)
  ]}]
variables []
widgets {
 info {
  inheritedWidgetCount=2
  parentClass=/Script/FortniteUI.FortSidebarOnboardTooltipWidget
  rootWidgetClass=/FortUILibrary/UIKit/CoachMark/WBP_UIKit_CoachMark.WBP_UIKit_CoachMark_C
  widgetCount=3
 }
 widgets [
  WBP_UIKit_CoachMark {
   widgetClassPath=/FortUILibrary/UIKit/CoachMark/WBP_UIKit_CoachMark
   bIsVariable=true
   widget=WBP_UIKit_CoachMark
   properties {
    bIsFocusable=true
    navigation=&self:WidgetTree.WBP_UIKit_CoachMark.WidgetNavigation_0
    shouldShowDismissButton=false
    textDescription="Discover their featured islands, browse hand-selected picks, and add creators to your favorites to have new content delivered straight to your Library!"
    textDismissButton=""
    textHeader="Developer Pages, Leveled Up."
    textOtherActionButton=GOT IT
    visibility=SelfHitTestInvisible
    widget Width=650
    input Action Other {dataTable=&InputActionDataTable, rowName=GenericFaceButtonBottom}
   }
  }
  Dismiss_Button {widgetClassPath=&CommonButtonLegacy, bInherited=true, widget=null}
  OtherAction_Button {widgetClassPath=&CommonButtonLegacy, bInherited=true, widget=null}
 ]
}
