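# Activatable modal in Discovery game details for managing an island's notifications, with one "Island Updates" toggle inside a UIKit dialog.
# SetIslandName formats the island name into both description texts.
# PreConstruct overrides the toggle's selected-state colors through its material instance, and the toggle is the desired focus target.
# The touch-only close button, dialog button event and focus loss deactivate the widget; this graph contains no subscription logic.
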WidgetBlueprint /FortniteCoreUI/UI/DiscoveryShared/GameDetails/WBP_ActivityDetails_ManageNotifications_Modal
graphs [
 BP_GetDesiredFocusTarget {nodes [
   Get Desired Focus Target()
   > Return Node!(
    execute
    ReturnValue:Widget Object Reference=Get ToggleButton_IslandUpdates() -> (ToggleButton_IslandUpdates:WBP UIKit Switcher Button Object Reference)
   )
  ]}
 EventGraph {nodes [
   n1 = "On Button Clicked (Button_BackTouch)"!() -> (OutputDelegate, then, Button:Common Button Legacy Object Reference) {is=title K2Node_ComponentBoundEvent}
   DeactivateWidget(execute=[@n1.then, @n2.then, @n6.then])
   n2 = OnCloseButtonClicked!() -> (OutputDelegate, then, ButtonIndex:int) {is=title K2Node_CustomEvent}
   n3 = Event PreConstruct()
   > Parent: PreConstruct(IsDesignTime=@n3.IsDesignTime)
   > SetSwitcherButtonMIDParameter()
   Event Construct()
   > Parent: Construct()
   n4 = Create Event()
   n5 = Get Dialog_Internal() -> (Dialog_Internal:WBP UIKit Dialog Internal Object Reference)
   > Bind Event to Dialog Button Clicked Dispatcher(self:WBP UIKit Dialog Internal Object Reference=@n5, Delegate=@n4) {is=title K2Node_AddDelegate}
   n6 = Event OnFocusLost!() -> (OutputDelegate, then, InFocusEvent:Focus Event Structure) {is=title K2Node_Event}
   Event Destruct()
   > Parent: Destruct()
   > Unbind Event from Dialog Button Clicked Dispatcher(self:WBP UIKit Dialog Internal Object Reference=@n5, Delegate=@n4) {is=title K2Node_RemoveDelegate}
  ]}
 SetIslandName {nodes [
   SetIslandName!() -> (then, IslandName:Text) {is=title K2Node_FunctionEntry}
   > "SetText (Text)"(
    self=Get Text_Desc() -> (Text_Desc:Common Text Object Reference)
    InText=Format Text!(Format="LOCTEXT(\"Receive in-game notifications from {IslandName} about:\")", IslandName:Text=Get IslandName() -> (IslandName:Text)) -> (Result)
   )
   > "SetText (Text)"(
    self=Get Text_NewReleases_Desc() -> (Text_NewReleases_Desc:Common Text Object Reference)
    InText=Format Text(Format="LOCTEXT(\"Get notified when {IslandName} is updated\")", IslandName=Get IslandName())
   )
  ]}
 SetSwitcherButtonMIDParameter {nodes [
   SetSwitcherButtonMIDParameter() {is=title K2Node_FunctionEntry}
   n1 = Get ImageMID!(self:WBP UIKit Block Switcher Object Reference=Get Block_Switcher!(self:WBP UIKit Switcher Button Object Reference=Get ToggleButton_IslandUpdates()) -> (Block_Switcher:WBP UIKit Block Switcher Object Reference)) -> (ImageMID:Material Instance Dynamic Object Reference)
   > SetVectorParameterValue(self=@n1, ParameterName=Color_SelectedIdle, Value=Get MI_ColorParameter() -> (MI_ColorParameter:Linear Color Structure))
   > SetVectorParameterValue(self=@n1, ParameterName=Color_SelectedHovered, Value=Get MI_ColorParameter())
   > SetVectorParameterValue(self=@n1, ParameterName=Color_SelectedFocused, Value=Get MI_ColorParameter())
   > SetVectorParameterValue(self=@n1, ParameterName=Color_SelectedPressed, Value=Get MI_ColorParameter())
   > SetVectorParameterValue(self=@n1, ParameterName=Color_SelectedDisabled, Value=Get MI_ColorParameter())
  ]}
]
variables []
widgets {
 info {
  inheritedWidgetCount=0
  namedSlotCount=2
  parentClass=&CommonActivatableWidget
  rootWidgetClass=/FortUILibrary/Systems/Dialog/WBP_UIKit_Dialog_Internal.WBP_UIKit_Dialog_Internal_C
  widgetCount=12
 }
 widgets [
  Dialog_Internal {
   widgetClassPath=/FortUILibrary/Systems/Dialog/WBP_UIKit_Dialog_Internal
   bIsVariable=true
   widget=Dialog_Internal
   properties {
    bIsFocusable=true
    designTimePreview_DialogVM=&/DialogVM_ActivityDetails_Subscribe.DialogVM_ActivityDetails_Subscribe_C
    useClickCatch=true
    useFooter=true
    buttons {
     each Button Horizontal Fill=false
     each Button Min Width=0
     each Button Size=Medium
     focuse Button=0
     space Between Buttons=30
     cTA Buttons Types [Quiet]
     group Placement {
      horizontal Alignment=HAlign_Fill
      vertical Alignment=VAlign_Bottom
      grid {column=1, column Span=0, layer=0, row=2, row Span=0, nudge=<0 0>}
      padding=<ltrb 80 0 80 -45>
     }
    }
    dialogSIze {
     height Override {enable=false, value=800}
     max Aspect Ratio {enable=false, value=0}
     max Desired Height {enable=false, value=800}
     max Desired Width {enable=false, value=0}
     min Aspect Ratio {enable=false, value=0}
     min Desired Height {enable=true, value=552}
     min Desired Width {enable=false, value=0}
     width Override {enable=true, value=961}
    }
    footerTextProperties {
     auto Wrap Text=true
     justification=Left
     line Height Percentage=1.35
     mobile Font Size Multiplier=1.3
     transform Policy=None
     wrap Text At=0
     color {colorUseRule=UseColor_Specified, specifiedColor=<rgba 0.5776 0.5776 0.5776 1>}
     font {
      bForceMonospaced=false
      bMaterialIsStencil=false
      fontMaterial=null
      fontObject=&HeadingNow
      letterSpacing=1
      monospacedWidth=1
      size=11.25
      skewAmount=0
      typefaceFontName=74 Regular
      outlineSettings {
       bApplyOutlineToDropShadows=false
       bMiteredCorners=false
       bSeparateFillAlpha=false
       outlineMaterial=null
       outlineSize=0
       outlineColor=<rgba 0 0 0 1>
      }
     }
     placement {
      horizontal Alignment=HAlign_Fill
      vertical Alignment=VAlign_Fill
      grid {column=1, column Span=0, layer=0, row=3, row Span=0, nudge=<0 0>}
      padding=<ltrb 80 0 160 68>
     }
    }
   }
  }
  Content {widgetClassPath=&Overlay, namedSlotHost=Dialog_Internal, widget=Content}
  VB_SupportACreator {
   widgetClassPath=&VerticalBox
   parent=Content
   widget=VB_SupportACreator
   slot {class=OverlaySlot, properties {horizontalAlignment=HAlign_Fill, verticalAlignment=VAlign_Fill, padding=<ltrb 80 0 80 0>}}
  }
  Text_Desc {
   widgetClassPath=&CommonTextBlock
   bIsVariable=true
   parent=VB_SupportACreator
   widget=Text_Desc
   properties {
    autoWrapText=true
    bIsScrollingEnabled=false
    text="Receive in-game notifications from {IslandName} about:"
    font {
     bForceMonospaced=false
     bMaterialIsStencil=false
     fontMaterial=null
     fontObject=&HeadingNow
     letterSpacing=0
     monospacedWidth=1
     size=18
     skewAmount=0
     typefaceFontName=74 Regular
     outlineSettings {
      bApplyOutlineToDropShadows=false
      bMiteredCorners=false
      bSeparateFillAlpha=false
      outlineMaterial=null
      outlineSize=0
      outlineColor=<rgba 0 0 0 1>
     }
    }
    shadowOffset=<0 0>
   }
   slot {class=VerticalBoxSlot, properties {verticalAlignment=VAlign_Top, padding=<ltrb 0 0 0 48>}}
  }
  VerticalBox_1 {
   widgetClassPath=&VerticalBox
   parent=VB_SupportACreator
   widget=VerticalBox_1
   properties {
    bOverride_Cursor=true
    navigation=&self:WidgetTree.VerticalBox_1.WidgetNavigation_0
   }
   slot {class=VerticalBoxSlot}
  }
  HorizontalBox_1 {
   widgetClassPath=&HorizontalBox
   parent=VerticalBox_1
   widget=HorizontalBox_1
   properties {
    bOverride_Cursor=true
    navigation=&self:WidgetTree.HorizontalBox_1.WidgetNavigation_0
   }
   slot {class=VerticalBoxSlot}
  }
  VerticalBox_0 {
   widgetClassPath=&VerticalBox
   parent=HorizontalBox_1
   widget=VerticalBox_0
   slot {class=HorizontalBoxSlot, properties {padding=<ltrb 0 0 8 0>, size {sizeRule=Fill, value=1}}}
  }
  CommonTextBlock_137 {
   widgetClassPath=&CommonTextBlock
   parent=VerticalBox_0
   widget=CommonTextBlock_137
   properties {
    bIsScrollingEnabled=false
    text=Island Updates
    font {
     bForceMonospaced=false
     bMaterialIsStencil=false
     fontMaterial=null
     fontObject=&NotoSans
     letterSpacing=0
     monospacedWidth=1
     size=24
     skewAmount=0
     typefaceFontName=NotoSans-Regular
     outlineSettings {
      bApplyOutlineToDropShadows=false
      bMiteredCorners=false
      bSeparateFillAlpha=false
      outlineMaterial=null
      outlineSize=0
      outlineColor=<rgba 0 0 0 1>
     }
    }
    shadowOffset=<0 0>
   }
   slot {class=VerticalBoxSlot}
  }
  Text_NewReleases_Desc {
   widgetClassPath=&CommonTextBlock
   bIsVariable=true
   parent=VerticalBox_0
   widget=Text_NewReleases_Desc
   properties {
    autoWrapText=true
    bIsScrollingEnabled=false
    text="Get notified when {IslandName} is updated"
    colorAndOpacity {colorUseRule=UseColor_Specified, specifiedColor=<rgba 0.5776 0.5776 0.5776 1>}
    font {
     bForceMonospaced=false
     bMaterialIsStencil=false
     fontMaterial=null
     fontObject=&NotoSans
     letterSpacing=0
     monospacedWidth=1
     size=24
     skewAmount=0
     typefaceFontName=NotoSans-Regular
     outlineSettings {
      bApplyOutlineToDropShadows=false
      bMiteredCorners=false
      bSeparateFillAlpha=false
      outlineMaterial=null
      outlineSize=0
      outlineColor=<rgba 0 0 0 1>
     }
    }
    shadowOffset=<0 0>
   }
   slot {class=VerticalBoxSlot}
  }
  ToggleButton_IslandUpdates {
   widgetClassPath=/FortUILibrary/UIKit/Buttons/WBP_UIKit_SwitcherButton
   bIsVariable=true
   parent=HorizontalBox_1
   widget=ToggleButton_IslandUpdates
   properties {
    bHideInputAction=true
    bIsFocusable=true
    navigation=&self:WidgetTree.ToggleButton_IslandUpdates.WidgetNavigation_1
    useTextBlock=false
    buttonPadding=<ltrb 0 0 0 0>
    triggeringInputAction {dataTable=null, rowName=None}
   }
   slot {class=HorizontalBoxSlot}
  }
  VisiblityBorder_Touch {
   widgetClassPath=&CommonHardwareVisibilityBorder
   namedSlotHost=Dialog_Internal
   widget=VisiblityBorder_Touch
   properties {
    horizontalAlignment=HAlign_Right
    verticalAlignment=VAlign_Top
    padding=<ltrb 0 32 32 0>
    visibilityQuery {
     autoDescription=" ANY( Input.Touch )"
     tokenStreamVersion=0
     userDescription=""
     queryTokenStream [0,1,1,1,0]
     tagDictionary [{tagName=Input.Touch}]
    }
   }
  }
  Button_BackTouch {
   widgetClassPath=/FortUILibrary/Foundation/Buttons/CloseButton
   bIsVariable=true
   parent=VisiblityBorder_Touch
   widget=Button_BackTouch
   properties {
    onButtonBaseClicked=()
    onButtonBaseDoubleClicked=()
    onButtonBaseHovered=()
    onButtonBaseUnhovered=()
    onSelectedChangedBase=()
    renderTransform {angle=0, scale=<1 1>, shear=<0 0>, translation=<0 10>}
   }
   slot {class=BorderSlot, properties {horizontalAlignment=HAlign_Right, verticalAlignment=VAlign_Top, padding=<ltrb 0 32 32 0>}}
  }
 ]
}
