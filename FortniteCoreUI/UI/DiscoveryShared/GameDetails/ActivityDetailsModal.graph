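# Widget Blueprint for the Fortnite Discovery activity details modal, showing a game's key art, description, tags, player count and accept, favorite, subscribe and context-menu buttons.
# It binds to ActivityDetailsDisplay events to refresh tags, lock info, favorite and disabled states.
# It toggles a sub-mode list, hosts an injected community row, and routes focus by gating state.
# UMG animations drive open, key-art reveal, details intro/outro and warning pulse.
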WidgetBlueprint /FortniteCoreUI/UI/DiscoveryShared/GameDetails/ActivityDetailsModal
animations [
 OnOpen {
  binding_count=2
  length_seconds=0.32
  object=&self:OnOpen
  sequence {
   bindings [
    ModalContent {
     id=#1
     possessed_class=Overlay
     tracks [
      {
       type=MovieSceneFloatTrack
       display_name=Render Opacity
       sections [{type=MovieSceneFloatSection, channels [None {type=MovieSceneScriptingFloatChannel, keys [[32,0],[48,1]]}]}]
      }
      {
       type=MovieScene2DTransformTrack
       sections [{
         type=MovieScene2DTransformSection
         channels [
          Scale.X {type=MovieSceneScriptingFloatChannel, keys [[32,0.875],[64,1]]}
          Scale.Y {type=MovieSceneScriptingFloatChannel, keys [[32,0.875],[64,1]]}
         ]
        }]
      }
     ]
    }
    Thumbnail {
     id=#2
     possessed_class=Image
     tracks [{
       type=MovieSceneFloatTrack
       display_name=Render Opacity
       sections [{type=MovieSceneFloatSection, channels [None {type=MovieSceneScriptingFloatChannel, keys [[48,0],[64,1]]}]}]
      }]
    }
   ]
   display_rate {numerator=100}
  }
 }
 ShowKeyArt {
  binding_count=1
  length_seconds=0.64
  object=&self:ShowKeyArt
  sequence {
   bindings [Thumbnail {
     id=#3
     possessed_class=Image
     tracks [{
       type=MovieSceneWidgetMaterialTrack
       sections [{
         type=MovieSceneParameterSection
         end_frame=64
         start_frame=0
         channels [
          Thumbnail Opacity {type=MovieSceneScriptingFloatChannel, keys [[0,0],[64,1]]}
          Thumbnail Zoom {type=MovieSceneScriptingFloatChannel, keys [[0,0.05],[64,0]]}
          Triangle Distance {type=MovieSceneScriptingFloatChannel, keys [[0,0.5],[64,0]]}
         ]
        }]
      }]
    }]
   display_rate {numerator=100}
  }
 }
 HideSpinner {
  binding_count=1
  length_seconds=0.32
  object=&self:HideSpinner
  sequence {
   bindings [WBP_UIKit_Throbber {
     id=#4
     possessed_class=WBP_UIKit_Throbber_C
     tracks [
      {
       type=MovieSceneFloatTrack
       display_name=Render Opacity
       sections [{type=MovieSceneFloatSection, channels [None {type=MovieSceneScriptingFloatChannel, keys [[8,0],[32,1]]}]}]
      }
      {
       type=MovieScene2DTransformTrack
       sections [{
         type=MovieScene2DTransformSection
         channels [
          Scale.X {type=MovieSceneScriptingFloatChannel, keys [[8,0.875],[32.0017,1]]}
          Scale.Y {type=MovieSceneScriptingFloatChannel, keys [[8,0.875],[32.0017,1]]}
         ]
        }]
      }
      {
       type=MovieSceneByteTrack
       display_name=Visibility
       sections [{type=MovieSceneByteSection, channels [None {type=MovieSceneScriptingByteChannel, keys [[0,1],[8,3]]}]}]
      }
     ]
    }]
   display_rate {numerator=100}
  }
 }
 DetailsIntro {
  binding_count=10
  length_seconds=0.48
  object=&self:DetailsIntro
  sequence {bindings [
    Button_Favorite {
     id=#5
     possessed_class=WBP_GameDetails_Toggle_C
     tracks [
      {
       type=MovieSceneFloatTrack
       display_name=Render Opacity
       sections [{type=MovieSceneFloatSection, channels [None {type=MovieSceneScriptingFloatChannel, keys [[6.4,0],[9.5993,1]]}]}]
      }
      {
       type=MovieScene2DTransformTrack
       sections [{
         type=MovieScene2DTransformSection
         channels [
          Scale.X {type=MovieSceneScriptingFloatChannel, keys [[6.4,0.875],[9.5997,1]]}
          Scale.Y {type=MovieSceneScriptingFloatChannel, keys [[6.4,0.875],[9.5997,1]]}
         ]
        }]
      }
     ]
    }
    Button_ContextMenu {
     id=#6
     display_name=WBP_UIKit_ContextMenuAnchorButton
     possessed_class=WBP_UIKit_ContextMenuAnchorButton_C
     tracks [
      {
       type=MovieSceneFloatTrack
       display_name=Render Opacity
       sections [{type=MovieSceneFloatSection, channels [None {type=MovieSceneScriptingFloatChannel, keys [[6.4,0],[9.5993,1]]}]}]
      }
      {
       type=MovieScene2DTransformTrack
       sections [{
         type=MovieScene2DTransformSection
         channels [
          Scale.X {type=MovieSceneScriptingFloatChannel, keys [[6.4,0.875],[9.5997,1]]}
          Scale.Y {type=MovieSceneScriptingFloatChannel, keys [[6.4,0.875],[9.5997,1]]}
         ]
        }]
      }
     ]
    }
    HorizontalBox_0 {
     id=#7
     possessed_class=HorizontalBox
     tracks [{
       type=MovieSceneFloatTrack
       display_name=Render Opacity
       sections [{type=MovieSceneFloatSection, channels [None {type=MovieSceneScriptingFloatChannel, keys [[3.2,0.125],[7.9993,1]]}]}]
      }]
    }
    Text_ActivityName {
     id=#8
     possessed_class=FortActivityScalingTextBlock
     tracks [
      {
       type=MovieSceneWidgetMaterialTrack
       display_name=Font.FontMaterial
       sections [{type=MovieSceneParameterSection, channels [AlphaOveride {type=MovieSceneScriptingFloatChannel, keys [[0,0.125],[8,1]]}]}]
      }
      {
       type=MovieSceneFloatTrack
       display_name=Line Height Percentage
       sections [{type=MovieSceneFloatSection, end_frame=9, start_frame=0, channels [None {type=MovieSceneScriptingFloatChannel, const [0.85,2]}]}]
      }
     ]
    }
    EntryBox_ActivityTags {
     id=#9
     possessed_class=DynamicEntryBox
     tracks [{
       type=MovieSceneFloatTrack
       display_name=Render Opacity
       sections [{type=MovieSceneFloatSection, channels [None {type=MovieSceneScriptingFloatChannel, keys [[3.2,0.125],[7.9997,1]]}]}]
      }]
    }
    Text_ActivityDescription {
     id=#10
     possessed_class=CommonTextBlock
     tracks [{
       type=MovieSceneWidgetMaterialTrack
       display_name=Font.FontMaterial
       sections [{type=MovieSceneParameterSection, channels [AlphaOveride {type=MovieSceneScriptingFloatChannel, keys [[3.2,0.125],[7.9997,1]]}]}]
      }]
    }
    Button_Accept {
     id=#11
     possessed_class=WBP_UIKit_Button_Loud_C
     tracks [
      {
       type=MovieSceneFloatTrack
       display_name=Render Opacity
       sections [{type=MovieSceneFloatSection, channels [None {type=MovieSceneScriptingFloatChannel, keys [[4.7997,0],[9.5993,1]]}]}]
      }
      {
       type=MovieScene2DTransformTrack
       sections [{type=MovieScene2DTransformSection, channels [Translation.Y {type=MovieSceneScriptingFloatChannel, keys [[4.8,8],[9.5997,0]]}]}]
      }
     ]
    }
    NamedSlot_AdditionalDetail {
     id=#12
     possessed_class=NamedSlot
     tracks [
      {
       type=MovieSceneFloatTrack
       display_name=Render Opacity
       sections [{type=MovieSceneFloatSection, channels [None {type=MovieSceneScriptingFloatChannel, keys [[4.7997,0],[9.5993,1]]}]}]
      }
      {
       type=MovieScene2DTransformTrack
       sections [{type=MovieScene2DTransformSection, channels [Translation.Y {type=MovieSceneScriptingFloatChannel, keys [[4.8,8],[9.5997,0]]}]}]
      }
     ]
    }
    DisabledWarningHB {
     id=#13
     possessed_class=HorizontalBox
     tracks [{
       type=MovieSceneFloatTrack
       display_name=Render Opacity
       sections [{type=MovieSceneFloatSection, channels [None {type=MovieSceneScriptingFloatChannel, keys [[3.2,0.125],[7.9997,1]]}]}]
      }]
    }
    GP_Content {
     id=#14
     possessed_class=GridPanel
     tracks [{
       type=MovieScene2DTransformTrack
       sections [{type=MovieScene2DTransformSection, channels [Translation.Y {type=MovieSceneScriptingFloatChannel, keys [[0,4],[8,0]]}]}]
      }]
    }
   ]}
 }
 DetailsOutro {
  binding_count=9
  length_seconds=0.32
  object=&self:DetailsOutro
  sequence {bindings [
    Button_Accept {
     id=#15
     possessed_class=ActivityBrowserSoloButton_C
     tracks [{
       type=MovieSceneFloatTrack
       display_name=Render Opacity
       sections [{type=MovieSceneFloatSection, channels [None {type=MovieSceneScriptingFloatChannel, keys [[0,1],[4,0]]}]}]
      }]
    }
    DisabledWarningHB {
     id=#16
     possessed_class=HorizontalBox
     tracks [{
       type=MovieSceneFloatTrack
       display_name=Render Opacity
       sections [{type=MovieSceneFloatSection, channels [None {type=MovieSceneScriptingFloatChannel, keys [[0,1],[4,0.125]]}]}]
      }]
    }
    Button_Favorite {
     id=#17
     possessed_class=WBP_GameDetails_Toggle_C
     tracks [
      {
       type=MovieSceneFloatTrack
       display_name=Render Opacity
       sections [{type=MovieSceneFloatSection, channels [None {type=MovieSceneScriptingFloatChannel, keys [[0,1],[2.4,0]]}]}]
      }
      {
       type=MovieScene2DTransformTrack
       sections [{
         type=MovieScene2DTransformSection
         channels [
          Scale.X {type=MovieSceneScriptingFloatChannel, keys [[0.0003,1],[2.4,0.875]]}
          Scale.Y {type=MovieSceneScriptingFloatChannel, keys [[0.0003,1],[2.4,0.875]]}
         ]
        }]
      }
     ]
    }
    Button_ContextMenu {
     id=#18
     display_name=WBP_UIKit_ContextMenuAnchorButton
     possessed_class=WBP_UIKit_ContextMenuAnchorButton_C
     tracks [
      {
       type=MovieSceneFloatTrack
       display_name=Render Opacity
       sections [{type=MovieSceneFloatSection, channels [None {type=MovieSceneScriptingFloatChannel, keys [[0,1],[2.4,0]]}]}]
      }
      {
       type=MovieScene2DTransformTrack
       sections [{
         type=MovieScene2DTransformSection
         channels [
          Scale.X {type=MovieSceneScriptingFloatChannel, keys [[0.0003,1],[2.4,0.875]]}
          Scale.Y {type=MovieSceneScriptingFloatChannel, keys [[0.0003,1],[2.4,0.875]]}
         ]
        }]
      }
     ]
    }
    GP_Content {
     id=#19
     possessed_class=GridPanel
     tracks [{
       type=MovieScene2DTransformTrack
       sections [{type=MovieScene2DTransformSection, channels [Translation.Y {type=MovieSceneScriptingFloatChannel, keys [[0,0],[6.4,4]]}]}]
      }]
    }
    Text_ActivityOrigin {
     id=#20
     possessed_class=CommonRichTextBlock
     tracks [{
       type=MovieSceneFloatTrack
       display_name=Render Opacity
       sections [{type=MovieSceneFloatSection, channels [None {type=MovieSceneScriptingFloatChannel, keys [[0,1],[4,0.125]]}]}]
      }]
    }
    EntryBox_ActivityTags {
     id=#21
     possessed_class=DynamicEntryBox
     tracks [{
       type=MovieSceneFloatTrack
       display_name=Render Opacity
       sections [{type=MovieSceneFloatSection, channels [None {type=MovieSceneScriptingFloatChannel, keys [[0,1],[4,0.125]]}]}]
      }]
    }
    Text_ActivityName {
     id=#22
     possessed_class=FortActivityScalingTextBlock
     tracks [
      {
       type=MovieSceneWidgetMaterialTrack
       display_name=Font.FontMaterial
       sections [{type=MovieSceneParameterSection, channels [AlphaOveride {type=MovieSceneScriptingFloatChannel, keys [[0,1],[6.4,0.125]]}]}]
      }
      {
       type=MovieSceneFloatTrack
       display_name=Line Height Percentage
       sections [{type=MovieSceneFloatSection, end_frame=6, start_frame=0, channels [None {type=MovieSceneScriptingFloatChannel, const [0.85,2]}]}]
      }
     ]
    }
    Text_ActivityDescription {
     id=#23
     possessed_class=CommonTextBlock
     tracks [{
       type=MovieSceneWidgetMaterialTrack
       display_name=Font.FontMaterial
       sections [{type=MovieSceneParameterSection, channels [AlphaOveride {type=MovieSceneScriptingFloatChannel, keys [[0,1],[4,0.125]]}]}]
      }]
    }
   ]}
 }
 DescriptionScrollRestarted {
  binding_count=1
  length_seconds=5
  object=&self:DescriptionScrollRestarted
  sequence {bindings [Text_ActivityDescription {
     id=#24
     possessed_class=CommonTextBlock
     tracks [
      {
       type=MovieSceneFloatTrack
       display_name=Render Opacity
       sections [{type=MovieSceneFloatSection, end_frame=4, start_frame=0, channels [None {type=MovieSceneScriptingFloatChannel, keys [[0,0],[4,1]]}]}]
      }
      {
       type=MovieSceneFloatTrack
       display_name=Line Height Percentage
       sections [{type=MovieSceneFloatSection, end_frame=4, start_frame=0, channels [None {type=MovieSceneScriptingFloatChannel, const [0.85,2]}]}]
      }
     ]
    }]}
 }
 DisabledIconPulse {
  binding_count=3
  length_seconds=5
  object=&self:DisabledIconPulse
  sequence {bindings [
    Text_ActivityName {
     id=#25
     possessed_class=FortActivityScalingTextBlock
     tracks [{
       type=MovieSceneFloatTrack
       display_name=Line Height Percentage
       sections [{type=MovieSceneFloatSection, end_frame=19, start_frame=0, channels [None {type=MovieSceneScriptingFloatChannel, const [0.85,2]}]}]
      }]
    }
    WarningTriangle_Pulse {
     id=#26
     possessed_class=Image
     tracks [
      {
       type=MovieSceneFloatTrack
       display_name=Render Opacity
       sections [{type=MovieSceneFloatSection, channels [None {type=MovieSceneScriptingFloatChannel, keys [[0,0],[9.6,1],[19.2,0]]}]}]
      }
      {
       type=MovieScene2DTransformTrack
       sections [{
         type=MovieScene2DTransformSection
         channels [
          Translation.X {type=MovieSceneScriptingFloatChannel, keys [[0,0],[9.6,1]]}
          Translation.Y {type=MovieSceneScriptingFloatChannel, keys [[0,0],[9.6,-7]]}
          Scale.X {type=MovieSceneScriptingFloatChannel, keys [[0,1],[9.6,2.25]]}
          Scale.Y {type=MovieSceneScriptingFloatChannel, keys [[0,1],[9.6,2.25]]}
         ]
        }]
      }
      {
       type=MovieSceneWidgetMaterialTrack
       sections [{
         type=MovieSceneParameterSection
         channels [Border_Thickness {type=MovieSceneScriptingFloatChannel, keys [[0,0.125],[11.2,0.05],[19.2,0.001]]}]
        }]
      }
     ]
    }
    IconOvr {
     id=#27
     possessed_class=Overlay
     tracks [{
       type=MovieScene2DTransformTrack
       sections [{
         type=MovieScene2DTransformSection
         channels [
          Scale.X {type=MovieSceneScriptingFloatChannel, keys [[4,1],[5.6,1.25],[8.8,1]]}
          Scale.Y {type=MovieSceneScriptingFloatChannel, keys [[4,1],[5.6,1.25],[8.8,1]]}
         ]
        }]
      }]
    }
   ]}
 }
]
graphs [
 BP_GetDesiredFocusTarget {nodes [
   Get Desired Focus Target()
   > n1 = Branch(Condition=Get IsShowingSubModeList() -> (IsShowingSubModeList:bool))
   > Return Node!(execute, ReturnValue:Widget Object Reference=Get List_SubModeList() -> (List_SubModeList:Fort Activity Mode Set Selection Object Reference))
   Parent: Get Desired Focus Target!(execute=@n1.else) -> (then, ReturnValue:Widget Object Reference) {is=title K2Node_CallParentFunction}
   > n2 = Branch(Condition=IsContentGated!(self:Fort Game Activity Object Reference=GetSelectedGameActivity!(self:Fort Activity Details Modal Object Reference) -> (ReturnValue:Fort Game Activity Object Reference)) -> (ReturnValue:bool))
   > Return Node(ReturnValue=Get SqueegeeSlot_LockInfo() -> (SqueegeeSlot_LockInfo:Squeegee Injection Slot Object Reference))
   n3 = Select(
    Option 0:User Widget Object Reference=Get Button_Favorite() -> (Button_Favorite:Common Button Base Object Reference)
    Option 1:User Widget Object Reference=Get LastFocusedWidget() -> (LastFocusedWidget:User Widget Object Reference)
    Index=IsVisible(self=Get WidgetSwitcher_Buttons() -> (WidgetSwitcher_Buttons:Widget Switcher Object Reference))
   ) -> (ReturnValue:User Widget Object Reference)
   n4 = Cast To CommonActivatableWidget!(execute=@n2.else, Object=@n3) -> (then, CastFailed, AsCommon Activatable Widget:Common Activatable Widget Object Reference) {is=title K2Node_DynamicCast}
   > Return Node(ReturnValue=GetDesiredFocusTarget!(self:Common Activatable Widget Object Reference=@n4) -> (ReturnValue:Widget Object Reference))
   Return Node(execute=@n4.CastFailed, ReturnValue=@n3)
  ]}
 CallDelayedShowKeyArtAnim {nodes [
   CallDelayedShowKeyArtAnim() {is=title K2Node_FunctionEntry}
   > PlayAnimationForward(InAnimation=Get ShowKeyArt() -> (ShowKeyArt:Widget Animation Object Reference))
  ]}
 CreateTag {nodes [
   n1 = CreateTag!() -> (then, Text:Text, Tag Type:EFortActivityBrowserTagType Enum=Default) {is=title K2Node_FunctionEntry}
   > n2 = Switch on EFortActivityBrowserTagType!(execute, Selection:EFortActivityBrowserTagType Enum=@n1.Tag Type) -> (Default:Exec, EpicOriginal:Exec, SquadSize:Exec, MaxPlayers:Exec, XP:Exec, PurchaseRequired:Exec, DownloadRequired:Exec, LinkCategory:Exec) {is=title K2Node_SwitchEnum}
   n3 = Get EntryBox_ActivityTags() -> (EntryBox_ActivityTags:Dynamic Entry Box Object Reference)
   n4 = Create Entry of Class!(
    execute=[@n2.Default, @n2.DownloadRequired, @n2.EpicOriginal, @n2.LinkCategory, @n2.MaxPlayers, @n2.PurchaseRequired, @n2.SquadSize]
    self:Dynamic Entry Box Object Reference=@n3
    EntryClass:User Widget Class Reference=/FortUILibrary/UIKit/Tag/WBP_UIKit_Tag.WBP_UIKit_Tag_C
   ) -> (then, ReturnValue:WBP UIKit Tag Object Reference)
   > n5 = Sequence()
   > n6 = ShouldShowLinkCategoryIconsInTags!(execute, self) -> (then, ReturnValue:bool)
   n7 = Get Tag Type() -> (Tag Type:EFortActivityBrowserTagType Enum=Default)
   > n8 = Branch(Condition=AND Boolean(A="Equal (Enum)"(A:EFortActivityBrowserTagType Enum=@n7, B:EFortActivityBrowserTagType Enum=LinkCategory), B=@n6))
   > Set UseIcon!(execute, UseIcon:bool=true, self:WBP UIKit Tag Base Object Reference=@n4) -> (then, Output_Get:bool)
   > Set IconLeftAligned!(execute, IconLeftAligned:bool=true, self:WBP UIKit Tag Base Object Reference=@n4) -> (then, Output_Get:bool)
   > SetIconPadding!(execute, self:WBP UIKit Tag Base Object Reference=@n4, Padding:Margin Structure=Make Margin(Left=0, Top=0, Right=4, Bottom=0)) -> (then)
   n9 = GetSelectedGameActivity()
   > SetIconMaterial!(
    execute
    self:WBP UIKit Tag Base Object Reference=@n4
    IconMaterial:Material Instance Object Reference=Select!(
     None:Material Instance Object Reference
     LEGO:Material Instance Object Reference=/FortUILibrary/Textures/MaterialInstances/MI_UIKit_Icon_LegoIsland
     RocketRacing:Material Instance Object Reference=/FortUILibrary/Textures/MaterialInstances/MI_UIKit_Icon_RocketRacingIsland
     FallGuys:Material Instance Object Reference=/FortUILibrary/Textures/MaterialInstances/MI_UIKit_Icon_Brand_FallGuys
     TMNT:Material Instance Object Reference=/FortUILibrary/Textures/MaterialInstances/MI_UIKit_Icon_Brand_TMNT
     TheWalkingDeadUniverse:Material Instance Object Reference=/FortUILibrary/Textures/MaterialInstances/MI_UIKit_Icon_Brand_TWDU
     SolidWave:Material Instance Object Reference=/FortUILibrary/Textures/MaterialInstances/MI_UIKit_Icon_Brand_SquidGame
     Akita:Material Instance Object Reference=/FortUILibrary/Textures/MaterialInstances/MI_UIKit_Icon_Brand_Akita
     KPOPDemonHunters:Material Instance Object Reference=/FortUILibrary/Textures/MaterialInstances/MI_UIKit_Icon_Refrigerator
     EpicGames:Material Instance Object Reference
     Creatures:Material Instance Object Reference
     Index:EFortActivityLinkCategory Enum=GetLinkCategory!(self:Fort Game Activity Object Reference=@n9) -> (ReturnValue:EFortActivityLinkCategory Enum=None)
    ) -> (ReturnValue:Material Instance Object Reference)
   ) -> (then)
   > Set Use Dark Background!(execute, Use Dark Background:bool=true, self:WBP UIKit Tag Base Object Reference=@n4) -> (then, Output_Get:bool)
   > InitializeTag(self:WBP UIKit Tag Base Object Reference=@n4)
   Set UseIcon(execute=@n8.else, UseIcon=false, self=@n4)
   > Set Use Dark Background(Use Dark Background=true, self=@n4)
   > InitializeTag(self=@n4)
   n10 = Get Tag Type() -> (Tag Type=Default)
   SetTagType!(
    execute=@n5.then_1
    self:WBP UIKit Tag Base Object Reference=@n4
    Type:E_UI_TagType Enum=Select!(Option 0:E_UI_TagType Enum=NewEnumerator2, Option 1:E_UI_TagType Enum=NewEnumerator1, Index:bool="Equal (Enum)"(A=@n10, B=EpicOriginal)) -> (ReturnValue:E_UI_TagType Enum)
   ) -> (then)
   > SetTagText!(execute, self:WBP UIKit Tag Base Object Reference=@n4, Text:Text=@n1) -> (then)
   > SetLabelText!(execute, self:WBP UIKit Tag Base Object Reference=@n4, Text:Text) -> (then)
   n11 = Create Entry of Class(execute=@n2.XP, self=@n3, EntryClass=/FortUILibrary/Systems/ModularButton/Blocks/WBP_UIKit_Block_Image.WBP_UIKit_Block_Image_C) -> (ReturnValue:WBP UIKit Block Image Object Reference)
   > SetRenderScale!(execute, self:Widget Object Reference=@n11, Scale:Vector 2D Structure=(X=1.3,Y=1.3)) -> (then)
   > SetBrush!(
    execute
    self:WBP UIKit Block Image Object Reference=@n11
    InBrush_TintColor:Slate Color Structure=(SpecifiedColor=(R=1,G=1,B=1,A=1),ColorUseRule=UseColor_Specified)
    InBrush_DrawAs:ESlateBrushDrawType Enum=Image
    InBrush_Tiling:ESlateBrushTileType Enum=NoTile
    InBrush_Mirroring:ESlateBrushMirrorType Enum=NoMirror
    InBrush_ImageSize:Vector 2D Structure=(X=32,Y=32)
    InBrush_Margin_Left:float=0
    InBrush_Margin_Top:float=0
    InBrush_Margin_Right:float=8
    InBrush_Margin_Bottom:float=0
    InBrush_ResourceObject:Object Reference=/Game/UI/Foundation/Textures/Icons/Items/UniversalXP/T_UI_UniversalXP_Icon_Standard_128
    InBrush_OutlineSettings:Slate Brush Outline Settings Structure=(CornerRadii=(X=0,Y=0,Z=0,W=0),Color=(SpecifiedColor=(R=0,G=0,B=0,A=0),ColorUseRule=UseColor_Specified),Width=0,RoundingType=HalfHeightRadius,bUseBrushTransparency=False)
   ) -> (then)
  ]}
 EventGraph {nodes [
   "On Clicked (Button_Subscribe)"!() -> (OutputDelegate, then, Button:Common Button Base Object Reference) {is=title K2Node_ComponentBoundEvent}
   > n1 = Construct Object from Class(Class=&/DialogVM_ActivityDetails_Subscribe.DialogVM_ActivityDetails_Subscribe_C) -> (ReturnValue:Dialog VM Activity Details Subscribe Object Reference)
   > n2 = Create Widget!(
    execute:Exec
    Class:User Widget Class Reference=&/WBP_ActivityDetails_ManageNotifications_Modal.WBP_ActivityDetails_ManageNotifications_Modal_C
    OwningPlayer:Player Controller Object Reference
    UIKitDialogViewModel:UI Kit Dialog Viewmodel Object Reference=@n1
   ) -> (then:Exec, ReturnValue:WBP Activity Details Manage Notifications Modal Object Reference)
   > ClearChildren(self=Get O_Modal() -> (O_Modal:Overlay Object Reference))
   > n3 = AddChildToOverlay!(execute, self:Overlay Object Reference=Get O_Modal(), Content:Widget Object Reference=@n2) -> (then, ReturnValue:Overlay Slot Object Reference)
   > SetHorizontalAlignment(self=@n3, InHorizontalAlignment=HAlign_Fill)
   > SetVerticalAlignment(self=@n3, InVerticalAlignment=VAlign_Fill)
   n4 = Get ActivityDetailsDisplay() -> (ActivityDetailsDisplay:Activity Details Display Object Reference)
   > SetIslandName!(
    execute
    self:WBP Activity Details Manage Notifications Modal Object Reference=@n2
    IslandName:Text=Get Text(self:Text Object Reference=Get Text_ActivityName!(self:Fort Activity Details Display Object Reference=@n4) -> (Text_ActivityName:Common Text Object Reference))
   ) -> (then)
   n5 = Get ToggleButton_IslandUpdates!(self:WBP Activity Details Manage Notifications Modal Object Reference=@n2) -> (ToggleButton_IslandUpdates:WBP UIKit Switcher Button Object Reference)
   > Set Selected!(
    execute
    Selected:bool=Get bAmISubscribed!(self:Discover Activity View Model Object Reference=Get FortIslandDetailsVM_Manual() -> (FortIslandDetailsVM_Manual:Fort Island Details VM Object Reference)) -> (bAmISubscribed:bool)
    self:WBP UIKit Switcher Button Object Reference=@n5
   ) -> (then, Output_Get:bool)
   > n6 = Bind Event to On Clicked(self:Common Button Base Object Reference=@n5, Delegate=Create Event()) {is=title K2Node_AddDelegate}
   n7 = OnManageNotificationModalDeactivated() {is=title K2Node_CustomEvent}
   Assign On Widget Deactivated!(execute:Exec=@n6.then, self:Common Activatable Widget Object Reference=@n2, Delegate:"Delegate (by ref)"=@n7) -> (then:Exec) {is=title K2Node_AssignDelegate}
   "On Clicked (Button_ShowSubModeList)"!() -> (OutputDelegate, then, Button:Common Button Base Object Reference) {is=title K2Node_ComponentBoundEvent}
   > UpdateSubModeVisibility!(execute, self, IsShowingSubModeList:bool=true) -> (then)
   > HideNotification(self:WBP Popup Tooltip Object Reference=Get DiscoveryTooltip() -> (DiscoveryTooltip:WBP Popup Tooltip Object Reference))
   > SetFocus(self=Get List_SubModeList())
   "On Context Menu Opened (Button_ContextMenu)"() {is=title K2Node_ComponentBoundEvent}
   > HandleDetailsClicked(self:Fort Activity Details Display Object Reference=Get ActivityDetailsDisplay())
   > SetVisibility(self=Get Button_CreatorPage() -> (Button_CreatorPage:Common Button Base Object Reference), InVisibility=Collapsed)
   "On Hovered (Button_Details_Hover)"() {is=title K2Node_ComponentBoundEvent}
   > n8 = Branch(Condition=Get IsShowingSubModeList())
   n9 = Get WBP_Outline() -> (WBP_Outline:WBP UIKit Block Outline Object Reference)
   RequestTransition!(execute=@n8.else, self:UIKit Block Object Reference=@n9, State:EUIKitBlockVisualState Enum=Focused) -> (then)
   "On Unhovered (Button_Details_Hover)"() {is=title K2Node_ComponentBoundEvent}
   > n10 = Branch(Condition=Get IsShowingSubModeList())
   RequestTransition(execute=@n10.else, self=@n9, State=Unfocused)
   n11 = "On Item Clicked (List_SubModeList)"!() -> (OutputDelegate, then, Item:Object Reference) {is=title K2Node_ComponentBoundEvent}
   > HandleItemClicked!(execute, self:Fort Activity Details Modal Object Reference, SelectedItem:Object Reference=@n11.Item) -> (then)
   "On Pressed (Button_Details_Hover)"() {is=title K2Node_ComponentBoundEvent}
   > SetFocus(self=Get Button_Details_Hover() -> (Button_Details_Hover:Button Object Reference))
   "On Context Menu Closed (Button_ContextMenu)"() {is=title K2Node_ComponentBoundEvent}
   > UpdateCreatorPageAccessButtonVisibility(self:Fort Activity Details Modal Object Reference)
   "On Clicked (Button_Accept)"!() -> (OutputDelegate, then, Button:Common Button Base Object Reference) {is=title K2Node_ComponentBoundEvent}
   > Fire Event Element Interacted(Element=Get Button_Accept() -> (Button_Accept:WBP UIKit Button Loud Object Reference))
   > AcceptActivity(self:Fort Activity Details Display Object Reference=Get ActivityDetailsDisplay())
   n12 = "On Widget Injected (CommunityRow)"!() -> (OutputDelegate, then, Widget:Widget Object Reference) {is=title K2Node_ComponentBoundEvent}
   n13 = Create Event()
   > SetNavigationRuleCustom!(execute, self:Widget Object Reference=@n12.Widget, Direction:EUINavigation Enum=Up, InCustomDelegate:Delegate=@n13) -> (then)
   > SetNavigationRuleCustom(self=@n12.Widget, Direction=Down, InCustomDelegate=@n13)
   > n14 = UpdateCommunityRowProvider(self:Fort Activity Details Modal Object Reference)
   n15 = Event OnSetCommunityRowVisibility!() -> (OutputDelegate, then, bVisible:bool) {is=title K2Node_Event}
   > n16 = Set DesiredCommunityRowVisibility(DesiredCommunityRowVisibility:bool=@n15.bVisible) -> (Output_Get:bool)
   SetVisibility(
    execute=[@n14.then, @n16.then]
    self=Get CommunityRow() -> (CommunityRow:Fort Plugin Widget Injection Slot Object Reference)
    InVisibility=Select(Option 0:ESlateVisibility Enum=Collapsed, Option 1:ESlateVisibility Enum=SelfHitTestInvisible, Index=@n16) -> (ReturnValue:ESlateVisibility Enum)
   )
   "On Clicked (Button_CreatorPage)"!() -> (OutputDelegate, then, Button:Common Button Base Object Reference) {is=title K2Node_ComponentBoundEvent}
   > HandleCreatorPageButtonClicked(self:Fort Activity Details Modal Object Reference)
   n17 = "On Clicked (Button_Back_Unbound)"!() -> (OutputDelegate, then, Button:Common Button Base Object Reference) {is=title K2Node_ComponentBoundEvent}
   > n18 = Branch(Condition=Get IsShowingSubModeList())
   > UpdateSubModeVisibility(IsShowingSubModeList=false)
   > SetFocus(self=Get Button_ShowSubModeList() -> (Button_ShowSubModeList:WBP UIKit Button Regular Object Reference))
   Fire Event Element Interacted(execute=@n18.else, Element=@n17.Button)
   > HandleBackButtonClicked(self:Fort Activity Details Modal Object Reference)
   "On Clicked (Button_Favorite)"!() -> (OutputDelegate, then, Button:Common Button Base Object Reference) {is=title K2Node_ComponentBoundEvent}
   > HandleFavoriteClicked(self:Fort Activity Details Display Object Reference=Get ActivityDetailsDisplay())
   > PlaySound2D(Sound=Select(
     Option 0:Sound Base Object Reference=Get Unfavorite_Sound() -> (Unfavorite_Sound:Sound Base Object Reference)
     Option 1:Sound Base Object Reference=Get Favorite_Sound() -> (Favorite_Sound:Sound Base Object Reference)
     Index=Get IsFavorited() -> (IsFavorited:bool)
    ) -> (ReturnValue:Sound Base Object Reference))
   On Create Tags() {is=title K2Node_CustomEvent}
   n19 = Get EntryBox_ActivityTags()
   > Reset!(execute, self:Dynamic Entry Box Object Reference=@n19, bDeleteWidgets:bool=true) -> (then)
   > SetEntrySpacing!(execute, self:Dynamic Entry Box Base Object Reference=@n19, InEntrySpacing_X:double=8, InEntrySpacing_Y:double=8) -> (then)
   n20 = Get ActivityDetailsDisplay()
   > n21 = For Each Loop(Array:Array of User Widget Object References=GetAllEntries!(self:Dynamic Entry Box Base Object Reference=Get EntryBox_ActivityTags!(self:Fort Activity Details Display Object Reference=@n20) -> (EntryBox_ActivityTags)) -> (ReturnValue:Array of User Widget Object References)) -> (Array Element:User Widget Object Reference)
   > n22 = Cast To ActivityBrowserTag!(execute, Object=@n21) -> (then, CastFailed, AsActivity Browser Tag:Activity Browser Tag Object Reference) {is=title K2Node_DynamicCast}
   > CreateTag!(
    execute
    self
    Text:Text=Get Text(self=Get Text_TagDescription!(self:Fort Activity Browser Tag Object Reference=@n22) -> (Text_TagDescription:Common Text Object Reference))
    Tag Type:EFortActivityBrowserTagType Enum=GetTagType!(self:Fort Activity Browser Tag Object Reference=@n22) -> (ReturnValue:EFortActivityBrowserTagType Enum=Default)
   ) -> (then) {is=title K2Node_CallFunction}
   n23 = On Activity Favorite Changed!() -> (OutputDelegate, then, IsFavorite:bool=~, PlayAnimation:bool=~) {is=title K2Node_CustomEvent}
   > UpdateFavoriteStatus!(execute, self, IsFavorite:bool=@n23.IsFavorite, PlayAnimation:bool=@n23.PlayAnimation) -> (then)
   n24 = HandleInputMethodChanged!() -> (OutputDelegate, then, bNewInputType:ECommonInputType Enum) {is=title K2Node_CustomEvent}
   > SetMobileStyle() {is=title K2Node_CallFunction}
   > Set Gamepad Action Bar Style() {is=title K2Node_CallFunction}
   > SetIsEnabled(
    self=Get Button_Details_Hover()
    bInIsEnabled="Not Equal (Enum)"!(A:ECommonInputType Enum=@n24.bNewInputType, B:ECommonInputType Enum=Touch) -> (ReturnValue:bool)
   )
   n25 = OnTooltipDeactivated() {is=title K2Node_CustomEvent}
   > SetFocus(self=Get Button_Accept())
   TestForceIslandCodeVisible() {is=title K2Node_CustomEvent}
   > SetVisibility(self=Get WBP_IslandCode() -> (WBP_IslandCode:WBP Island Label And Details Object Reference), InVisibility=SelfHitTestInvisible)
   OnSwitcherButtonIslandUpdatedClicked!() -> (OutputDelegate, then, Button:Common Button Base Object Reference) {is=title K2Node_CustomEvent}
   > ToggleAmISubscribed(self:Discover Activity View Model Object Reference=Get FortIslandDetailsVM_Manual())
   DebugPlaylistInfoTrue() {is=title K2Node_CustomEvent}
   > Set Visibility(Visibility=HitTestInvisible, self=Get VB_Version() -> (VB_Version:Vertical Box Object Reference))
   n26 = On Update Link Code!() -> (OutputDelegate, then, LinkCode:Text, IsCreativeActivity:bool=~) {is=title K2Node_CustomEvent}
   > UpdateLinkCode!(execute, self, LinkCode:Text=@n26.LinkCode, IsCreativeActivity:bool=@n26.IsCreativeActivity) -> (then)
   On Refresh Disabled State() {is=title K2Node_CustomEvent}
   > RefreshDisabledState()
   n27 = On Update CCU!() -> (OutputDelegate, then, CCUCount:int) {is=title K2Node_CustomEvent}
   > SetCCUCountValueText!(execute, self, CCUCount:int=@n27.CCUCount) -> (then)
   On Reset Widget State() {is=title K2Node_CustomEvent}
   n28 = Get DetailsIntro() -> (DetailsIntro:Widget Animation Object Reference)
   > PlayAnimation(InAnimation=@n28, StartAtTime=GetEndTime(self=@n28))
   n29 = On Details Updated!() -> (OutputDelegate, then, IsDetailsValid:bool=~, IsFavoritable:bool=~) {is=title K2Node_CustomEvent}
   > SetButtonsInteractable!(execute, self, IsInteractionEnabled:bool=@n29.IsDetailsValid) -> (then)
   > SetupTimerViewModel!(execute, self) -> (then, IsCountdown:bool)
   n30 = IsActivityTimeLocked!(
    self:Fort Game Activity Object Reference=GetSelectedGameActivity()
    Context:Local Player Object Reference=GetOwningLocalPlayer!(self:Widget Object Reference) -> (ReturnValue:Local Player Object Reference)
   ) -> (ReturnValue:bool)
   > Set IsAcceptButtonEnabled(IsAcceptButtonEnabled:bool=NOT Boolean(A=@n30)) -> (Output_Get:bool)
   n31 = Get Button_Accept()
   n32 = Get LastFocusedWidget()
   > n33 = Branch(Condition=AND Boolean(A=@n30, B="Equal (Object)"(A=@n32, B=@n31)))
   > n34 = Set LastFocusedWidget(LastFocusedWidget:User Widget Object Reference=Get Button_Favorite()) -> (Output_Get:User Widget Object Reference)
   UpdateDetailVisibility!(execute=[@n34.then, @n33.else], self, IsFavoritable:bool=@n29.IsFavoritable) -> (then)
   > RefreshContentWarnings()
   > UpdateSubModeVisibility(IsShowingSubModeList=Get IsShowingDetailsOverride() -> (IsShowingDetailsOverride:bool))
   n35 = GetSelectedGameActivity()
   > ShowActivityData!(
    execute
    self:Fort Squeegee Widget Injection Slot Object Reference=[Get SqueegeeSlot_LockInfo(), Get SqueegeeSlot_Card() -> (SqueegeeSlot_Card:Squeegee Injection Slot Object Reference)]
    GameActivity:Fort Game Activity Object Reference=@n35
   ) -> (then)
   > UpdateDetailsTexts()
   > n36 = On Create Tags() {is=title K2Node_CallFunction}
   SetAcceptButtonState!(execute=[@n37.then, @n36.then], self, IsEnable:bool=Get IsAcceptButtonEnabled() -> (IsAcceptButtonEnabled:bool)) -> (then)
   > SetVisibility(
    self=Get NamedSlot_AdditionalDetail() -> (NamedSlot_AdditionalDetail:Named Slot Object Reference)
    InVisibility=Select(Option 0:ESlateVisibility Enum=Visible, Option 1:ESlateVisibility Enum=Collapsed, Index=IsContentGated(self=@n35)) -> (ReturnValue:ESlateVisibility Enum)
   )
   > RefreshDisabledState()
   > SetMobileStyleSqueegeeSlotCard()
   On Lock State Changed() {is=title K2Node_CustomEvent}
   > n37 = ShowActivityData(self=[Get SqueegeeSlot_LockInfo(), Get SqueegeeSlot_Card()], GameActivity=GetSelectedGameActivity())
   On Description Scroll Restarted() {is=title K2Node_CustomEvent}
   > PlayAnimation(InAnimation=Get DescriptionScrollRestarted() -> (DescriptionScrollRestarted:Widget Animation Object Reference))
   Event Construct()
   > n38 = Sequence!(execute) -> (then_0, then_1, then_2:Exec, then_3:Exec)
   > n39 = Cast To ActivityDetailsDisplay!(execute, Object=Get DetailsDisplay_SelectedActivity() -> (DetailsDisplay_SelectedActivity:Fort Activity Details Display Object Reference)) -> (then, CastFailed, AsActivity Details Display:Activity Details Display Object Reference) {is=title K2Node_DynamicCast}
   > Set ActivityDetailsDisplay(ActivityDetailsDisplay:Activity Details Display Object Reference=@n39) -> (Output_Get:Activity Details Display Object Reference)
   > Set LastFocusedWidget(LastFocusedWidget=Get Button_Accept())
   > Add(
    TargetArray:Array of Common Button Base Object References=Get Buttons() -> (Buttons:Array of Common Button Base Object References)
    NewItem:"Common Button Base Object Reference (by ref)"=Get Button_Accept()
   ) {is=title K2Node_CallArrayFunction}
   > Add(TargetArray=Get Buttons(), NewItem=Get Button_Favorite())
   > Set Gamepad Action Bar Style()
   > n40 = SetButtonsInteractable(IsInteractionEnabled=false)
   UpdateDetailVisibility(execute=[@n40.then, @n62.then], IsFavoritable=false)
   > UpdateFavoriteStatus(IsFavorite=false, PlayAnimation=false)
   > UpdateVisualForArabic()
   > SetNavigationRules()
   n41 = Get WBP_ActivityDetails_FollowCreator_CoachMark() -> (WBP_ActivityDetails_FollowCreator_CoachMark:WBP Activity Details Follow Creator Coach Mark Object Reference)
   > Branch(Condition=ShouldShowOnboarding!(self:Fort Sidebar Onboard Tooltip Widget Object Reference=@n41) -> (ReturnValue:bool))
   > ActivateWidget(self=@n41)
   > Bind Event to On Widget Deactivated(
    self:Common Activatable Widget Object Reference=Get WBP_UIKit_CoachMark!(self:WBP Activity Details Follow Creator Coach Mark Object Reference=@n41) -> (WBP_UIKit_CoachMark:WBP UIKit Coach Mark Object Reference)
    Delegate=@n25
   ) {is=title K2Node_AddDelegate}
   Parent: Construct(execute=@n38.then_1)
   > Set bIsLoading(bIsLoading:bool=false) -> (Output_Get:bool)
   > PlayAnimationForward(InAnimation=Get OnOpen() -> (OnOpen:Widget Animation Object Reference))
   DelayUntilNextTick!(execute=[@n38.then_3, @n43.then]) -> (then)
   > n42 = GetScreenSizeAndSetPadding!(execute, self) -> (then, IsScreenSizeZero:bool)
   > n43 = Branch(Condition=@n42)
   n44 = Create Event()
   n45 = Get ActivityDetailsDisplay()
   Bind Event to On Activity Favorite Changed Event Dispatcher(execute=@n38.then_2, self:Activity Details Display Object Reference=@n45, Delegate=@n44) {is=title K2Node_AddDelegate}
   n46 = Create Event()
   > Bind Event to On Refresh Disabled State Event Dispatcher(self:Activity Details Display Object Reference=@n45, Delegate=@n46) {is=title K2Node_AddDelegate}
   n47 = Create Event()
   > Bind Event to On Update CCUEvent Dispatcher(self:Activity Details Display Object Reference=@n45, Delegate=@n47) {is=title K2Node_AddDelegate}
   n48 = Create Event()
   > Bind Event to On Reset Widget State Event Dispatcher(self:Activity Details Display Object Reference=@n45, Delegate=@n48) {is=title K2Node_AddDelegate}
   n49 = Create Event()
   > Bind Event to On Details Updated Event Dispatcher(self:Activity Details Display Object Reference=@n45, Delegate=@n49) {is=title K2Node_AddDelegate}
   n50 = Create Event()
   > Bind Event to On Lock State Changed Event Dispatcher(self:Activity Details Display Object Reference=@n45, Delegate=@n50) {is=title K2Node_AddDelegate}
   n51 = Create Event()
   > Bind Event to On Description Scroll Restarted Event Dispatcher(self:Activity Details Display Object Reference=@n45, Delegate=@n51) {is=title K2Node_AddDelegate}
   n52 = Create Event()
   > Bind Event to On Update Link Code Event Dispatcher(self:Activity Details Display Object Reference=@n45, Delegate=@n52) {is=title K2Node_AddDelegate}
   n53 = Create Event()
   n54 = Get CommonInputSubsystem()
   > Bind Event to On Input Method Changed(self=@n54, Delegate=@n53)
   n55 = Event OnPreviewImageChanged!() -> (OutputDelegate, then, bIsLoading:bool, Texture:Texture Object Reference) {is=title K2Node_Event}
   > UpdateImage!(execute, self, bInIsLoading:bool=@n55.bIsLoading, InTexture:Texture Object Reference=@n55.Texture) -> (then)
   n56 = Event OnAttributionsVisibilityUpdated!() -> (OutputDelegate, then, bShouldShowAttributionsOption:bool) {is=title K2Node_Event}
   > n57 = Branch(Condition=@n56.bShouldShowAttributionsOption)
   > n58 = AddInteraction!(
    execute
    self:UIKit Context Menu Action Context Object Reference=Get ContextMenuActionContext() -> (ContextMenuActionContext:UIKit Context Menu Action Context Object Reference)
    InActionInfo_DisplayName:Text="LOCTEXT(\"View Attributions\")"
    InActionInfo_IconTexture:Texture 2D Soft Object Reference
    InActionInfo_IconMaterial:Material Instance Soft Object Reference
    InActionInfo_SortPriority:int=10
    InActionInfo_bIsAvailable:bool=True
    InActionInfo_bIsEnabled:bool=True
    InActionInfo_bIsSelected:bool=False
    InActionInfo_bAddDivider:bool=False
    InActionInfo_bPositionDividerAfter:bool=False
    InActionInfo_bCloseMenuOnSelect:bool=False
    InActionInfo_InteractionNameID:Name=ViewAttributions
    InActionInfo_Interaction:Data Table Row Handle Structure=(DataTable=None,RowName="")
    InActionInfo_PressedDelegate:Delegate="(null).None"
    InActionInfo_ReleasedDelegate:Delegate="(null).None"
    InInteractionDelegate:"Delegate (by ref)"=Create Event()
   ) -> (then)
   SetContext!(
    execute=[@n58.then, @n59.then]
    self:WBP UIKit Context Menu Anchor Button Base Object Reference=Get Button_ContextMenu() -> (Button_ContextMenu:WBP UIKit Context Menu Anchor Button Object Reference)
    InteractionContext_Data:UIKit Context Menu Action Context Object Reference=Get ContextMenuActionContext()
   ) -> (then)
   > Update Context Menu Visibility()
   n59 = RemoveInteraction!(
    execute=@n57.else
    self:UIKit Context Menu Action Context Object Reference=Get ContextMenuActionContext()
    InteractionNameID:Name=ViewAttributions
   ) -> (then)
   Event OnLockStateChanged() {is=title K2Node_Event}
   > On Lock State Changed() {is=title K2Node_CallFunction}
   n60 = Event PreConstruct()
   > DisableButtonWithReason!(
    execute
    self:Common Button Base Object Reference=Get Button_TimeLocked() -> (Button_TimeLocked:WBP UIKit Button Loud Object Reference)
    DisabledReason:"Text (by ref)"=Format Text(Format="LOCTEXT(\"The event is not live yet!\")")
   ) -> (then)
   > SetCultureLayout()
   > n61 = Branch(Condition=@n60.IsDesignTime)
   > Is Valid(InputObject=GetParent!(self:Widget Object Reference) -> (ReturnValue:Panel Widget Object Reference))
   > n62 = SetRenderOpacity(
    self=Get GridPanel_DynamicDetails() -> (GridPanel_DynamicDetails:Grid Panel Object Reference)
    InOpacity=Select(Option 0:double=0, Option 1:double=1, Index=Get DesignTimeShowDetails() -> (DesignTimeShowDetails:bool)) -> (ReturnValue:double)
   )
   SetMobileStyleSqueegeeSlotCard(execute=@n61.else)
   > SetMinDimensions!(
    execute
    self:Common Button Base Object Reference=Get Button_Subscribe() -> (Button_Subscribe:WBP UIKit Button Regular Object Reference)
    InMinWidth:int=88
    InMinHeight:int=88
   ) -> (then)
   n63 = Event OnButtonToggleValueSet!() -> (OutputDelegate, then, ButtonToggle:Common Button Base Object Reference, bValue:bool) {is=title K2Node_Event}
   > n64 = Cast To IslandPrivacyButton!(execute, Object=@n63.ButtonToggle) -> (then, CastFailed, AsIsland Privacy Button:Island Privacy Button Object Reference) {is=title K2Node_DynamicCast}
   > SetToggleState!(execute, self:Island Privacy Button Object Reference=@n64, InIsOn:bool=@n63.bValue) -> (then)
   Event OnRemovedFromFocusPath!() -> (OutputDelegate, then, InFocusEvent:Focus Event Structure) {is=title K2Node_Event}
   > Branch(Condition=IsActivated!(self:Common Activatable Widget Object Reference) -> (ReturnValue:bool))
   > RequestRefreshFocus(self:Common Activatable Widget Object Reference)
   Event On Activated()
   > SpawnSound2D!(
    execute
    Sound:Sound Base Object Reference=Get IntroWhooshSound() -> (IntroWhooshSound:Sound Base Object Reference)
    VolumeMultiplier:float=1
    PitchMultiplier:float=1
    StartTime:float=0
    ConcurrencySettings:Sound Concurrency Object Reference
    bPersistAcrossLevelTransition:bool=false
    bAutoDestroy:bool=true
   ) -> (then, ReturnValue:Audio Component Object Reference)
   > Set IsShowingSubModeList(IsShowingSubModeList:bool=false) -> (Output_Get:bool)
   n65 = ShouldOfferSubModeSelection!() -> (bShouldOfferSubModeSelection:bool)
   n66 = Get bAllowSubmodeSelect() -> (bAllowSubmodeSelect:bool)
   > SetVisibility(
    self=Get VerticalBox_ModeSelectionAvailable() -> (VerticalBox_ModeSelectionAvailable:Vertical Box Object Reference)
    InVisibility=Select(Option 0:ESlateVisibility Enum=Collapsed, Option 1:ESlateVisibility Enum=SelfHitTestInvisible, Index=AND Boolean(A=@n65, B=@n66)) -> (ReturnValue:ESlateVisibility Enum)
   )
   > SetVisibility(
    self=[
     Get HorizontalBox_ModeSelectionAvailable_SubMode() -> (HorizontalBox_ModeSelectionAvailable_SubMode:Horizontal Box Object Reference)
     Get HorizontalBox_ModeSelectionAvailable() -> (HorizontalBox_ModeSelectionAvailable:Horizontal Box Object Reference)
    ]
    InVisibility=Select(
     Option 0=SelfHitTestInvisible
     Option 1=Collapsed
     Index=IsCreativeModeSetActivity!(self:Fort Activity Details Modal Object Reference) -> (ReturnValue:bool)
    )
   )
   n67 = GetInvalidActivityReason!(self:Fort Activity Details Modal Object Reference) -> (ReturnValue:EFortInvalidActivityReason Enum=None)
   > SetIsEnabled(
    self=Get Button_ShowSubModeList()
    bInIsEnabled=AND Boolean(
     A=IsLocalPlayerPartyLeader!(self:Fort Activity Details Modal Object Reference) -> (ReturnValue:bool)
     B="Equal (Enum)"(A:EFortInvalidActivityReason Enum=@n67, B:EFortInvalidActivityReason Enum=None)
    )
   )
   > SetupContextMenu()
   n68 = GetPlayerController!(PlayerIndex:int=0) -> (ReturnValue:Player Controller Object Reference)
   > SetIsEnabled(self=Get Button_Details_Hover(), bInIsEnabled=NOT Boolean(A=IsUsingTouch(PlayerController=@n68)))
   > ClearChildren(self=Get O_Modal())
   > Set FortIslandDetailsVM_Manual(FortIslandDetailsVM_Manual:Fort Island Details VM Object Reference=Get FortIslandDetailsVM() -> (FortIslandDetailsVM:Fort Island Details VM Object Reference)) -> (Output_Get:Fort Island Details VM Object Reference)
   Event OnInitialized()
   > n69 = GetDynamicMaterial(self=Get Thumbnail() -> (Thumbnail:Image Object Reference))
   > n70 = Get Texture Parameter Value!(execute, self:Material Instance Dynamic Object Reference=@n69, ParameterName:Name=Get ThumbnailParam() -> (ThumbnailParam:Name=None)) -> (then, ReturnValue:Texture Object Reference) {is=title K2Node_CallFunction}
   > Set DefaultImage(DefaultImage:Texture Object Reference=@n70) -> (Output_Get:Texture Object Reference)
   > SetMobileStyle()
   n71 = Event OnSelectedSubModeTextChanged!() -> (OutputDelegate, then, ActivityText:"Text (by ref)", SettingText:"Text (by ref)") {is=title K2Node_Event}
   n72 = Get IsHabaneroEnabled() -> (IsHabaneroEnabled:bool)
   > SetText!(
    execute
    self:Fort CTAButton Object Reference=Get Button_ShowSubModeList()
    ButtonText:"Text (by ref)"=Select(
     Option 0:"Text (by ref)"=@n71.ActivityText
     Option 1:"Text (by ref)"=Select(
      Option 0:Text="LOCTABLE(\"/Game/UI/Habanero/HabaneroStrings\", \"submodeselect_rankedoff\")"
      Option 1:Text="LOCTABLE(\"&HabaneroStrings\", \"submodeselect_rankedon\")"
      Index=@n72
     ) -> (ReturnValue:Text)
     Index=Get IsHabaneroMode() -> (IsHabaneroMode:bool)
    ) -> (ReturnValue:"Text (by ref)")
   ) -> (then)
   n73 = TextIsEmpty(InText=@n71.SettingText)
   n74 = Format Text!(Format="LOCTEXT(\"{0} - {1}\")", 0:"Text (by ref)"=@n71.ActivityText, 1:"Text (by ref)"=@n71.SettingText) -> (Result)
   > SetSecondaryText!(
    execute
    self:Fort CTAButton Object Reference=Get Button_ShowSubModeList()
    ButtonSecondaryText:"Text (by ref)"=Select(
     Option 0:"Text (by ref)"=@n71.SettingText
     Option 1:"Text (by ref)"=Select(Option 0=@n74, Option 1=@n71.ActivityText, Index=@n73)
     Index=Get IsHabaneroMode()
    ) -> (ReturnValue:"Text (by ref)")
   ) -> (then)
   n75 = Event OnShowSubModeFirstTimeNotification!() -> (OutputDelegate, then, ActivityAndSettingText:"Text (by ref)") {is=title K2Node_Event}
   > ShowNotification(self:WBP Popup Tooltip Object Reference=Get DiscoveryTooltip())
   > Branch(Condition=IsCreativeModeSetActivity())
   > SetFirstLineText!(execute, self:WBP Popup Tooltip Object Reference=Get DiscoveryTooltip(), FirstLineText:Text=@n75.ActivityAndSettingText) -> (then)
   Event On Deactivated()
   > HideNotification(self=Get DiscoveryTooltip())
   > SpawnSound2D(
    Sound=Get OutroWhooshSound() -> (OutroWhooshSound:Sound Base Object Reference)
    VolumeMultiplier=1
    PitchMultiplier=1
    StartTime=0
    bPersistAcrossLevelTransition=false
    bAutoDestroy=true
   )
   Event OnSubModeSelectionChanged() {is=title K2Node_Event}
   > UpdateSubModeVisibility(IsShowingSubModeList=false)
   > SetFocus(self=Get Button_ShowSubModeList())
   n76 = Event OnButtonToggleAvailabilityUpdated!() -> (OutputDelegate, then, ButtonToggle:Common Button Base Object Reference, bIsAvailable:bool, bIsForcedMode:bool) {is=title K2Node_Event}
   > n77 = Cast To IslandPrivacyButton(Object=@n76.ButtonToggle)
   > SetAvailability!(execute, self:Island Privacy Button Object Reference=@n77, InIsAvailable:bool=@n76.bIsAvailable) -> (then)
   > SetForcedMode!(execute, self:Island Privacy Button Object Reference=@n77, bIsForcedMode:bool=@n76.bIsForcedMode) -> (then)
   Event Destruct()
   > Unbind Event from On Activity Favorite Changed Event Dispatcher(self:Activity Details Display Object Reference=@n45, Delegate=@n44) {is=title K2Node_RemoveDelegate}
   > Unbind Event from On Refresh Disabled State Event Dispatcher(self:Activity Details Display Object Reference=@n45, Delegate=@n46) {is=title K2Node_RemoveDelegate}
   > Unbind Event from On Update CCUEvent Dispatcher(self:Activity Details Display Object Reference=@n45, Delegate=@n47) {is=title K2Node_RemoveDelegate}
   > Unbind Event from On Reset Widget State Event Dispatcher(self:Activity Details Display Object Reference=@n45, Delegate=@n48) {is=title K2Node_RemoveDelegate}
   > Unbind Event from On Details Updated Event Dispatcher(self:Activity Details Display Object Reference=@n45, Delegate=@n49) {is=title K2Node_RemoveDelegate}
   > Unbind Event from On Lock State Changed Event Dispatcher(self:Activity Details Display Object Reference=@n45, Delegate=@n50) {is=title K2Node_RemoveDelegate}
   > Unbind Event from On Description Scroll Restarted Event Dispatcher(self:Activity Details Display Object Reference=@n45, Delegate=@n51) {is=title K2Node_RemoveDelegate}
   > Unbind Event from On Update Link Code Event Dispatcher(self:Activity Details Display Object Reference=@n45, Delegate=@n52) {is=title K2Node_RemoveDelegate}
   > Unbind Event from On Input Method Changed(self=@n54, Delegate=@n53)
   Reroute Node()
   > SetIcon!(
    execute
    self:WBP Habanero Icon Object Reference=Get Icon_Habanero() -> (Icon_Habanero:WBP Habanero Icon Object Reference)
    ForegroundInstance:Material Instance Soft Object Reference=Reroute Node(InputPin:Material Instance Soft Object Reference) -> (OutputPin:Material Instance Soft Object Reference)
    BackgroundInstance:Material Instance Soft Object Reference=Reroute Node()
   ) -> (then)
   > Set Text(Text=Reroute Node(InputPin:Text) -> (OutputPin:Text), self=Get Text_Habanero() -> (Text_Habanero:Common Text Object Reference))
   n78 = Select(Option 0:ESlateVisibility Enum=Collapsed, Option 1:ESlateVisibility Enum=SelfHitTestInvisible, Index=Get IsHabaneroMode()) -> (ReturnValue:ESlateVisibility Enum)
   > SetVisibility(self=Get SizeBox_Habanero() -> (SizeBox_Habanero:Size Box Object Reference), InVisibility=@n78)
   > SetVisibility(self=Get Button_Activity_Habanero() -> (Button_Activity_Habanero:Activity Habanero Button Object Reference), InVisibility=@n78)
   n79 = Set LastFocusedWidget(execute=@n80.then, LastFocusedWidget=Get Button_Subscribe())
   > SetFocus(self=@n79)
   HandleFavoriteButtonClicked(self:Fort Activity Details Modal Object Reference)
   Event OnIsFavoriteChanged!() -> (OutputDelegate, then, bIsFavorite:bool) {is=title K2Node_Event}
   Get IsHabaneroMode()
   Get IsHabaneroEnabled()
   Get Image_KeyArt_768x768() -> (Image_KeyArt_768x768:Image Object Reference)
   n80 = RemoveFromParent(execute=@n82.then, self=@n2)
   n81 = Create Event()
   Unbind Event from On Clicked(execute=@n7.then, self:Common Button Base Object Reference=Get ToggleButton_IslandUpdates(self=@n2), Delegate=@n81) {is=title K2Node_RemoveDelegate}
   > n82 = Unbind Event from On Widget Deactivated(self:Common Activatable Widget Object Reference=@n2, Delegate=@n7) {is=title K2Node_RemoveDelegate}
  ]}
 GetAcceptButtonEnableState {nodes [
   n1 = GetAcceptButtonEnableState!() -> (then, TimerState:ECountdownTimerState Enum=None) {is=title K2Node_FunctionEntry}
   > n2 = Branch(Condition="Equal (Enum)"(A:ECountdownTimerState Enum=@n1, B:ECountdownTimerState Enum=Live))
   Return Node!(execute=[@n2.then, @n4.Is Not Valid], IsEnable:bool=true)
   n3 = GetSelectedGameActivity()
   n4 = Is Valid(exec=@n2.else, InputObject=@n3)
   n5 = GetOwningLocalPlayer()
   > Return Node(IsEnable=NOT Boolean(A=IsActivityTimeLocked(self=@n3, Context=@n5)))
  ]}
 GetScreenSizeAndSetPadding {nodes [
   GetScreenSizeAndSetPadding() {is=title K2Node_FunctionEntry}
   n1 = GetLocalSize!(Geometry:"Geometry Structure (by ref)"=GetPaintSpaceGeometry!(self:Widget Object Reference=Get Overlay_6() -> (Overlay_6:Overlay Object Reference)) -> (ReturnValue:"Geometry Structure (by ref)")) -> (ReturnValue:Vector 2D Structure)
   n2 = IsZero2D!(A:"Vector 2D Structure (by ref)"=@n1) -> (ReturnValue:bool)
   > n3 = Branch(Condition=@n2)
   > Return Node!(execute, IsScreenSizeZero:bool=@n2)
   n4 = SlotAsGridSlot(Widget=Get SZ_Content_DynamicPadding() -> (SZ_Content_DynamicPadding:Safe Zone Object Reference))
   n5 = Break Margin!(Margin:"Margin Structure (by ref)"=Get Padding!(self:Grid Slot Object Reference=@n4) -> (Padding:Margin Structure)) -> (Left:float=~, Top:float=~, Right:float=~, Bottom:float=~) {is=title K2Node_BreakStruct}
   n6 = BreakVector2D(InVec=@n1)
   Set Padding!(
    execute=@n3.else
    Padding:Margin Structure=Make Margin(Left=@n5, Top=float - float(A=@n6.Y, B=128), Right=@n5.Right, Bottom=@n5.Bottom)
    self:Grid Slot Object Reference=@n4
   ) -> (then, Output_Get:Margin Structure)
   > Return Node(IsScreenSizeZero=@n2)
  ]}
 NavigateDownFromTopRow {nodes [
   NavigateDownFromTopRow() {is=title K2Node_FunctionEntry}
   > n1 = NavigateToCommunityRow!(execute, self) -> (then, ReturnValue:Widget Object Reference)
   > n2 = Is Valid(InputObject=@n1)
   n3 = NavigateToDescription!(execute=@n2.Is Not Valid, self) -> (then, Focus Widget:Button Object Reference)
   > Return Node!(execute, ReturnValue:Widget Object Reference=@n3)
   Return Node(execute=@n2.Is Valid, ReturnValue=@n1)
  ]}
 NavigateFromAcceptButton {nodes [
   n1 = NavigateFromAcceptButton!() -> (then, Navigation:EUINavigation Enum) {is=title K2Node_FunctionEntry}
   > n2 = Switch on EUINavigation!(execute, Selection:EUINavigation Enum=@n1) -> (Left:Exec, Right:Exec, Up:Exec, Down:Exec, Next:Exec, Previous:Exec, Invalid:Exec) {is=title K2Node_SwitchEnum}
   n3 = "Equal Exactly (String)"(A=GetCurrentLanguage!() -> (ReturnValue:String), B=ar)
   > n4 = Branch(Condition=@n3)
   n5 = Set LastFocusedWidget(execute=[@n6.else, @n4.then], LastFocusedWidget=Get Button_Favorite())
   > Return Node(ReturnValue=@n5)
   n6 = Branch(execute=@n2.Right, Condition=@n3)
   n7 = ShouldOfferSubModeSelection()
   n8 = Get bAllowSubmodeSelect()
   n9 = Set LastFocusedWidget(
    execute=@n2.Up
    LastFocusedWidget=Select(
     Option 0:User Widget Object Reference=Get Button_Accept()
     Option 1:User Widget Object Reference=Get Button_ShowSubModeList()
     Index=AND Boolean(A=@n7, B=@n8)
    ) -> (ReturnValue:User Widget Object Reference)
   )
   > Return Node(ReturnValue=@n9)
   n10 = NavigateDownFromTopRow!(execute=@n2.Down, self) -> (then, ReturnValue:Widget Object Reference) {is=title K2Node_CallFunction}
   > Return Node(ReturnValue=@n10)
  ]}
 NavigateFromContentGatingCard {nodes [
   n1 = NavigateFromContentGatingCard!() -> (then, Navigation:EUINavigation Enum) {is=title K2Node_FunctionEntry}
   > n2 = Switch on EUINavigation(Selection=@n1)
   n3 = "Equal Exactly (String)"(A=GetCurrentLanguage(), B=ar)
   > n4 = Branch(Condition=@n3)
   n5 = Set LastFocusedWidget(execute=[@n6.then, @n4.else], LastFocusedWidget=Get Button_ContextMenu())
   > Return Node(ReturnValue=@n5)
   n6 = Branch(execute=@n2.Right, Condition=@n3)
   n7 = ShouldOfferSubModeSelection()
   n8 = Get bAllowSubmodeSelect()
   n9 = Set LastFocusedWidget(
    execute=@n2.Up
    LastFocusedWidget=Select(Option 0=Get Button_Accept(), Option 1=Get Button_ShowSubModeList(), Index=AND Boolean(A=@n7, B=@n8))
   )
   > Return Node(ReturnValue=@n9)
   n10 = NavigateDownFromTopRow(execute=@n2.Down)
   > Return Node(ReturnValue=@n10)
  ]}
 NavigateFromContextMenuButton {nodes [
   n1 = NavigateFromContextMenuButton!() -> (then, Navigation:EUINavigation Enum) {is=title K2Node_FunctionEntry}
   > n2 = Switch on EUINavigation(Selection=@n1)
   n3 = "Equal Exactly (String)"(A=GetCurrentLanguage(), B=ar)
   > n4 = Branch(Condition=@n3)
   n5 = GetOwningLocalPlayer()
   n6 = GetSelectedGameActivity()
   n7 = Set LastFocusedWidget(
    execute=[@n10.else, @n4.then]
    LastFocusedWidget=Select(
     Option 0=Get Button_ContextMenu()
     Option 1=Get SqueegeeSlot_Card()
     Index=HasAgeRating!(self:Fort Game Activity Object Reference=@n6, PlayerContext:Local Player Object Reference=@n5) -> (ReturnValue:bool)
    )
   )
   > Return Node(ReturnValue=@n7)
   n8 = GetRuntimeOptions()
   n9 = Set LastFocusedWidget(
    execute=[@n10.then, @n4.else]
    LastFocusedWidget=Select(
     Option 0=Get Button_Favorite()
     Option 1=Get Button_Subscribe()
     Index=GetAreIslandNotificationsEnabled!(self:Fort Runtime Options Object Reference=@n8) -> (ReturnValue:bool)
    )
   )
   > Return Node(ReturnValue=@n9)
   n10 = Branch(execute=@n2.Right, Condition=@n3)
   n11 = ShouldOfferSubModeSelection()
   n12 = Get bAllowSubmodeSelect()
   n13 = Set LastFocusedWidget(
    execute=@n2.Up
    LastFocusedWidget=Select(Option 0=Get Button_ContextMenu(), Option 1=Get Button_ShowSubModeList(), Index=AND Boolean(A=@n11, B=@n12))
   )
   > Return Node(ReturnValue=@n13)
   n14 = NavigateDownFromTopRow(execute=@n2.Down)
   > Return Node(ReturnValue=@n14)
  ]}
 NavigateFromDescription {nodes [
   n1 = NavigateFromDescription!() -> (then, Navigation:EUINavigation Enum) {is=title K2Node_FunctionEntry}
   > n2 = Switch on EUINavigation(Selection=@n1)
   n3 = NavigateToCommunityRow(execute=@n2.Up)
   > n4 = Is Valid(InputObject=@n3)
   n5 = NavigateToTopRow!(execute=@n4.Is Not Valid, self) -> (then, ReturnValue:Widget Object Reference)
   > Return Node(ReturnValue=@n5)
   Return Node(execute=@n4.Is Valid, ReturnValue=@n3)
  ]}
 NavigateFromFavoriteButton {nodes [
   n1 = NavigateFromFavoriteButton!() -> (then, Navigation:EUINavigation Enum) {is=title K2Node_FunctionEntry}
   > n2 = Switch on EUINavigation(Selection=@n1)
   n3 = NavigateDownFromTopRow(execute=@n2.Down)
   > Return Node(ReturnValue=@n3)
   n4 = "Equal Exactly (String)"(A=GetCurrentLanguage(), B=ar)
   n5 = Branch(execute=@n2.Left, Condition=@n4)
   n6 = GetRuntimeOptions()
   n7 = Set LastFocusedWidget(
    execute=[@n10.else, @n5.then]
    LastFocusedWidget=Select(Option 0=Get Button_ContextMenu(), Option 1=Get Button_Subscribe(), Index=GetAreIslandNotificationsEnabled(self=@n6))
   )
   > Return Node(ReturnValue=@n7)
   n8 = GetSelectedGameActivity()
   n9 = Set LastFocusedWidget(
    execute=[@n10.then, @n5.else]
    LastFocusedWidget=Select(Option 0=Get Button_Accept(), Option 1=Get SqueegeeSlot_LockInfo(), Index=IsContentGated(self=@n8))
   )
   > Return Node(ReturnValue=@n9)
   n10 = Branch(execute=@n2.Right, Condition=@n4)
   n11 = ShouldOfferSubModeSelection()
   n12 = Get bAllowSubmodeSelect()
   n13 = Set LastFocusedWidget(
    execute=@n2.Up
    LastFocusedWidget=Select(Option 0=Get Button_Favorite(), Option 1=Get Button_ShowSubModeList(), Index=AND Boolean(A=@n11, B=@n12))
   )
   > Return Node(ReturnValue=@n13)
  ]}
 NavigateFromLockButton {nodes [
   n1 = NavigateFromLockButton!() -> (then, Navigation:EUINavigation Enum) {is=title K2Node_FunctionEntry}
   > n2 = Switch on EUINavigation(Selection=@n1)
   n3 = "Equal Exactly (String)"(A=GetCurrentLanguage(), B=ar)
   > n4 = Branch(Condition=@n3)
   n5 = Set LastFocusedWidget(execute=[@n6.else, @n4.then], LastFocusedWidget=Get Button_Favorite())
   > Return Node(ReturnValue=@n5)
   n6 = Branch(execute=@n2.Right, Condition=@n3)
   n7 = NavigateDownFromTopRow(execute=@n2.Down)
   > Return Node(ReturnValue=@n7)
  ]}
 NavigateFromSubscribeButton {nodes [
   n1 = NavigateFromSubscribeButton!() -> (then, Navigation:EUINavigation Enum) {is=title K2Node_FunctionEntry}
   > n2 = Switch on EUINavigation(Selection=@n1)
   n3 = "Equal Exactly (String)"(A=GetCurrentLanguage(), B=ar)
   > n4 = Branch(Condition=@n3)
   n5 = Set LastFocusedWidget(execute=[@n7.else, @n4.then], LastFocusedWidget=Get Button_ContextMenu())
   > Return Node(ReturnValue=@n5)
   n6 = Set LastFocusedWidget(execute=[@n7.then, @n4.else], LastFocusedWidget=Get Button_Favorite())
   > Return Node(ReturnValue=@n6)
   n7 = Branch(execute=@n2.Right, Condition=@n3)
   n8 = ShouldOfferSubModeSelection()
   n9 = Get bAllowSubmodeSelect()
   n10 = Set LastFocusedWidget(
    execute=@n2.Up
    LastFocusedWidget=Select(Option 0=Get Button_Subscribe(), Option 1=Get Button_ShowSubModeList(), Index=AND Boolean(A=@n8, B=@n9))
   )
   > Return Node(ReturnValue=@n10)
   n11 = NavigateDownFromTopRow(execute=@n2.Down)
   > Return Node(ReturnValue=@n11)
  ]}
 NavigateToCommunityRow {nodes [
   NavigateToCommunityRow() {is=title K2Node_FunctionEntry}
   n1 = Get CommunityRow()
   > n2 = Branch(Condition=IsVisible(self=@n1))
   > n3 = Cast To CommonActivatableWidget(Object=GetInjectedWidget!(self:Fort Plugin Widget Injection Slot Object Reference=@n1) -> (ReturnValue:Widget Object Reference))
   > ScrollWidgetIntoView!(
    execute
    self:Scroll Box Object Reference=Get ScrollBox() -> (ScrollBox:Scroll Box Object Reference)
    WidgetToFind:Widget Object Reference=@n1
    AnimateScroll:bool=true
    ScrollDestination:EDescendantScrollDestination Enum=IntoView
    Padding:float=100
   ) -> (then)
   > Set LastFocusedWidget(LastFocusedWidget=@n3)
   > Return Node(ReturnValue=GetDesiredFocusTarget(self=@n3))
   Return Node(execute=[@n3.CastFailed, @n2.else])
  ]}
 NavigateToDescription {nodes [
   NavigateToDescription() {is=title K2Node_FunctionEntry}
   > ScrollWidgetIntoView(
    self=Get ScrollBox()
    WidgetToFind=Get Text_ActivityDescription() -> (Text_ActivityDescription:Common Text Object Reference)
    AnimateScroll=true
    ScrollDestination=IntoView
    Padding=100
   )
   > Return Node!(execute, Focus Widget:Button Object Reference=Get Button_Details_Hover())
  ]}
 NavigateToTopRow {nodes [
   NavigateToTopRow() {is=title K2Node_FunctionEntry}
   > ScrollWidgetIntoView(
    self=Get ScrollBox()
    WidgetToFind=Get Text_ActivityName!() -> (Text_ActivityName:Fort Activity Scaling Text Block Object Reference)
    AnimateScroll=true
    ScrollDestination=BottomOrRight
    Padding=0
   )
   > n1 = Branch(Condition=Get IsAcceptButtonEnabled())
   > n2 = Branch(Condition=IsContentGated(self=GetSelectedGameActivity()))
   n3 = Get SqueegeeSlot_LockInfo()
   > Set LastFocusedWidget(LastFocusedWidget=@n3)
   > Return Node!(execute, ReturnValue:Widget Object Reference=@n3)
   n4 = Select(Option 0=Get Button_Favorite(), Option 1=Get Button_Accept(), Index=IsVisible(self=Get WidgetSwitcher_Buttons()))
   Set LastFocusedWidget(execute=@n2.else, LastFocusedWidget=@n4)
   > Return Node(ReturnValue=@n4)
   Return Node(execute=@n1.else, ReturnValue=Get Button_Favorite())
  ]}
 NavigationFromCommunityRow {nodes [
   n1 = NavigationFromCommunityRow!() -> (then, Navigation:EUINavigation Enum) {is=title K2Node_FunctionEntry}
   > n2 = Switch on EUINavigation(Selection=@n1)
   n3 = NavigateToDescription(execute=@n2.Down) {is=title K2Node_CallFunction}
   > Return Node(ReturnValue=@n3)
   n4 = NavigateToTopRow(execute=@n2.Up) {is=title K2Node_CallFunction}
   > Return Node(ReturnValue=@n4)
  ]}
 NavigationFromSubmodeButton {nodes [
   n1 = NavigationFromSubmodeButton!() -> (then, Navigation:EUINavigation Enum) {is=title K2Node_FunctionEntry}
   > n2 = Switch on EUINavigation(Selection=@n1)
   n3 = Set LastFocusedWidget(execute=@n2.Down, LastFocusedWidget=Get Button_Accept())
   > Return Node(ReturnValue=@n3)
  ]}
 RefreshContentWarnings {nodes [
   RefreshContentWarnings() {is=title K2Node_FunctionEntry}
   > Reset(self=Get WarningEntryBox() -> (WarningEntryBox:Dynamic Entry Box Object Reference), bDeleteWidgets=false)
   n1 = GetContentWarningStrings!(self:Fort Activity Details Display Object Reference=Get ActivityDetailsDisplay()) -> (ReturnValue:Array of Strings)
   > Branch(Condition=Is Array Non Empty!(InputArray:Array of Strings=@n1) -> (Non-Empty:bool=~) {is=title K2Node_MacroInstance})
   > n2 = For Each Loop(Array:Array of Strings=@n1) -> (Array Element:String)
   > Branch(Condition=NOT Boolean(A=IsEmpty(InString=@n2)))
   n3 = Find(TargetMap:Map of Strings to Texts=Get Content Warning Map() -> (Content Warning Map:Map of Strings to Texts), Key=@n2) -> (Value:Text)
   > Branch(Condition=@n3.ReturnValue)
   > n4 = Create Entry of Class(
    self=Get WarningEntryBox()
    EntryClass=&/ActivityBrowserStatusWarning.ActivityBrowserStatusWarning_C
   ) -> (ReturnValue:Activity Browser Status Warning Object Reference)
   > Play Start Animation!(execute, self:Activity Browser Status Warning Object Reference=@n4, In Warning Text:Text=@n3) -> (then)
  ]}
 RefreshDisabledState {nodes [
   RefreshDisabledState() {is=title K2Node_FunctionEntry}
   > n1 = Set Local_InvalidActivityReason(Local_InvalidActivityReason:EFortInvalidActivityReason Enum=GetInvalidActivityReason() -> (ReturnValue=None)) -> (Output_Get:EFortInvalidActivityReason Enum=None)
   n2 = IsPreviewTile!(self:Fort Activity Details Display Object Reference=Get ActivityDetailsDisplay()) -> (ReturnValue:bool)
   > n3 = Branch(Condition=AND Boolean!(
     A=Get ShowAcceptButton() -> (ShowAcceptButton:bool)
     B="Not Equal (Enum)"(A:EFortInvalidActivityReason Enum=@n1, B:EFortInvalidActivityReason Enum=InvalidData)
     C:bool=NOT Boolean(A=@n2)
    ) -> (ReturnValue))
   n4 = GetSelectedGameActivity()
   n5 = Get Local_InvalidActivityReason() -> (Local_InvalidActivityReason:EFortInvalidActivityReason Enum=None)
   > UpdateVisibility!(
    execute
    self
    Target:Widget Object Reference=Get WidgetSwitcher_Buttons()
    ShowWidget:bool=OR Boolean(A="Equal (Enum)"(A:EFortInvalidActivityReason Enum=@n5, B:EFortInvalidActivityReason Enum=None), B=IsContentGated(self=@n4))
   ) -> (then)
   n6 = GetBestLockedReason!(self:Fort Game Activity Object Reference=GetSelectedGameActivity(), Context:Player Controller Object Reference=GetOwningPlayer()) -> (ReturnValue:Per Player Activity Locked Reason Structure)
   n7 = AND Boolean!(
    A="Not Equal (Enum)"(A=Get Local_InvalidActivityReason() -> (Local_InvalidActivityReason=None), B=None)
    B="Equal (Enum)"(
     A:EFortActivityIsLockedReason Enum=Break Per Player Activity Locked Reason!(PerPlayerActivityLockedReason:"Per Player Activity Locked Reason Structure (by ref)"=@n6) -> (LockedFor:Fort Player Controller Object Reference, Reason:EFortActivityIsLockedReason Enum) {is=title K2Node_BreakStruct}.Reason
     B:EFortActivityIsLockedReason Enum=None
    )
   ) -> (ReturnValue)
   > UpdateButtonVisibility!(
    execute
    self
    Target:Widget Object Reference=Get DisabledWarningHB() -> (DisabledWarningHB:Horizontal Box Object Reference)
    ShowWidget:bool=@n7
   ) -> (then)
   > Branch(Condition=@n7)
   > "SetText (Text)"(
    self=Get DisabledWarningMessage() -> (DisabledWarningMessage:Common Text Object Reference)
    InText=Select!(
     None:Text
     PartyTooBig:Text="LOCTABLE(\"/Game/UI/DiscoveryShared/StringTable_ActivityBrowser\", \"Warning_PartyTooLarge_Long\")"
     NotPartyLeader:Text="LOCTABLE(\"&StringTable_ActivityBrowser\", \"Warning_NotPartyLeader\")"
     MatchmakingAlready:Text="LOCTABLE(\"&StringTable_ActivityBrowser\", \"Warning_CurrentlyMatchmaking\")"
     PlaylistDisabled:Text="LOCTABLE(\"&StringTable_ActivityBrowser\", \"Warning_ModeDisabled\")"
     PlayerIsLfg:Text="LOCTABLE(\"&StringTable_ActivityBrowser\", \"Warning_PlayerIsLfg\")"
     InvalidData:Text="LOCTABLE(\"&StringTable_ActivityBrowser\", \"Warning_InvalidMode\")"
     ExpiredTournamentLockCache:Text
     Index:EFortInvalidActivityReason Enum=Get Local_InvalidActivityReason() -> (Local_InvalidActivityReason=None)
    ) -> (ReturnValue:Text)
   )
   > PlayAnimation(InAnimation=Get DisabledIconPulse() -> (DisabledIconPulse:Widget Animation Object Reference), NumLoopsToPlay=3)
   SetVisibility(execute=@n3.else, self=[Get DisabledWarningHB(), Get WidgetSwitcher_Buttons()], InVisibility=Collapsed)
  ]}
 RemoveIndentation {nodes [
   n1 = RemoveIndentation!() -> (then, InputText:Text) {is=title K2Node_FunctionEntry}
   n2 = "To String (Text)"(InText=@n1)
   n3 = Split(SourceString=@n2, InStr="\r\n", SearchDir=FromStart)
   n4 = Append!(A=@n3, B="\r\n", C:String=Trim!(SourceString:String=@n3.RightS) -> (ReturnValue:String)) -> (ReturnValue)
   n5 = "integer > integer"(A=FindSubstring(SearchIn=@n2, Substring="\r\n"))
   > Return Node!(execute, OutputText:Text="To Text (String)"(InString=SelectString(A=@n4, B=@n2, bPickA=@n5)))
  ]}
 Set Gamepad Action Bar Style {nodes [
   Set Gamepad Action Bar Style() {is=title K2Node_FunctionEntry}
   n1 = Get CommonInputSubsystem()
   > Branch(Condition="Equal (Enum)"(A:ECommonInputType Enum=GetCurrentInputType(self=@n1), B:ECommonInputType Enum=Gamepad))
   > SetPadding!(
    execute
    self:Vertical Box Slot Object Reference=SlotAsVerticalBoxSlot(Widget=Get Button_CreatorPage())
    InPadding_Left:float=0
    InPadding_Top:float=0
    InPadding_Right:float=0
    InPadding_Bottom:float=0
   ) -> (then)
  ]}
 SetAcceptButtonState {nodes [
   n1 = SetAcceptButtonState!() -> (then, IsEnable:bool=~) {is=title K2Node_FunctionEntry}
   n2 = GetSelectedGameActivity()
   > Is Valid(InputObject=@n2)
   n3 = IsContentGated(self=@n2)
   > SetActiveWidgetIndex(self=Get WidgetSwitcher_Buttons(), Index=SelectInt(A=SelectInt(A=1, B=0, bPickA=@n3), B=2, bPickA=@n1))
  ]}
 SetButtonsInteractable {nodes [
   n1 = SetButtonsInteractable!() -> (then, IsInteractionEnabled:bool=~) {is=title K2Node_FunctionEntry}
   > n2 = For Each Loop(Array:Array of Common Button Base Object References=Get Buttons()) -> (Array Element:Common Button Base Object Reference)
   > SetIsInteractionEnabled!(execute, self:Common Button Base Object Reference=@n2, bInIsInteractionEnabled:bool=@n1) -> (then)
  ]}
 SetCCUCountValueText {nodes [
   n1 = SetCCUCountValueText!() -> (then, CCUCount:int) {is=title K2Node_FunctionEntry}
   n2 = Get WBP_CCUCount() -> (WBP_CCUCount:WBP Island Label And Details Object Reference)
   > SetVisibility(
    self=@n2
    InVisibility=Select!(Option 0:ESlateVisibility Enum=HitTestInvisible, Option 1:ESlateVisibility Enum=Collapsed, Index:bool="integer < integer"(A=@n1)) -> (ReturnValue:ESlateVisibility Enum)
   )
   > SetValueText!(
    execute
    self:WBP Island Label And Details Object Reference=@n2
    NewParam:Text=Select(
     Option 0:Text="To Text (Integer)"(Value=@n1)
     Option 1:Text=GetShorthandQuantity!(Quantity:int=@n1, MinimumFractionalDigits:int=0, MaximumFractionalDigits:int=1, bShouldRoundDown:bool=false) -> (ReturnValue:Text)
     Index="integer > integer"(A=@n1, B=999)
    ) -> (ReturnValue:Text)
   ) -> (then)
  ]}
 SetCultureLayout {nodes [
   SetCultureLayout() {is=title K2Node_FunctionEntry}
   n1 = "Equal Exactly (String)"(A=GetCurrentCulture!() -> (ReturnValue:String), B=ar)
   > Set ContentAnchor!(
    execute
    ContentAnchor:Vector 2D Structure=Select(Option 0:Vector 2D Structure=(X=0,Y=1), Option 1:Vector 2D Structure=(X=1,Y=1), Index=@n1) -> (ReturnValue:Vector 2D Structure)
    self:Common Visual Attachment Object Reference=Get FortVisualAttachment__TopContent() -> (FortVisualAttachment__TopContent:Common Visual Attachment Object Reference)
   ) -> (then, Output_Get:Vector 2D Structure)
   > Set ContentAnchor(
    ContentAnchor=Select(Option 0=(X=1,Y=1), Option 1=(X=0,Y=1), Index=@n1)
    self=Get FortVisualAttachment__RatingCard() -> (FortVisualAttachment__RatingCard:Common Visual Attachment Object Reference)
   )
   > Set HorizontalAlignment!(
    execute
    HorizontalAlignment:EHorizontalAlignment Enum=Select(Option 0:EHorizontalAlignment Enum=HAlign_Right, Option 1:EHorizontalAlignment Enum=HAlign_Left, Index=@n1) -> (ReturnValue:EHorizontalAlignment Enum)
    self:Overlay Slot Object Reference=SlotAsOverlaySlot(Widget=Get SafeZone_Prompts() -> (SafeZone_Prompts:Safe Zone Object Reference))
   ) -> (then, Output_Get:EHorizontalAlignment Enum=HAlign_Fill)
  ]}
 SetInGamePurchasesVisibility {nodes [
   n1 = SetInGamePurchasesVisibility!() -> (then, Has In-Game Purchases:bool=~) {is=title K2Node_FunctionEntry}
   n2 = Select(Option 0:ESlateVisibility Enum=Collapsed, Option 1:ESlateVisibility Enum=HitTestInvisible, Index=@n1) -> (ReturnValue:ESlateVisibility Enum)
   > SetVisibility(self=Get Description_InGamePurchase_Divider() -> (Description_InGamePurchase_Divider:Image Object Reference), InVisibility=@n2)
   > SetVisibility(self=Get InGamePurchaseContainer() -> (InGamePurchaseContainer:Horizontal Box Object Reference), InVisibility=@n2)
  ]}
 SetMobileStyle {nodes [
   SetMobileStyle() {is=title K2Node_FunctionEntry}
   > Branch(Condition=IsMobileGame())
   n1 = Get Button_ContextMenu()
   > Set Min Width!(execute, Min Width:int=100, self:WBP UIKit Context Menu Anchor Button Base Object Reference=@n1) -> (then, Output_Get:int=0)
   > Set Min Height!(execute, Min Height:int=100, self:WBP UIKit Context Menu Anchor Button Base Object Reference=@n1) -> (then, Output_Get:int=0)
   > SetPadding(
    self:Overlay Slot Object Reference=SlotAsOverlaySlot(Widget=Get SafeZone_Prompts())
    InPadding_Left=0
    InPadding_Top=0
    InPadding_Right=48
    InPadding_Bottom=24
   )
   n2 = SlotAsGridSlot(Widget=Get SZ_Content_DynamicPadding())
   n3 = Get!(self:Grid Slot Object Reference=@n2) -> (Padding_Left:float=~, Padding_Top:float=~, Padding_Right:float=~, Padding_Bottom:float=~)
   > SetPadding(
    self:Grid Slot Object Reference=@n2
    InPadding_Left=48
    InPadding_Top=@n3.Padding_Top
    InPadding_Right=48
    InPadding_Bottom=@n3.Padding_Bottom
   )
   > SetSize!(execute, self:Spacer Object Reference=Get Spacer_70() -> (Spacer_70:Spacer Object Reference), InSize:Vector 2D Structure=(X=10,Y=1)) -> (then)
  ]}
 SetMobileStyleSqueegeeSlotCard {nodes [
   SetMobileStyleSqueegeeSlotCard() {is=title K2Node_FunctionEntry}
   n1 = GetViewportSize!() -> (ReturnValue_X:double, ReturnValue_Y:double)
   > Branch(Condition="float >= float"(A=float / float(A=@n1, B=@n1.ReturnValue_Y), B=Get SqueegeeMobileScreenRatioLimit() -> (SqueegeeMobileScreenRatioLimit:double)))
   n2 = IsMobileGame()
   n3 = "Not Equal (Enum)"(
    A:ESlateVisibility Enum=GetVisibility!(self:Widget Object Reference=Get WidgetSwitcher_Buttons()) -> (ReturnValue:ESlateVisibility Enum=Visible)
    B:ESlateVisibility Enum=Collapsed
   )
   > Set RenderTransform!(
    execute
    RenderTransform:Widget Transform Structure=Select(
     Option 0:Widget Transform Structure=Get SqueegeeDefaultTransform() -> (SqueegeeDefaultTransform:Widget Transform Structure)
     Option 1:Widget Transform Structure=Get SqueegeeMobileScreenRatioLimitTransform() -> (SqueegeeMobileScreenRatioLimitTransform:Widget Transform Structure)
     Index=AND Boolean(A=@n3, B=@n2)
    ) -> (ReturnValue:Widget Transform Structure)
    self:Widget Object Reference=Get SqueegeeSlot_Card()
   ) -> (then, Output_Get:Widget Transform Structure)
  ]}
 SetNavigationRules {nodes [
   SetNavigationRules() {is=title K2Node_FunctionEntry}
   > SetNavigationRuleCustom(self=Get Button_ShowSubModeList(), Direction=Down, InCustomDelegate=Create Event())
   n1 = Create Event()
   n2 = Get Button_Accept()
   > SetNavigationRuleCustom(self=@n2, Direction=Up, InCustomDelegate=@n1)
   > SetNavigationRuleCustom(self=@n2, Direction=Down, InCustomDelegate=@n1)
   > SetNavigationRuleCustom(self=@n2, Direction=Left, InCustomDelegate=@n1)
   > SetNavigationRuleCustom(self=@n2, Direction=Right, InCustomDelegate=@n1)
   n3 = Create Event()
   n4 = Get SqueegeeSlot_LockInfo()
   > SetNavigationRuleCustom(self=@n4, Direction=Down, InCustomDelegate=@n3)
   > SetNavigationRuleCustom(self=@n4, Direction=Left, InCustomDelegate=@n3)
   > SetNavigationRuleCustom(self=@n4, Direction=Right, InCustomDelegate=@n3)
   n5 = Create Event()
   n6 = Get Button_TimeLocked()
   > SetNavigationRuleCustom(self=@n6, Direction=Up, InCustomDelegate=@n5)
   > SetNavigationRuleCustom(self=@n6, Direction=Down, InCustomDelegate=@n5)
   > SetNavigationRuleCustom(self=@n6, Direction=Left, InCustomDelegate=@n5)
   > SetNavigationRuleCustom(self=@n6, Direction=Right, InCustomDelegate=@n5)
   n7 = Create Event()
   n8 = Get Button_Favorite()
   > SetNavigationRuleCustom(self=@n8, Direction=Up, InCustomDelegate=@n7)
   > SetNavigationRuleCustom(self=@n8, Direction=Down, InCustomDelegate=@n7)
   > SetNavigationRuleCustom(self=@n8, Direction=Left, InCustomDelegate=@n7)
   > SetNavigationRuleCustom(self=@n8, Direction=Right, InCustomDelegate=@n7)
   n9 = Create Event()
   n10 = Get Button Context Menu() -> (Button Context Menu:WBP UIKit Context Menu Anchor Button Object Reference)
   > SetNavigationRuleCustom(self=@n10, Direction=Up, InCustomDelegate=@n9)
   > SetNavigationRuleCustom(self=@n10, Direction=Down, InCustomDelegate=@n9)
   > SetNavigationRuleCustom(self=@n10, Direction=Left, InCustomDelegate=@n9)
   > SetNavigationRuleCustom(self=@n10, Direction=Right, InCustomDelegate=@n9)
   n11 = Create Event()
   n12 = Get SqueegeeSlot_Card()
   > SetNavigationRuleCustom(self=@n12, Direction=Up, InCustomDelegate=@n11)
   > SetNavigationRuleCustom(self=@n12, Direction=Down, InCustomDelegate=@n11)
   > SetNavigationRuleCustom(self=@n12, Direction=Left, InCustomDelegate=@n11)
   > SetNavigationRuleCustom(self=@n12, Direction=Right, InCustomDelegate=@n11)
   n13 = Create Event()
   n14 = Get Button_Subscribe()
   > SetNavigationRuleCustom(self=@n14, Direction=Up, InCustomDelegate=@n13)
   > SetNavigationRuleCustom(self=@n14, Direction=Down, InCustomDelegate=@n13)
   > SetNavigationRuleCustom(self=@n14, Direction=Left, InCustomDelegate=@n13)
   > SetNavigationRuleCustom(self=@n14, Direction=Right, InCustomDelegate=@n13)
  ]}
 SetSubscribeButtonIcon {nodes [
   n1 = SetSubscribeButtonIcon!() -> (then, IsSubscribed:bool=~) {is=title K2Node_FunctionEntry}
   > SetMaterial!(
    execute
    self:WBP UIKit Block Image Object Reference=Get Block_Icon!(self:WBP UIKit Button CTA Base Object Reference=Get Button_Subscribe()) -> (Block_Icon:WBP UIKit Block Image Object Reference)
    InMaterial:"Material Instance Object Reference (by ref)"=SelectMaterialInstance!(
     A:Material Instance Object Reference=/FortniteCoreUI/UI/DiscoveryShared/GameDetails/MaterialInstances/MI_UI_ActivityDetails_NotificationOn
     B:Material Instance Object Reference=/FortniteCoreUI/UI/DiscoveryShared/GameDetails/MaterialInstances/MI_UI_ActivityDetails_NotificationOff
     bSelectA:bool=@n1
    ) -> (ReturnValue:Material Instance Object Reference)
   ) -> (then)
  ]}
 SetupContextMenu {nodes [
   SetupContextMenu() {is=title K2Node_FunctionEntry}
   > Branch(Condition=Get bEnableContextMenuInDiscover!(self:Fort Runtime Options Object Reference=GetRuntimeOptions()) -> (bEnableContextMenuInDiscover:bool))
   > n1 = Construct Object from Class(Class=/Script/FortniteUI.UIKitContextMenuActionContext) -> (ReturnValue:UIKit Context Menu Action Context Object Reference)
   > Set ContextMenuActionContext(ContextMenuActionContext:UIKit Context Menu Action Context Object Reference=@n1) -> (Output_Get:UIKit Context Menu Action Context Object Reference)
   > n2 = Sequence!(execute) -> (then_0, then_1)
   > AddInteraction(
    self=Get ContextMenuActionContext()
    InActionInfo_DisplayName="LOCTEXT(\"View Creator\")"
    InActionInfo_SortPriority=1000
    InActionInfo_bIsAvailable=True
    InActionInfo_bIsEnabled=True
    InActionInfo_bIsSelected=False
    InActionInfo_bAddDivider=False
    InActionInfo_bPositionDividerAfter=False
    InActionInfo_bCloseMenuOnSelect=False
    InActionInfo_InteractionNameID=ViewCreator
    InActionInfo_Interaction=(DataTable=None,RowName="")
    InActionInfo_PressedDelegate="(null).None"
    InActionInfo_ReleasedDelegate="(null).None"
    InInteractionDelegate=Create Event()
   )
   SetContext(execute=@n2.then_1, self=Get Button_ContextMenu(), InteractionContext_Data=Get ContextMenuActionContext())
   > Update Context Menu Visibility()
  ]}
 SetupTimerViewModel {nodes [
   SetupTimerViewModel() {is=title K2Node_FunctionEntry}
   n1 = Get FortCountdownTimerVM() -> (FortCountdownTimerVM:Countdown Timer View Model Object Reference)
   > SetupViewModel!(
    execute
    self:Countdown Timer View Model Object Reference=@n1
    InFutureTime:"Date Time Structure (by ref)"=GetNextStartTime!(self:Fort Discover Link Entry Tile Item VM Object Reference=Get FortIslandDetailsVM()) -> (ReturnValue:Date Time Structure)
    InLiveText:"Text (by ref)"=Format Text!(Format="LOCTEXT(\"Live\")") -> (Result)
   ) -> (then)
   > SetTimerUnitDisplayType!(execute, self:Countdown Timer View Model Object Reference=@n1, InTimerUnitDisplayType:ETimespanUnitDisplayFormat Enum=Abbreviated) -> (then)
   > SetApproximateLongTimers!(execute, self:Countdown Timer View Model Object Reference=@n1, bInApproximateLongTimers:bool=false) -> (then)
   > StartTimer(self:Countdown Timer View Model Object Reference=@n1)
   > Return Node!(
    execute
    IsCountdown:bool="Equal (Enum)"(
     A:ECountdownTimerState Enum=Get TimerState!(self:Countdown Timer View Model Object Reference=@n1) -> (TimerState:ECountdownTimerState Enum=None)
     B:ECountdownTimerState Enum=Countdown
    )
   )
  ]}
 ShouldOfferSubModeSelection {nodes [
   ShouldOfferSubModeSelection() {is=title K2Node_FunctionEntry}
   n1 = GetSelectedGameActivity()
   n2 = Contains Item(
    TargetArray:Array of Strings=Get LinkCodesWithoutSubModeSelection() -> (LinkCodesWithoutSubModeSelection:Array of Strings)
    ItemToFind:"String (by ref)"=Break Online Link Id!(OnlineLinkId:"Online Link Id Structure (by ref)"=Get LinkId!(self:Fort Game Activity Object Reference=@n1) -> (LinkId:Online Link Id Structure)) -> (Mnemonic:String, Version:int) {is=title K2Node_BreakStruct}
   )
   > Return Node!(
    execute
    bShouldOfferSubModeSelection:bool=AND Boolean(A=CanShowModeSetButton!(self:Fort Activity Details Modal Object Reference) -> (ReturnValue:bool), B=NOT Boolean(A=@n2))
   )
  ]}
 ShouldShowLinkCategoryIconsInTags {nodes [
   ShouldShowLinkCategoryIconsInTags() {is=title K2Node_FunctionEntry}
   n1 = GetSelectedGameActivity()
   n2 = IsEpicCreated!(self:Fort Game Activity Object Reference=@n1) -> (ReturnValue:bool)
   n3 = GetLinkCategory(self=@n1) -> (ReturnValue=None)
   > Return Node!(
    execute
    ReturnValue:bool=AND Boolean!(
     A=ShouldShowLinkCategoryIcon!() -> (ReturnValue:bool)
     B=NOT Boolean(A=@n2)
     C:bool="Not Equal (Enum)"(A:EFortActivityLinkCategory Enum=@n3, B:EFortActivityLinkCategory Enum=None)
    ) -> (ReturnValue)
   )
  ]}
 Update Accept Text {nodes [
   n1 = Update Accept Text!() -> (then, ButtonWidget:Object Reference, IsCategory:bool=~, IsDownloadRequired:bool=~) {is=title K2Node_FunctionEntry}
   > n2 = Cast To FortCTAButton!(execute, Object=@n1) -> (then, CastFailed, AsFort CTAButton:Fort CTAButton Object Reference) {is=title K2Node_DynamicCast}
   > SetText(
    self=@n2
    ButtonText=Select(
     Option 0:Text="LOCTABLE(\"&StringTable_ActivityBrowser\", \"Activity_Select_Button\")"
     Option 1:Text="LOCTABLE(\"&StringTable_ActivityBrowser\", \"Category_Browse_Button\")"
     Index=@n1.IsCategory
    ) -> (ReturnValue:Text)
   )
   > SetSecondaryText(
    self=@n2
    ButtonSecondaryText=Select(Option 1="LOCTABLE(\"&StringTable_ActivityBrowser\", \"Warning_DownloadRequired\")", Index=@n1.IsDownloadRequired)
   )
  ]}
 Update Context Menu Visibility {nodes [
   Update Context Menu Visibility() {is=title K2Node_FunctionEntry}
   n1 = HasInteractionItems!(self:UIKit Context Menu Action Context Object Reference=Get ContextMenuActionContext()) -> (ReturnValue:bool)
   n2 = NOT Boolean(A=IsPreviewTile(self=Get ActivityDetailsDisplay()))
   > SetVisibility(
    self=Get Button_ContextMenu()
    InVisibility=Select(Option 0:ESlateVisibility Enum=Collapsed, Option 1:ESlateVisibility Enum=Visible, Index=AND Boolean!(A=@n1, B=@n2) -> (ReturnValue)) -> (ReturnValue:ESlateVisibility Enum)
   )
  ]}
 UpdateButtonVisibility {nodes [
   n1 = UpdateButtonVisibility!() -> (then, Target:Widget Object Reference, ShowWidget:bool=~) {is=title K2Node_FunctionEntry}
   > Is Valid(InputObject=@n1)
   > SetVisibility(self=@n1, InVisibility=Select(Option 0=Collapsed, Option 1=Visible, Index=@n1.ShowWidget))
  ]}
 UpdateDetailVisibility {nodes [
   UpdateDetailVisibility!() -> (then, IsFavoritable:bool=~) {is=title K2Node_FunctionEntry}
   > Set IsDownloadRequired(IsDownloadRequired:bool=IsDownloadRequired!(self:Fort Activity Details Display Object Reference=Get ActivityDetailsDisplay()) -> (ReturnValue:bool)) -> (Output_Get:bool)
   > n1 = Set IsCategory(IsCategory:bool=IsDisplayingCategory!(self:Fort Activity Details Display Object Reference=Get ActivityDetailsDisplay()) -> (ReturnValue:bool)) -> (Output_Get:bool)
   > UpdateVisibility(
    Target=Get Text_ActivityOrigin() -> (Text_ActivityOrigin:Common Rich Text Block Object Reference)
    ShowWidget=AND Boolean(A=Get ShowCreator() -> (ShowCreator:bool), B=NOT Boolean(A=@n1))
   )
   n2 = Get IsCategory() -> (IsCategory:bool)
   > UpdateVisibility(
    Target=Get EntryBox_ActivityTags!() -> (EntryBox_ActivityTags)
    ShowWidget=AND Boolean(A=Get ShowTags() -> (ShowTags:bool), B=NOT Boolean(A=@n2))
   )
   > UpdateVisibility(Target=Get WarningEntryBox(), ShowWidget=Get ShowContentWarning() -> (ShowContentWarning:bool))
   > UpdateVisibility(Target=Get WidgetSwitcher_Buttons(), ShowWidget=Get ShowAcceptButton())
   > SetActiveWidgetIndex(self=Get WidgetSwitcher_Buttons(), Index=0)
   > Update Accept Text!(
    execute
    self
    ButtonWidget:Object Reference=Get Button_Accept()
    IsCategory:bool=Get IsCategory()
    IsDownloadRequired:bool=Get IsDownloadRequired() -> (IsDownloadRequired:bool)
   ) -> (then) {is=title K2Node_CallFunction}
   n3 = Get IsCategory()
   > UpdateButtonVisibility(Target=Get DisabledWarningHB(), ShowWidget=AND Boolean(A=Get ShowAcceptButton(), B=NOT Boolean(A=@n3))) {is=title K2Node_CallFunction}
   n4 = IsMobileGame()
   n5 = Get IsCategory()
   n6 = IsPreviewTile(self=Get ActivityDetailsDisplay())
   > UpdateButtonVisibility(
    Target=Get Button_ContextMenu()
    ShowWidget=AND Boolean!(A=Get ShowDetailsButton() -> (ShowDetailsButton:bool), B=NOT Boolean(A=@n4), C:bool=NOT Boolean(A=@n5), D:bool=NOT Boolean(A=@n6)) -> (ReturnValue)
   )
   n7 = Get ActivityDetailsDisplay()
   > UpdateButtonVisibility(Target=Get Button_Favorite(), ShowWidget=NOT Boolean(A=IsPreviewTile(self=@n7)))
   > UpdateButtonVisibility(Target=Get Button_Subscribe(), ShowWidget=GetAreIslandNotificationsEnabled(self=GetRuntimeOptions()))
  ]}
 UpdateDetailsTexts {nodes [
   UpdateDetailsTexts() {is=title K2Node_FunctionEntry}
   n1 = Get ActivityDetailsDisplay()
   > Set Text(
    Text=Get Text(self=Get Text_ActivityName!(self:Fort Activity Details Display Object Reference=@n1) -> (Text_ActivityName:Common Text Object Reference))
    self=Get Text_ActivityName!() -> (Text_ActivityName:Fort Activity Scaling Text Block Object Reference)
   )
   > n2 = RemoveIndentation!(
    execute
    self
    InputText:Text=Get Text(self=Get Text_ActivityDescription!(self:Fort Activity Details Display Object Reference=@n1) -> (Text_ActivityDescription))
   ) -> (then, OutputText:Text) {is=title K2Node_CallFunction}
   > Set Text(Text=@n2, self=Get Text_ActivityDescription!() -> (Text_ActivityDescription))
   > Set Text(
    Text=Get Text(self:Rich Text Block Object Reference=Get Text_ActivityOrigin!(self:Fort Activity Details Display Object Reference=@n1) -> (Text_ActivityOrigin))
    self:Rich Text Block Object Reference=Get Text_ActivityOrigin!() -> (Text_ActivityOrigin)
   )
  ]}
 UpdateFavoriteStatus {nodes [
   n1 = UpdateFavoriteStatus!() -> (then, IsFavorite:bool=~, PlayAnimation:bool=~) {is=title K2Node_FunctionEntry}
   > Set IsFavorited(IsFavorited:bool=@n1) -> (Output_Get:bool)
   > n2 = Cast To WBP_GameDetails_Toggle!(execute, Object=Get Button_Favorite()) -> (then, CastFailed, AsWBP Game Details Toggle:WBP Game Details Toggle Object Reference) {is=title K2Node_DynamicCast}
   > Set Toggle!(execute, self:WBP Game Details Toggle Object Reference=@n2, IsInit:bool=false, IsToggled:bool=Get IsFavorite() -> (IsFavorite:bool)) -> (then) {is=title K2Node_CallFunction}
  ]}
 UpdateImage {nodes [
   UpdateImage!() -> (then, bInIsLoading:bool=~, InTexture:Texture Object Reference) {is=title K2Node_FunctionEntry}
   > Set bWasPreviouslyLoading(bWasPreviouslyLoading:bool=Get bIsLoading() -> (bIsLoading:bool)) -> (Output_Get:bool)
   > Set bIsLoading(bIsLoading=Get bInIsLoading() -> (bInIsLoading:bool))
   > n1 = GetDynamicMaterial(self=Get Thumbnail())
   > n2 = Set DynamicMaterial(DynamicMaterial:Material Instance Dynamic Object Reference=@n1) -> (Output_Get:Material Instance Dynamic Object Reference)
   n3 = Get InTexture() -> (InTexture:Texture Object Reference)
   > SetTextureParameterValue(
    self=@n2
    ParameterName=Get ThumbnailParam() -> (ThumbnailParam=None)
    Value=Select(
     Option 0:Texture Object Reference=Get DefaultImage() -> (DefaultImage:Texture Object Reference)
     Option 1:Texture Object Reference=@n3
     Index=IsValid(Object=@n3)
    ) -> (ReturnValue:Texture Object Reference)
   )
   > SetScalarParameterValue(
    self=Get DynamicMaterial() -> (DynamicMaterial:Material Instance Dynamic Object Reference)
    ParameterName=Get ThumbnailOpacity() -> (ThumbnailOpacity:Name=None)
    Value=0
   )
   > n4 = Branch(Condition=Get bIsLoading())
   > SetVisibility(self=Get WBP_UIKit_Throbber() -> (WBP_UIKit_Throbber:WBP UIKit Throbber Object Reference), InVisibility=SelfHitTestInvisible)
   > SetScalarParameterValue(self=Get DynamicMaterial(), ParameterName=Get ThumbnailOpacity() -> (ThumbnailOpacity=None), Value=1)
   n5 = Branch(execute=@n4.else, Condition=Get bWasPreviouslyLoading() -> (bWasPreviouslyLoading:bool))
   > PlayAnimationReverse(InAnimation=Get HideSpinner() -> (HideSpinner:Widget Animation Object Reference))
   > Is Valid(InputObject=Get InTexture())
   > PlayAnimationForward(InAnimation=Get ShowKeyArt())
   Set Timer by Event(execute=@n5.else, Delegate=Create Event(), Time=0.2, bLooping=false)
  ]}
 UpdateLinkCode {nodes [
   n1 = UpdateLinkCode!() -> (then, LinkCode:Text, IsCreativeActivity:bool=~) {is=title K2Node_FunctionEntry}
   n2 = NOT Boolean(A=TextIsEmpty(InText=Get LinkCode() -> (LinkCode:Text)))
   n3 = NOT Boolean(A=IsPreviewTile(self=Get ActivityDetailsDisplay()))
   n4 = IsSubModeSelectVisualsEnabled!() -> (ReturnValue:bool)
   n5 = Get ShowLinkCode() -> (ShowLinkCode:bool)
   > SetVisibility(
    self=Get WBP_IslandCode()
    InVisibility=Select(
     Option 0:ESlateVisibility Enum=Collapsed
     Option 1:ESlateVisibility Enum=SelfHitTestInvisible
     Index=AND Boolean!(A=@n1.IsCreativeActivity, B=@n4, C=@n5, D=@n2, E:bool=@n3) -> (ReturnValue)
    ) -> (ReturnValue:ESlateVisibility Enum)
   )
   > SetValueText(self=Get WBP_IslandCode(), NewParam=Get LinkCode())
  ]}
 UpdateSubModeVisibility {nodes [
   n1 = UpdateSubModeVisibility!() -> (then, IsShowingSubModeList:bool=~) {is=title K2Node_FunctionEntry}
   > n2 = Set IsShowingSubModeList(IsShowingSubModeList=@n1)
   > n3 = Branch(Condition=@n2)
   n4 = Get GridPanel_DynamicDetails()
   > SetVisibility(self=@n4, InVisibility=Collapsed)
   n5 = Get NamedSlot_DetailsOverride() -> (NamedSlot_DetailsOverride:Named Slot Object Reference)
   > SetVisibility(self=@n5, InVisibility=SelfHitTestInvisible)
   SetVisibility(execute=@n3.else, self=@n4, InVisibility=SelfHitTestInvisible)
   > SetVisibility(self=@n5, InVisibility=Collapsed)
   > n6 = Cast To IslandPrivacyButton(Object=Get Button_IslandFill() -> (Button_IslandFill:Common Button Base Object Reference))
   > n7 = Cast To IslandPrivacyButton(Object=Get Button_IslandPrivacy() -> (Button_IslandPrivacy:Common Button Base Object Reference))
   > UpdatePrivacyAndFill!(
    execute
    self:Fort Activity Details Modal Object Reference
    bIsPrivate:bool=IsToggleOn!(self:Island Privacy Button Object Reference=@n7) -> (bIsToggleOn:bool)
    bIsFill:bool=IsToggleOn(self=@n6)
   ) -> (then)
  ]}
 UpdateVisibility {nodes [
   n1 = UpdateVisibility!() -> (then, Target:Widget Object Reference, ShowWidget:bool=~) {is=title K2Node_FunctionEntry}
   > Is Valid(InputObject=@n1)
   > SetVisibility(self=@n1, InVisibility=Select(Option 0=Collapsed, Option 1=SelfHitTestInvisible, Index=@n1.ShowWidget))
  ]}
 UpdateVisualForArabic {nodes [
   UpdateVisualForArabic() {is=title K2Node_FunctionEntry}
   n1 = "Equal Exactly (String)"(A=GetCurrentCulture(), B=ar)
   > Set HorizontalAlignment(
    HorizontalAlignment=Select(Option 0:EHorizontalAlignment Enum=HAlign_Left, Option 1:EHorizontalAlignment Enum=HAlign_Right, Index=@n1) -> (ReturnValue:EHorizontalAlignment Enum)
    self=SlotAsOverlaySlot(Widget=Get FortVisualAttachment__TopContent())
   ) -> (Output_Get=HAlign_Fill)
   > Set ContentAnchor(
    ContentAnchor=Select(Option 0:Vector 2D Structure=(X=0,Y=1), Option 1:Vector 2D Structure=(X=1,Y=1), Index=@n1) -> (ReturnValue:Vector 2D Structure)
    self=Get FortVisualAttachment__TopContent()
   )
  ]}
 ViewAttributions {nodes [
   ViewAttributions!() -> (then, NameID:Name) {is=title K2Node_FunctionEntry}
   > HandleAttributionsButtonClicked(self:Fort Activity Details Modal Object Reference)
   > Close(self:Menu Anchor Object Reference=Get Anchor!(self:UIKit Context Menu Anchor Object Reference=Get Button_ContextMenu()) -> (Anchor:Menu Anchor Object Reference))
  ]}
 ViewCreator {nodes [
   ViewCreator!() -> (then, NameID:Name) {is=title K2Node_FunctionEntry}
   > HandleCreatorPageButtonClicked()
   > Close(self=Get Anchor(self=Get Button_ContextMenu()))
  ]}
]
variables [
 ActivityDetailsDisplay {type=ActivityDetailsDisplay_C, default=null}
 bIsLoading {type=boolean}
 Button Context Menu {type=WBP_UIKit_ContextMenuAnchorButton_C, default=null}
 Buttons {type="array<CommonButtonBase>", default []}
 CachedGameActivity {type=FortGameActivity, default=null}
 Content Warning Map {
  type="map<string>"
  default {
   PartyLargerThanTeam=Your party is larger than Maximum team size. Playing this island will cause some party members to be on a different team.
   Photosensitivity=Photosensitivity Warning
  }
 }
 ContextMenuActionContext {type=UIKitContextMenuActionContext, default=null}
 DefaultImage {type=Texture, default=/Game/UI/DiscoveryShared/Art/T_UI_Discovery_DefaultEmpty}
 DesignTimeShowDetails {type=boolean, default=true}
 DesiredCommunityRowVisibility {type=boolean}
 Favorite_Sound {type=SoundBase, default=/Game/Sounds/Fort_UI_Sounds/FrontEnd/MultiProduct/Common/Buttons/MSS_UI_MP_Button_Click_Favorite}
 IntroWhooshSound {
  type=SoundBase
  default=/Game/Sounds/Fort_UI_Sounds/FrontEnd/MultiProduct/Common/Whooshes/Screens/MSS_UI_MP_Whoosh_Screen_Intro_Experience_Preview
 }
 IsAcceptButtonEnabled {type=boolean}
 IsFavorited {type=boolean}
 IsHabaneroEnabled {type=boolean}
 IsHabaneroMode {type=boolean}
 IsShowingDetailsOverride {type=boolean}
 IsShowingSubModeList {type=boolean}
 LastFocusedWidget {type=UserWidget, default=null}
 LinkCodesWithoutSubModeSelection {type="array<string>", default [set_delmar_mrs_ranked, set_delmar_mrs_testing_ranked, set_delmar_casual]}
 OutroWhooshSound {
  type=SoundBase
  default=/Game/Sounds/Fort_UI_Sounds/FrontEnd/MultiProduct/Common/Whooshes/Screens/MSS_UI_MP_Whoosh_Screen_Outro_Experience_Preview
 }
 ShowAcceptButton {type=boolean, default=true}
 ShowContentWarning {type=boolean}
 ShowCreator {type=boolean, default=true}
 ShowDetailsButton {type=boolean, default=true}
 ShowFavoriteButton {type=boolean, default=true}
 ShowLinkCode {type=boolean, default=true}
 ShowTags {type=boolean, default=true}
 SqueegeeDefaultTransform {type=WidgetTransform, default {angle=0, scale=<1 1>, shear=<0 0>, translation=<0 0>}}
 SqueegeeMobileScreenRatioLimit {type=number, default=1.34}
 SqueegeeMobileScreenRatioLimitTransform {type=WidgetTransform, default {angle=0, scale=<1 1>, shear=<0 0>, translation=<0 -64>}}
 ThumbnailOpacity {type=string, default=Thumbnail Opacity}
 ThumbnailParam {type=string, default=Thumbnail}
 Unfavorite_Sound {type=SoundBase, default=/Game/Sounds/Fort_UI_Sounds/FrontEnd/MultiProduct/Common/Buttons/MSS_UI_MP_Button_Click_Unfavorite}
]
widgets {
 info {
  inheritedWidgetCount=15
  namedSlotCount=1
  parentClass=/Script/FortniteUI.FortActivityDetailsModal
  rootWidgetClass=&Overlay
  widgetCount=108
 }
 widgets [
  Overlay_6 {widgetClassPath=&Overlay, bIsVariable=true, widget=Overlay_6}
  Button_BackBoard {
   widgetClassPath=/Game/Athena/UI/Matchmaking/SimpleCommonButton
   bInherited=true
   parent=Overlay_6
   widget=Button_BackBoard
   properties {
    onButtonBaseClicked=(&self:WidgetTree.Button_BackBoard.HandleOnButtonClicked)
    onButtonBaseDoubleClicked=(&self:WidgetTree.Button_BackBoard.HandleOnButtonDoubleClicked)
    onButtonBaseHovered=(&self:WidgetTree.Button_BackBoard.HandleOnButtonHovered)
    onButtonBaseUnhovered=(&self:WidgetTree.Button_BackBoard.HandleOnButtonUnhovered)
    onSelectedChangedBase=(&self:WidgetTree.Button_BackBoard.HandleOnSelectedChanged)
    style=/FortUILibrary/Foundation/Buttons/ButtonStyles/ButtonStyle_Empty_NoSound.ButtonStyle_Empty_NoSound_C
   }
   slot {class=OverlaySlot, properties {horizontalAlignment=HAlign_Fill, verticalAlignment=VAlign_Fill}}
  }
  ModalContent {
   widgetClassPath=&Overlay
   parent=Overlay_6
   widget=ModalContent
   slot {class=OverlaySlot, properties {horizontalAlignment=HAlign_Fill, verticalAlignment=VAlign_Fill}}
  }
  Background {
   widgetClassPath=&Image
   parent=ModalContent
   widget=Background
   properties {
    visibility=HitTestInvisible
    brush {
     bIsDynamicallyLoaded=false
     drawAs=Image
     imageType=NoImage
     mirroring=NoMirror
     resourceName=None
     resourceObject=/FortniteCoreUI/UI/DiscoveryShared/GameDetails/MaterialInstances/MI_UI_Discovery_GameDetails_BG
     tiling=NoTile
     imageSize=<32 32>
     margin=<ltrb 0 0 0 0>
     outlineSettings {
      bUseBrushTransparency=false
      roundingType=HalfHeightRadius
      width=0
      color {colorUseRule=UseColor_Specified, specifiedColor=<rgba 0 0 0 0>}
      cornerRadii=<0 0 0 0>
     }
     tintColor {colorUseRule=UseColor_Specified, specifiedColor=<rgba 1 1 1 1>}
     uVRegion {bIsValid=false, max=<0 0>, min=<0 0>}
    }
   }
   slot {class=OverlaySlot, properties {horizontalAlignment=HAlign_Fill, verticalAlignment=VAlign_Fill}}
  }
  WBP_CaptureForPostBufferUpdate_1 {
   widgetClassPath=&WBP_CaptureForPostBufferUpdate
   bIsVariable=true
   parent=ModalContent
   widget=WBP_CaptureForPostBufferUpdate_1
   slot {class=OverlaySlot, properties {horizontalAlignment=HAlign_Fill, verticalAlignment=VAlign_Fill}}
  }
  ScrollBox {
   widgetClassPath=&ScrollBox
   bIsVariable=true
   parent=ModalContent
   widget=ScrollBox
   properties {bAnimateWheelScrolling=true, scrollBarVisibility=Collapsed}
   slot {class=OverlaySlot, properties {horizontalAlignment=HAlign_Fill, verticalAlignment=VAlign_Fill}}
  }
  ContentWrapper {
   widgetClassPath=&GridPanel
   parent=ScrollBox
   widget=ContentWrapper
   properties {columnFill [1], rowFill [1]}
   slot {class=ScrollBoxSlot, properties {size {sizeRule=Fill, value=1}}}
  }
  GridPanel_HeroImage {
   widgetClassPath=&GridPanel
   parent=ContentWrapper
   widget=GridPanel_HeroImage
   properties {columnFill [1], rowFill [1]}
   slot {class=GridSlot, properties {verticalAlignment=VAlign_Top}}
  }
  Spacer_148 {
   widgetClassPath=&Spacer
   parent=GridPanel_HeroImage
   widget=Spacer_148
   properties {size=<1920 1080>}
   slot {class=GridSlot, properties {layer=-100}}
  }
  Thumbnail {
   widgetClassPath=&Image
   bIsVariable=true
   parent=GridPanel_HeroImage
   widget=Thumbnail
   properties {
    visibility=HitTestInvisible
    brush {
     bIsDynamicallyLoaded=false
     drawAs=Image
     imageType=NoImage
     mirroring=NoMirror
     resourceName=None
     resourceObject=/FortniteCoreUI/UI/DiscoveryShared/GameDetails/MaterialInstances/MI_UI_Discovery_GameDetails_Header_Image
     tiling=NoTile
     imageSize=<32 32>
     margin=<ltrb 0.5 0 0.5 0>
     outlineSettings {
      bUseBrushTransparency=false
      roundingType=HalfHeightRadius
      width=0
      color {colorUseRule=UseColor_Specified, specifiedColor=<rgba 0 0 0 0>}
      cornerRadii=<0 0 0 0>
     }
     tintColor {colorUseRule=UseColor_Specified, specifiedColor=<rgba 1 1 1 1>}
     uVRegion {bIsValid=false, max=<0 0>, min=<0 0>}
    }
   }
   slot {class=GridSlot}
  }
  ActivityVideoCycle {
   widgetClassPath=&/ActivityVideoCycleWidget
   bInherited=true
   bIsVariable=true
   parent=GridPanel_HeroImage
   widget=ActivityVideoCycle
   properties {
    bLoopVideo=true
    videoMaterial=/FortniteCoreUI/UI/DiscoveryShared/GameDetails/MaterialInstances/MI_UI_Discovery_GameDetails_Header_Video
    visibility=HitTestInvisible
   }
   slot {class=GridSlot, properties {layer=1}}
  }
  WBP_CaptureForPostBufferUpdate {
   widgetClassPath=&WBP_CaptureForPostBufferUpdate
   bIsVariable=true
   parent=GridPanel_HeroImage
   slot {class=GridSlot, properties {layer=100}}
  }
  Text_LinkCode {
   widgetClassPath=&CommonTextBlock
   bInherited=true
   parent=ContentWrapper
   widget=Text_LinkCode
   properties {
    justification=Center
    mobileFontSizeMultiplier=1.4
    text="{XXXX-XXXX-XXXX-XXXX}"
    font {
     bForceMonospaced=false
     bMaterialIsStencil=false
     fontMaterial=null
     fontObject=&Burbank
     letterSpacing=256
     monospacedWidth=1
     size=18
     skewAmount=0
     typefaceFontName=Burbank-PlayerPower
     outlineSettings {
      bApplyOutlineToDropShadows=false
      bMiteredCorners=false
      bSeparateFillAlpha=false
      outlineMaterial=null
      outlineSize=0
      outlineColor=<rgba 0 0 0 1>
     }
    }
    shadowOffset=<0 0>
   }
   slot {class=GridSlot, properties {horizontalAlignment=HAlign_Center, layer=40, verticalAlignment=VAlign_Bottom, padding=<ltrb 0 0 0 32>}}
  }
  WBP_UIKit_Throbber {
   widgetClassPath=/FortUILibrary/UIKit/Throbber/WBP_UIKit_Throbber
   bIsVariable=true
   parent=ContentWrapper
   widget=WBP_UIKit_Throbber
   properties {displayBackground=false, throbberSize=Large, visibility=Collapsed}
   slot {class=GridSlot, properties {horizontalAlignment=HAlign_Center, layer=50, verticalAlignment=VAlign_Center}}
  }
  SZ_Content_DynamicPadding {
   widgetClassPath=&SafeZone
   bIsVariable=true
   parent=ContentWrapper
   widget=SZ_Content_DynamicPadding
   properties {padBottom=false, padTop=false}
   slot {class=GridSlot, properties {layer=100, padding=<ltrb 80 952 80 112>}}
  }
  GP_Content {
   widgetClassPath=&GridPanel
   parent=SZ_Content_DynamicPadding
   widget=GP_Content
   properties {columnFill [1], rowFill [1]}
   slot {class=SafeZoneSlot}
  }
  GridPanel_ModeSelection {
   widgetClassPath=&GridPanel
   parent=GP_Content
   widget=GridPanel_ModeSelection
   properties {columnFill [0.3,0.7], rowFill [0]}
   slot {class=GridSlot}
  }
  NamedSlot_DetailsOverride {
   widgetClassPath=&NamedSlot
   bIsVariable=true
   parent=GridPanel_ModeSelection
   widget=NamedSlot_DetailsOverride
   properties {bExposeOnInstanceOnly=true}
   slot {class=GridSlot}
  }
  VBox_SubModeList {widgetClassPath=&VerticalBox, bIsVariable=true, parent=NamedSlot_DetailsOverride, widget=VBox_SubModeList, slot {class=PanelSlot}}
  HorizontalBox_ModeSelectionAvailable_SubMode {
   widgetClassPath=&HorizontalBox
   bIsVariable=true
   parent=VBox_SubModeList
   widget=HorizontalBox_ModeSelectionAvailable_SubMode
   slot {class=VerticalBoxSlot, properties {padding=<ltrb 0 0 0 8>}}
  }
  Image {
   widgetClassPath=&Image
   parent=HorizontalBox_ModeSelectionAvailable_SubMode
   widget=Image
   properties {
    visibility=SelfHitTestInvisible
    brush {
     bIsDynamicallyLoaded=false
     drawAs=Image
     imageType=NoImage
     mirroring=NoMirror
     resourceName=None
     resourceObject=/FortniteCoreUI/UI/DiscoveryShared/Art/T_UI_Discovery_SubModeIcon
     tiling=NoTile
     imageSize=<32 32>
     margin=<ltrb 0 0 0 0>
     outlineSettings {
      bUseBrushTransparency=false
      roundingType=HalfHeightRadius
      width=0
      color {colorUseRule=UseColor_Specified, specifiedColor=<rgba 0 0 0 0>}
      cornerRadii=<0 0 0 0>
     }
     tintColor {colorUseRule=UseColor_Specified, specifiedColor=<rgba 1 1 1 1>}
     uVRegion {bIsValid=false, max=<0 0>, min=<0 0>}
    }
    colorAndOpacity=<rgba 0.0331 0.402 1 1>
   }
   slot {class=HorizontalBoxSlot, properties {horizontalAlignment=HAlign_Left, verticalAlignment=VAlign_Center, padding=<ltrb 0 0 10 0>}}
  }
  CommonTextBlock {
   widgetClassPath=&CommonTextBlock
   parent=HorizontalBox_ModeSelectionAvailable_SubMode
   widget=CommonTextBlock
   properties {
    mobileFontSizeMultiplier=1.2
    text=Mode Selection
    textTransformPolicy=ToUpper
    colorAndOpacity {colorUseRule=UseColor_Specified, specifiedColor=<rgba 0.0331 0.402 1 1>}
    font {
     bForceMonospaced=false
     bMaterialIsStencil=false
     fontMaterial=null
     fontObject=&BurbankSmall
     letterSpacing=0
     monospacedWidth=1
     size=18
     skewAmount=0
     typefaceFontName=Black
     outlineSettings {
      bApplyOutlineToDropShadows=false
      bMiteredCorners=false
      bSeparateFillAlpha=false
      outlineMaterial=null
      outlineSize=0
      outlineColor=<rgba 0 0 0 1>
     }
    }
    shadowOffset=<0 0>
   }
   slot {class=HorizontalBoxSlot}
  }
  SizeBox_Habanero {
   widgetClassPath=&SizeBox
   bIsVariable=true
   parent=VBox_SubModeList
   widget=SizeBox_Habanero
   properties {bOverride_HeightOverride=true, heightOverride=69, visibility=Collapsed}
   slot {class=VerticalBoxSlot}
  }
  Border_Habanero {
   widgetClassPath=&Border
   parent=SizeBox_Habanero
   widget=Border_Habanero
   properties {
    background {
     bIsDynamicallyLoaded=false
     drawAs=Image
     imageType=NoImage
     mirroring=NoMirror
     resourceName=None
     resourceObject=/FortniteCoreUI/UI/Subscriptions/Art/M_UI_ShopTimerModalBg
     tiling=NoTile
     imageSize=<256 1>
     margin=<ltrb 0 0 0 0>
     outlineSettings {
      bUseBrushTransparency=false
      roundingType=HalfHeightRadius
      width=0
      color {colorUseRule=UseColor_Specified, specifiedColor=<rgba 0 0 0 0>}
      cornerRadii=<0 0 0 0>
     }
     tintColor {colorUseRule=UseColor_Specified, specifiedColor=<rgba 1 1 1 0.75>}
     uVRegion {bIsValid=false, max=<0 0>, min=<0 0>}
    }
    brushColor=<rgba 0.1719 0.1719 0.1719 1>
    padding=<ltrb 20 5 20 5>
   }
   slot {class=SizeBoxSlot}
  }
  HorizontalBox {
   widgetClassPath=&HorizontalBox
   parent=Border_Habanero
   widget=HorizontalBox
   slot {class=BorderSlot, properties {padding=<ltrb 20 5 20 5>}}
  }
  Icon_Habanero {
   widgetClassPath=/Game/UI/Frontend/Lobby/Habanero/WBP_Habanero_Icon
   bIsVariable=true
   parent=HorizontalBox
   widget=Icon_Habanero
   properties {iconMinSize=<64 64>}
   slot {class=HorizontalBoxSlot}
  }
  Text_Habanero {
   widgetClassPath=&CommonTextBlock
   bIsVariable=true
   parent=HorizontalBox
   widget=Text_Habanero
   properties {
    mobileFontSizeMultiplier=1.2
    text="[PH] -"
    textTransformPolicy=ToUpper
    colorAndOpacity {colorUseRule=UseColor_Specified, specifiedColor=<rgba 0.5089 0.5583 0.7084 1>}
    font {
     bForceMonospaced=false
     bMaterialIsStencil=false
     fontMaterial=null
     fontObject=&BurbankSmall
     letterSpacing=0
     monospacedWidth=1
     size=18
     skewAmount=0
     typefaceFontName=Black
     outlineSettings {
      bApplyOutlineToDropShadows=false
      bMiteredCorners=false
      bSeparateFillAlpha=false
      outlineMaterial=null
      outlineSize=0
      outlineColor=<rgba 0 0 0 1>
     }
    }
    shadowOffset=<0 0>
   }
   slot {class=HorizontalBoxSlot, properties {verticalAlignment=VAlign_Center}}
  }
  Button_Activity_Habanero {
   widgetClassPath=/Game/UI/Frontend/Lobby/Habanero/ActivityHabaneroButton
   bIsVariable=true
   parent=VBox_SubModeList
   widget=Button_Activity_Habanero
   properties {visibility=Collapsed}
   slot {class=VerticalBoxSlot}
  }
  Button_IslandPrivacy {
   widgetClassPath=/FortniteCoreUI/UI/DiscoveryShared/ModeSet/IslandPrivacyButton
   bInherited=true
   bIsVariable=true
   parent=VBox_SubModeList
   widget=Button_IslandPrivacy
   properties {forcedModeText="", offText=Public, onText=Private, titleText=Privacy}
   slot {class=VerticalBoxSlot}
  }
  Button_IslandFill {
   widgetClassPath=&IslandPrivacyButton
   bInherited=true
   bIsVariable=true
   parent=VBox_SubModeList
   widget=Button_IslandFill
   properties {forcedModeText=Fill team will be enabled automatically, offText=No Fill, onText=Fill, titleText=Team Fill}
   slot {class=VerticalBoxSlot}
  }
  List_SubModeList {
   widgetClassPath=/Script/FortniteUI.FortActivityModeSetSelection
   bInherited=true
   bIsVariable=true
   parent=VBox_SubModeList
   widget=List_SubModeList
   properties {
    entryWidgetClass=/FortniteCoreUI/UI/DiscoveryShared/ModeSet/ActivityBrowserModeSetListEntry.ActivityBrowserModeSetListEntry_C
    horizontalEntrySpacing=5
    navigation=&self:WidgetTree.List_SubModeList.WidgetNavigation_0
    numDesignerPreviewEntries=4
    verticalEntrySpacing=5
   }
   slot {class=VerticalBoxSlot, properties {size {sizeRule=Fill, value=1}}}
  }
  GridPanel_DynamicDetails {
   widgetClassPath=&GridPanel
   bIsVariable=true
   parent=GP_Content
   widget=GridPanel_DynamicDetails
   properties {columnFill [1], rowFill [0,0]}
   slot {class=GridSlot}
  }
  HorizontalBox_3 {
   widgetClassPath=&HorizontalBox
   parent=GridPanel_DynamicDetails
   widget=HorizontalBox_3
   properties {flowDirectionPreference=Culture}
   slot {class=GridSlot}
  }
  FortVisualAttachment__TopContent {
   widgetClassPath=&CommonVisualAttachment
   bIsVariable=true
   parent=HorizontalBox_3
   widget=FortVisualAttachment__TopContent
   properties {flowDirectionPreference=Culture, contentAnchor=<0 1>}
   slot {class=HorizontalBoxSlot, properties {horizontalAlignment=HAlign_Left, verticalAlignment=VAlign_Bottom, size {sizeRule=Fill, value=1}}}
  }
  VerticalBox_1 {widgetClassPath=&VerticalBox, parent=FortVisualAttachment__TopContent, widget=VerticalBox_1, slot {class=SizeBoxSlot}}
  HorizontalBox_0 {
   widgetClassPath=&HorizontalBox
   parent=VerticalBox_1
   widget=HorizontalBox_0
   slot {class=VerticalBoxSlot, properties {horizontalAlignment=HAlign_Left, padding=<ltrb 0 0 0 -2>}}
  }
  Text_ActivityOrigin {
   widgetClassPath=&CommonRichTextBlock
   bIsVariable=true
   parent=HorizontalBox_0
   widget=Text_ActivityOrigin
   properties {
    mobileTextBlockScale=1.3
    text="Created by <CreatorStyle>{Creator}</> "
    textStyleSet=/Game/UI/DiscoveryShared/DT_UI_RichDetails_Tags
    defaultTextStyleOverride {
     overflowPolicy=Clip
     transformPolicy=None
     colorAndOpacity {colorUseRule=UseColor_Specified, specifiedColor=<rgba 0.0331 0.402 1 1>}
     font {
      bForceMonospaced=false
      bMaterialIsStencil=false
      fontMaterial=null
      fontObject=&BurbankSmall
      letterSpacing=0
      monospacedWidth=1
      size=15
      skewAmount=0
      typefaceFontName=Bold
      outlineSettings {
       bApplyOutlineToDropShadows=false
       bMiteredCorners=false
       bSeparateFillAlpha=false
       outlineMaterial=null
       outlineSize=0
       outlineColor=<rgba 0 0 0 1>
      }
     }
     highlightColor {colorUseRule=UseColor_Specified, specifiedColor=<rgba 0 0 0 1>}
     highlightShape {
      bIsDynamicallyLoaded=false
      drawAs=Image
      imageType=NoImage
      mirroring=NoMirror
      resourceName=None
      resourceObject=null
      tiling=NoTile
      imageSize=<32 32>
      margin=<ltrb 0 0 0 0>
      outlineSettings {
       bUseBrushTransparency=false
       roundingType=HalfHeightRadius
       width=0
       color {colorUseRule=UseColor_Specified, specifiedColor=<rgba 0 0 0 0>}
       cornerRadii=<0 0 0 0>
      }
      tintColor {colorUseRule=UseColor_Specified, specifiedColor=<rgba 1 1 1 1>}
      uVRegion {bIsValid=false, max=<0 0>, min=<0 0>}
     }
     selectedBackgroundColor {colorUseRule=UseColor_Foreground, specifiedColor=<rgba 1 0 1 1>}
     shadowColorAndOpacity=<rgba 0 0 0 1>
     shadowOffset=<0 0>
     strikeBrush {
      bIsDynamicallyLoaded=false
      drawAs=Image
      imageType=NoImage
      mirroring=NoMirror
      resourceName=None
      resourceObject=null
      tiling=NoTile
      imageSize=<32 32>
      margin=<ltrb 0 0 0 0>
      outlineSettings {
       bUseBrushTransparency=false
       roundingType=HalfHeightRadius
       width=0
       color {colorUseRule=UseColor_Specified, specifiedColor=<rgba 0 0 0 0>}
       cornerRadii=<0 0 0 0>
      }
      tintColor {colorUseRule=UseColor_Specified, specifiedColor=<rgba 1 1 1 1>}
      uVRegion {bIsValid=false, max=<0 0>, min=<0 0>}
     }
     underlineBrush {
      bIsDynamicallyLoaded=false
      drawAs=Image
      imageType=NoImage
      mirroring=NoMirror
      resourceName=None
      resourceObject=null
      tiling=NoTile
      imageSize=<32 32>
      margin=<ltrb 0 0 0 0>
      outlineSettings {
       bUseBrushTransparency=false
       roundingType=HalfHeightRadius
       width=0
       color {colorUseRule=UseColor_Specified, specifiedColor=<rgba 0 0 0 0>}
       cornerRadii=<0 0 0 0>
      }
      tintColor {colorUseRule=UseColor_Specified, specifiedColor=<rgba 1 1 1 1>}
      uVRegion {bIsValid=false, max=<0 0>, min=<0 0>}
     }
    }
    margin=<ltrb 0 -4 0 -4>
   }
   slot {class=HorizontalBoxSlot, properties {padding=<ltrb 0 0 32 0>}}
  }
  Text_LinkCode_1 {
   widgetClassPath=&CommonRichTextBlock
   bIsVariable=true
   parent=HorizontalBox_0
   widget=Text_LinkCode_1
   properties {
    mobileTextBlockScale=1.3
    text="Island Code <CreatorStyle>{XXXX-XXXX-XXXX-XXXX}</> "
    textStyleSet=&DT_UI_RichDetails_Tags
    visibility=Collapsed
    defaultTextStyleOverride {
     overflowPolicy=Clip
     transformPolicy=None
     colorAndOpacity {colorUseRule=UseColor_Specified, specifiedColor=<rgba 0.0331 0.402 1 1>}
     font {
      bForceMonospaced=false
      bMaterialIsStencil=false
      fontMaterial=null
      fontObject=&BurbankSmall
      letterSpacing=0
      monospacedWidth=1
      size=15
      skewAmount=0
      typefaceFontName=Bold
      outlineSettings {
       bApplyOutlineToDropShadows=false
       bMiteredCorners=false
       bSeparateFillAlpha=false
       outlineMaterial=null
       outlineSize=0
       outlineColor=<rgba 0 0 0 1>
      }
     }
     highlightColor {colorUseRule=UseColor_Specified, specifiedColor=<rgba 0 0 0 1>}
     highlightShape {
      bIsDynamicallyLoaded=false
      drawAs=Image
      imageType=NoImage
      mirroring=NoMirror
      resourceName=None
      resourceObject=null
      tiling=NoTile
      imageSize=<32 32>
      margin=<ltrb 0 0 0 0>
      outlineSettings {
       bUseBrushTransparency=false
       roundingType=HalfHeightRadius
       width=0
       color {colorUseRule=UseColor_Specified, specifiedColor=<rgba 0 0 0 0>}
       cornerRadii=<0 0 0 0>
      }
      tintColor {colorUseRule=UseColor_Specified, specifiedColor=<rgba 1 1 1 1>}
      uVRegion {bIsValid=false, max=<0 0>, min=<0 0>}
     }
     selectedBackgroundColor {colorUseRule=UseColor_Foreground, specifiedColor=<rgba 1 0 1 1>}
     shadowColorAndOpacity=<rgba 0 0 0 1>
     shadowOffset=<0 0>
     strikeBrush {
      bIsDynamicallyLoaded=false
      drawAs=Image
      imageType=NoImage
      mirroring=NoMirror
      resourceName=None
      resourceObject=null
      tiling=NoTile
      imageSize=<32 32>
      margin=<ltrb 0 0 0 0>
      outlineSettings {
       bUseBrushTransparency=false
       roundingType=HalfHeightRadius
       width=0
       color {colorUseRule=UseColor_Specified, specifiedColor=<rgba 0 0 0 0>}
       cornerRadii=<0 0 0 0>
      }
      tintColor {colorUseRule=UseColor_Specified, specifiedColor=<rgba 1 1 1 1>}
      uVRegion {bIsValid=false, max=<0 0>, min=<0 0>}
     }
     underlineBrush {
      bIsDynamicallyLoaded=false
      drawAs=Image
      imageType=NoImage
      mirroring=NoMirror
      resourceName=None
      resourceObject=null
      tiling=NoTile
      imageSize=<32 32>
      margin=<ltrb 0 0 0 0>
      outlineSettings {
       bUseBrushTransparency=false
       roundingType=HalfHeightRadius
       width=0
       color {colorUseRule=UseColor_Specified, specifiedColor=<rgba 0 0 0 0>}
       cornerRadii=<0 0 0 0>
      }
      tintColor {colorUseRule=UseColor_Specified, specifiedColor=<rgba 1 1 1 1>}
      uVRegion {bIsValid=false, max=<0 0>, min=<0 0>}
     }
    }
    margin=<ltrb 0 -4 0 -4>
   }
   slot {class=HorizontalBoxSlot}
  }
  Text_ActivityName {
   widgetClassPath=/Script/FortniteUI.FortActivityScalingTextBlock
   bIsVariable=true
   parent=VerticalBox_1
   widget=Text_ActivityName
   properties {
    applyLineHeightToBottomLine=false
    autoWrapText=true
    flowDirectionPreference=Culture
    lineHeightPercentage=0.8
    scaling=/Game/UI/DiscoveryShared/ActivityNameFontScaleData.ActivityNameFontScaleData_C
    text="{Game Mode}{Game Mode}{Game Mode}{Game Mode}"
    textTransformPolicy=ToUpper
    wrapTextAt=1300
    font {
     bForceMonospaced=false
     bMaterialIsStencil=false
     fontMaterial=null
     fontObject=&HeadingNow
     letterSpacing=-40
     monospacedWidth=1
     size=39
     skewAmount=0
     typefaceFontName=86 Bold Italic
     outlineSettings {
      bApplyOutlineToDropShadows=false
      bMiteredCorners=false
      bSeparateFillAlpha=false
      outlineMaterial=/FortUILibrary/MaterialInstances/Text/MI_UI_Block_Text_Outline
      outlineSize=4
      outlineColor=<rgba 0 0 0 1>
     }
    }
    shadowOffset=<0 0>
    shapedTextOptions {bOverride_TextFlowDirection=true, bOverride_TextShapingMethod=false, textFlowDirection=Culture, textShapingMethod=Auto}
    strikeBrush {
     bIsDynamicallyLoaded=false
     drawAs=NoDrawType
     imageType=NoImage
     mirroring=NoMirror
     resourceName=None
     resourceObject=null
     tiling=NoTile
     imageSize=<32 32>
     margin=<ltrb 0 0 0 0>
     outlineSettings {
      bUseBrushTransparency=false
      roundingType=HalfHeightRadius
      width=0
      color {colorUseRule=UseColor_Specified, specifiedColor=<rgba 0 0 0 0>}
      cornerRadii=<0 0 0 0>
     }
     tintColor {colorUseRule=UseColor_Specified, specifiedColor=<rgba 1 1 1 1>}
     uVRegion {bIsValid=false, max=<0 0>, min=<0 0>}
    }
   }
   slot {class=VerticalBoxSlot, properties {padding=<ltrb 0 0 0 8>}}
  }
  SizeBox_3 {
   widgetClassPath=&SizeBox
   bIsVariable=true
   parent=VerticalBox_1
   widget=SizeBox_3
   properties {bOverride_WidthOverride=true, widthOverride=1300}
   slot {class=VerticalBoxSlot, properties {padding=<ltrb 0 0 0 24>, size {sizeRule=Fill, value=1}}}
  }
  EntryBox_ActivityTags {
   widgetClassPath=&DynamicEntryBox
   bIsVariable=true
   parent=SizeBox_3
   widget=EntryBox_ActivityTags
   properties {
    entryBoxType=Wrap
    entryHorizontalAlignment=HAlign_Right
    entryWidgetClass=&WBP_UIKit_Tag_C
    flowDirectionPreference=Culture
    numDesignerPreviewEntries=4
    entrySpacing=<8 8>
   }
   slot {class=SizeBoxSlot}
  }
  WarningEntryBox {
   widgetClassPath=&DynamicEntryBox
   bIsVariable=true
   parent=VerticalBox_1
   widget=WarningEntryBox
   properties {
    entryBoxType=Vertical
    entryWidgetClass=&ActivityBrowserStatusWarning_C
    numDesignerPreviewEntries=1
    entrySpacing=<25 0>
   }
   slot {class=VerticalBoxSlot, properties {horizontalAlignment=HAlign_Left, padding=<ltrb 0 0 0 16>}}
  }
  DisabledWarningHB {
   widgetClassPath=&HorizontalBox
   bIsVariable=true
   parent=VerticalBox_1
   widget=DisabledWarningHB
   slot {class=VerticalBoxSlot, properties {horizontalAlignment=HAlign_Left, padding=<ltrb 0 0 0 16>}}
  }
  Overlay {
   widgetClassPath=&Overlay
   parent=DisabledWarningHB
   widget=Overlay
   slot {class=HorizontalBoxSlot, properties {horizontalAlignment=HAlign_Left, verticalAlignment=VAlign_Center}}
  }
  WarningTriangle_Pulse {
   widgetClassPath=&Image
   parent=Overlay
   widget=WarningTriangle_Pulse
   properties {
    renderOpacity=0
    visibility=SelfHitTestInvisible
    brush {
     bIsDynamicallyLoaded=false
     drawAs=Image
     imageType=NoImage
     mirroring=NoMirror
     resourceName=None
     resourceObject=/Game/UI/DiscoveryShared/Art/M_UI_Discovery_WarningTriangle_Pulse
     tiling=NoTile
     imageSize=<40 40>
     margin=<ltrb 0 0 0 0>
     outlineSettings {
      bUseBrushTransparency=false
      roundingType=HalfHeightRadius
      width=0
      color {colorUseRule=UseColor_Specified, specifiedColor=<rgba 0 0 0 0>}
      cornerRadii=<0 0 0 0>
     }
     tintColor {colorUseRule=UseColor_Specified, specifiedColor=<rgba 1 1 1 1>}
     uVRegion {bIsValid=false, max=<0 0>, min=<0 0>}
    }
   }
   slot {class=OverlaySlot, properties {horizontalAlignment=HAlign_Fill, verticalAlignment=VAlign_Fill}}
  }
  IconOvr {
   widgetClassPath=&Overlay
   parent=Overlay
   widget=IconOvr
   properties {visibility=HitTestInvisible}
   slot {class=OverlaySlot, properties {horizontalAlignment=HAlign_Fill, verticalAlignment=VAlign_Fill}}
  }
  WarningTriangle {
   widgetClassPath=&Image
   parent=IconOvr
   widget=WarningTriangle
   properties {
    visibility=SelfHitTestInvisible
    brush {
     bIsDynamicallyLoaded=false
     drawAs=Image
     imageType=NoImage
     mirroring=NoMirror
     resourceName=None
     resourceObject=/Game/UI/DiscoveryShared/Art/M_UI_Discovery_WarningTriangle
     tiling=NoTile
     imageSize=<32 32>
     margin=<ltrb 0 0 0 0>
     outlineSettings {
      bUseBrushTransparency=false
      roundingType=HalfHeightRadius
      width=0
      color {colorUseRule=UseColor_Specified, specifiedColor=<rgba 0 0 0 0>}
      cornerRadii=<0 0 0 0>
     }
     tintColor {colorUseRule=UseColor_Specified, specifiedColor=<rgba 1 1 1 1>}
     uVRegion {bIsValid=false, max=<0 0>, min=<0 0>}
    }
   }
   slot {class=OverlaySlot, properties {horizontalAlignment=HAlign_Fill, verticalAlignment=VAlign_Fill}}
  }
  Exclamation {
   widgetClassPath=&CommonTextBlock
   parent=IconOvr
   widget=Exclamation
   properties {
    text="!"
    colorAndOpacity {colorUseRule=UseColor_Specified, specifiedColor=<rgba 0 0 0 1>}
    font {
     bForceMonospaced=false
     bMaterialIsStencil=false
     fontMaterial=null
     fontObject=&Burbank
     letterSpacing=0
     monospacedWidth=1
     size=18
     skewAmount=0
     typefaceFontName=BurbankRegular-Black
     outlineSettings {
      bApplyOutlineToDropShadows=false
      bMiteredCorners=false
      bSeparateFillAlpha=false
      outlineMaterial=null
      outlineSize=0
      outlineColor=<rgba 0 0 0 1>
     }
    }
    margin=<ltrb 0 -4 0 -4>
    renderTransform {angle=0, scale=<1 1>, shear=<4 0>, translation=<-1 5>}
    shadowOffset=<0 0>
   }
   slot {class=OverlaySlot, properties {horizontalAlignment=HAlign_Center, verticalAlignment=VAlign_Center}}
  }
  DisabledWarningMessage {
   widgetClassPath=&CommonTextBlock
   bIsVariable=true
   parent=DisabledWarningHB
   widget=DisabledWarningMessage
   properties {
    text="{Disabled warning message. Lorem ipsum.}"
    wrapTextAt=544
    font {
     bForceMonospaced=false
     bMaterialIsStencil=false
     fontMaterial=null
     fontObject=&HeadingNow
     letterSpacing=10
     monospacedWidth=1
     size=15.75
     skewAmount=0
     typefaceFontName=74 Regular
     outlineSettings {
      bApplyOutlineToDropShadows=false
      bMiteredCorners=false
      bSeparateFillAlpha=false
      outlineMaterial=&MI_UI_Block_Text_Outline
      outlineSize=2
      outlineColor=<rgba 0 0 0 1>
     }
    }
    shadowOffset=<0 0>
    shapedTextOptions {bOverride_TextFlowDirection=true, bOverride_TextShapingMethod=false, textFlowDirection=Culture, textShapingMethod=Auto}
   }
   slot {class=HorizontalBoxSlot, properties {verticalAlignment=VAlign_Center, padding=<ltrb 8 0 0 0>, size {sizeRule=Fill, value=1}}}
  }
  NamedSlot_AdditionalDetail {
   widgetClassPath=&NamedSlot
   bIsVariable=true
   parent=VerticalBox_1
   widget=NamedSlot_AdditionalDetail
   properties {bExposeOnInstanceOnly=true, visibility=Collapsed}
   slot {class=VerticalBoxSlot}
  }
  VerticalBox_ModeSelectionAvailable {
   widgetClassPath=&VerticalBox
   bIsVariable=true
   parent=NamedSlot_AdditionalDetail
   widget=VerticalBox_ModeSelectionAvailable
   slot {class=PanelSlot}
  }
  HorizontalBox_ModeSelectionAvailable {
   widgetClassPath=&HorizontalBox
   bIsVariable=true
   parent=VerticalBox_ModeSelectionAvailable
   widget=HorizontalBox_ModeSelectionAvailable
   slot {class=VerticalBoxSlot, properties {horizontalAlignment=HAlign_Left, padding=<ltrb 0 0 0 10>}}
  }
  Image_77 {
   widgetClassPath=&Image
   parent=HorizontalBox_ModeSelectionAvailable
   widget=Image_77
   properties {
    visibility=SelfHitTestInvisible
    brush {
     bIsDynamicallyLoaded=false
     drawAs=Image
     imageType=NoImage
     mirroring=NoMirror
     resourceName=None
     resourceObject=&T_UI_Discovery_SubModeIcon
     tiling=NoTile
     imageSize=<32 32>
     margin=<ltrb 0 0 0 0>
     outlineSettings {
      bUseBrushTransparency=false
      roundingType=HalfHeightRadius
      width=0
      color {colorUseRule=UseColor_Specified, specifiedColor=<rgba 0 0 0 0>}
      cornerRadii=<0 0 0 0>
     }
     tintColor {colorUseRule=UseColor_Specified, specifiedColor=<rgba 1 1 1 1>}
     uVRegion {bIsValid=false, max=<0 0>, min=<0 0>}
    }
    colorAndOpacity=<rgba 0 0.3968 0.9823 1>
   }
   slot {class=HorizontalBoxSlot, properties {horizontalAlignment=HAlign_Left, verticalAlignment=VAlign_Center, padding=<ltrb 0 0 10 0>}}
  }
  CommonTextBlock_75 {
   widgetClassPath=&CommonTextBlock
   parent=HorizontalBox_ModeSelectionAvailable
   widget=CommonTextBlock_75
   properties {
    mobileFontSizeMultiplier=1.2
    text=Multiple Modes
    textTransformPolicy=ToUpper
    colorAndOpacity {colorUseRule=UseColor_Specified, specifiedColor=<rgba 0 0.3968 0.9823 1>}
    font {
     bForceMonospaced=false
     bMaterialIsStencil=false
     fontMaterial=null
     fontObject=&BurbankSmall
     letterSpacing=0
     monospacedWidth=1
     size=18
     skewAmount=0
     typefaceFontName=Black
     outlineSettings {
      bApplyOutlineToDropShadows=false
      bMiteredCorners=false
      bSeparateFillAlpha=false
      outlineMaterial=null
      outlineSize=0
      outlineColor=<rgba 0 0 0 1>
     }
    }
    shadowOffset=<0 0>
   }
   slot {class=HorizontalBoxSlot}
  }
  HorizontalBox_2 {
   widgetClassPath=&HorizontalBox
   parent=VerticalBox_ModeSelectionAvailable
   widget=HorizontalBox_2
   slot {class=VerticalBoxSlot, properties {horizontalAlignment=HAlign_Left}}
  }
  Button_ShowSubModeList {
   widgetClassPath=&WBP_UIKit_Button_Regular
   bIsVariable=true
   parent=HorizontalBox_2
   widget=Button_ShowSubModeList
   properties {
    bDisplayInputActionWhenNotInteractable=false
    bHideInputAction=true
    buttonSize=Large
    minWidthOverride=320
    navigation=&self:WidgetTree.Button_ShowSubModeList.WidgetNavigation_0
    text="{Text}"
    triggeringInputAction {dataTable=/FortUILibrary/Foundation/ActivityBrowserInputActionTable, rowName=ShowSubModes_DetailsModal}
   }
   slot {class=HorizontalBoxSlot}
  }
  Spacer_70 {
   widgetClassPath=&Spacer
   bIsVariable=true
   parent=HorizontalBox_2
   widget=Spacer_70
   properties {size=<5 1>}
   slot {class=HorizontalBoxSlot}
  }
  DiscoveryTooltip {
   widgetClassPath=/Game/UI/Tooltips/WBP_PopupTooltip
   bIsVariable=true
   parent=HorizontalBox_2
   widget=DiscoveryTooltip
   properties {
    arrowHorizontalAlignment=HAlign_Left
    arrowVerticalAlignment=VAlign_Center
    firstLineText=Change mode here
    secondLineRichText=""
    arrowPadding=<ltrb 0 0 0 0>
    arrowTransform {angle=-90, scale=<1 1>, shear=<0 0>, translation=<-20 0>}
    renderTransformPivot=<0 0.5>
   }
   slot {class=HorizontalBoxSlot, properties {verticalAlignment=VAlign_Center, padding=<ltrb 16 0 0 0>}}
  }
  Spacer_63 {
   widgetClassPath=&Spacer
   bIsVariable=true
   parent=VerticalBox_ModeSelectionAvailable
   widget=Spacer_63
   properties {size=<1 16>}
   slot {class=VerticalBoxSlot}
  }
  HB_Buttons {
   widgetClassPath=&HorizontalBox
   parent=VerticalBox_1
   widget=HB_Buttons
   slot {class=VerticalBoxSlot, properties {horizontalAlignment=HAlign_Left, verticalAlignment=VAlign_Bottom}}
  }
  WidgetSwitcher_Buttons {
   widgetClassPath=&WidgetSwitcher
   bIsVariable=true
   parent=HB_Buttons
   widget=WidgetSwitcher_Buttons
   slot {class=HorizontalBoxSlot, properties {verticalAlignment=VAlign_Bottom}}
  }
  Button_Accept {
   widgetClassPath=/FortUILibrary/UIKit/Buttons/WBP_UIKit_Button_Loud
   bIsVariable=true
   parent=WidgetSwitcher_Buttons
   widget=Button_Accept
   properties {
    bHideInputAction=true
    buttonSize=Large
    minHeight=88
    minWidth=240
    minWidthOverride=320
    mobileMinWidth=400
    navigation=&self:WidgetTree.Button_Accept.WidgetNavigation_0
    text=Play
   }
   slot {class=WidgetSwitcherSlot, properties {horizontalAlignment=HAlign_Left, verticalAlignment=VAlign_Bottom}}
  }
  SqueegeeSlot_LockInfo {
   widgetClassPath=/Game/UI/Plugin/Squeegee/SqueegeeInjectionSlot
   bIsVariable=true
   parent=WidgetSwitcher_Buttons
   widget=SqueegeeSlot_LockInfo
   properties {
    bIsFocusable=true
    navigation=&self:WidgetTree.SqueegeeSlot_LockInfo.WidgetNavigation_0
    pluginId=Squeegee
    pluginWidgetId=Game Mode Info
   }
   slot {class=WidgetSwitcherSlot, properties {horizontalAlignment=HAlign_Left, verticalAlignment=VAlign_Bottom}}
  }
  Button_TimeLocked {
   widgetClassPath=&WBP_UIKit_Button_Loud
   bIsVariable=true
   parent=WidgetSwitcher_Buttons
   widget=Button_TimeLocked
   properties {
    bHideInputAction=true
    buttonSize=Large
    icon=/FortUILibrary/Textures/Icons/T_UI_Icon_UIKit_UpdatePending_SDF
    iconMaterial=/FortUILibrary/Textures/MaterialInstances/MI_UIKit_Icon_UpdatePending
    minHeight=88
    minWidth=400
    navigation=&self:WidgetTree.Button_TimeLocked.WidgetNavigation_0
    text="{Timer}"
   }
   slot {class=WidgetSwitcherSlot, properties {horizontalAlignment=HAlign_Left, verticalAlignment=VAlign_Bottom}}
  }
  HB_SideButtonsAndInfos {
   widgetClassPath=&HorizontalBox
   parent=HB_Buttons
   widget=HB_SideButtonsAndInfos
   slot {class=HorizontalBoxSlot, properties {horizontalAlignment=HAlign_Left, verticalAlignment=VAlign_Bottom, padding=<ltrb 16 0 0 0>}}
  }
  Button_Favorite {
   widgetClassPath=/Game/UI/DiscoveryShared/GameDetails/WBP_GameDetails_Toggle
   bInherited=true
   bIsVariable=true
   parent=HB_SideButtonsAndInfos
   widget=Button_Favorite
   properties {
    bHideInputAction=true
    minHeightOverride=88
    minWidthOverride=88
    mobileMinHeight=100
    mobileMinWidth=100
    navigation=&self:WidgetTree.Button_Favorite.WidgetNavigation_0
   }
   slot {class=HorizontalBoxSlot, properties {horizontalAlignment=HAlign_Left, verticalAlignment=VAlign_Bottom, padding=<ltrb 0 0 16 0>}}
  }
  Button_Subscribe {
   widgetClassPath=&WBP_UIKit_Button_Regular
   bIsVariable=true
   parent=HB_SideButtonsAndInfos
   widget=Button_Subscribe
   properties {
    bHideInputAction=true
    bUseNaturalNavigation=true
    iconMaterial=&MI_UI_ActivityDetails_NotificationOff
    minHeight=56
    minHeightOverride=80
    minWidth=80
    minWidthOverride=80
    navigation=&self:WidgetTree.Button_Subscribe.WidgetNavigation_0
    text=""
   }
   slot {class=HorizontalBoxSlot, properties {horizontalAlignment=HAlign_Left, padding=<ltrb 0 0 16 0>}}
  }
  Button_ContextMenu {
   widgetClassPath=/FortUILibrary/UIKit/ContextMenu/WBP_UIKit_ContextMenuAnchorButton
   bIsVariable=true
   parent=HB_SideButtonsAndInfos
   widget=Button_ContextMenu
   properties {
    bIsFocusable=true
    buttonClass=/Game/UI/DiscoveryShared/GameDetails/WBP_GameDetails_Details.WBP_GameDetails_Details_C
    buttonIcon=/FortUILibrary/Textures/Icons/T_UI_Icon_UIKit_ContextMenu
    buttonText=""
    min Height=88
    min Width=88
    navigation=&self:WidgetTree.Button_ContextMenu.WidgetNavigation_0
    visibility=HitTestInvisible
    button Padding=<ltrb 8 8 8 8>
   }
   slot {class=HorizontalBoxSlot, properties {horizontalAlignment=HAlign_Left, verticalAlignment=VAlign_Bottom}}
  }
  WBP_CCUCount {
   widgetClassPath=&/WBP_IslandLabelAndDetails
   bIsVariable=true
   parent=HB_SideButtonsAndInfos
   widget=WBP_CCUCount
   properties {icon=/Game/UI/DiscoveryShared/Art/MI_UI_Icon_CCU, label_Text=Playing Now}
   slot {class=HorizontalBoxSlot, properties {horizontalAlignment=HAlign_Left, verticalAlignment=VAlign_Center, padding=<ltrb 40 0 0 0>}}
  }
  WBP_IslandCode {
   widgetClassPath=&WBP_IslandLabelAndDetails
   bIsVariable=true
   parent=HB_SideButtonsAndInfos
   widget=WBP_IslandCode
   properties {hasIcon=false, label_Text=Island Code}
   slot {class=HorizontalBoxSlot, properties {horizontalAlignment=HAlign_Left, verticalAlignment=VAlign_Center, padding=<ltrb 40 0 0 0>}}
  }
  FortVisualAttachment__RatingCard {
   widgetClassPath=&CommonVisualAttachment
   bIsVariable=true
   parent=HorizontalBox_3
   widget=FortVisualAttachment__RatingCard
   properties {contentAnchor=<1 1>}
   slot {class=HorizontalBoxSlot, properties {horizontalAlignment=HAlign_Right, verticalAlignment=VAlign_Bottom}}
  }
  SqueegeeSlot_Card {
   widgetClassPath=&SqueegeeInjectionSlot
   bIsVariable=true
   parent=FortVisualAttachment__RatingCard
   widget=SqueegeeSlot_Card
   properties {
    bIsFocusable=true
    navigation=&self:WidgetTree.SqueegeeSlot_Card.WidgetNavigation_0
    pluginId=Squeegee
    pluginWidgetId=RatingCard
   }
   slot {class=SizeBoxSlot}
  }
  VerticalBox_0 {
   widgetClassPath=&VerticalBox
   parent=GridPanel_DynamicDetails
   widget=VerticalBox_0
   slot {class=GridSlot, properties {row=1, padding=<ltrb -80 42 -80 0>}}
  }
  CommunityRow {
   widgetClassPath=/Game/UI/Plugin/FortPluginWidgetInjectionSlot
   bInherited=true
   bIsVariable=true
   parent=VerticalBox_0
   widget=CommunityRow
   properties {pluginId=DiscoveryBrowser, pluginWidgetId=CommunityRow, visibility=Collapsed}
   slot {class=VerticalBoxSlot}
  }
  VerticalBox_2 {
   widgetClassPath=&VerticalBox
   parent=GridPanel_DynamicDetails
   widget=VerticalBox_2
   properties {bOverride_Cursor=true, flowDirectionPreference=Culture}
   slot {class=GridSlot, properties {horizontalAlignment=HAlign_Left, row=2, padding=<ltrb 0 42 0 0>}}
  }
  Text_Details {
   widgetClassPath=&CommonTextBlock
   bIsVariable=true
   parent=VerticalBox_2
   widget=Text_Details
   properties {
    applyLineHeightToBottomLine=false
    autoWrapText=true
    mobileFontSizeMultiplier=1.2
    text=Description
    wrappingPolicy=AllowPerCharacterWrapping
    colorAndOpacity {colorUseRule=UseColor_Specified, specifiedColor=<rgba 1 1 1 0.75>}
    font {
     bForceMonospaced=false
     bMaterialIsStencil=false
     fontMaterial=null
     fontObject=&HeadingNow
     letterSpacing=-40
     monospacedWidth=1
     size=15.75
     skewAmount=0
     typefaceFontName=75 Medium
     outlineSettings {
      bApplyOutlineToDropShadows=false
      bMiteredCorners=false
      bSeparateFillAlpha=false
      outlineMaterial=null
      outlineSize=0
      outlineColor=<rgba 0 0 0 1>
     }
    }
    shadowOffset=<0 0>
    shapedTextOptions {bOverride_TextFlowDirection=true, bOverride_TextShapingMethod=false, textFlowDirection=Culture, textShapingMethod=Auto}
    strikeBrush {
     bIsDynamicallyLoaded=false
     drawAs=NoDrawType
     imageType=NoImage
     mirroring=NoMirror
     resourceName=None
     resourceObject=null
     tiling=NoTile
     imageSize=<32 32>
     margin=<ltrb 0 0 0 0>
     outlineSettings {
      bUseBrushTransparency=false
      roundingType=HalfHeightRadius
      width=0
      color {colorUseRule=UseColor_Specified, specifiedColor=<rgba 0 0 0 0>}
      cornerRadii=<0 0 0 0>
     }
     tintColor {colorUseRule=UseColor_Specified, specifiedColor=<rgba 1 1 1 1>}
     uVRegion {bIsValid=false, max=<0 0>, min=<0 0>}
    }
   }
   slot {class=VerticalBoxSlot, properties {padding=<ltrb 0 0 0 24>}}
  }
  Button_Details_Hover {
   widgetClassPath=&Button
   parent=VerticalBox_2
   widget=Button_Details_Hover
   properties {
    navigation=&self:WidgetTree.Button_Details_Hover.WidgetNavigation_0
    visibility=SelfHitTestInvisible
    widgetStyle {
     clickedSlateSound {resourceObject=null}
     disabled {
      bIsDynamicallyLoaded=false
      drawAs=NoDrawType
      imageType=NoImage
      mirroring=NoMirror
      resourceName=None
      resourceObject=null
      tiling=NoTile
      imageSize=<0 0>
      margin=<ltrb 0 0 0 0>
      outlineSettings {
       bUseBrushTransparency=false
       roundingType=HalfHeightRadius
       width=0
       color {colorUseRule=UseColor_Specified, specifiedColor=<rgba 0 0 0 0>}
       cornerRadii=<0 0 0 0>
      }
      tintColor {colorUseRule=UseColor_Specified, specifiedColor=<rgba 1 1 1 1>}
      uVRegion {bIsValid=false, max=<0 0>, min=<0 0>}
     }
     disabledForeground {colorUseRule=UseColor_Specified, specifiedColor=<rgba 0.5271 0.5271 0.5271 1>}
     hovered {
      bIsDynamicallyLoaded=false
      drawAs=NoDrawType
      imageType=NoImage
      mirroring=NoMirror
      resourceName=None
      resourceObject=null
      tiling=NoTile
      imageSize=<32 32>
      margin=<ltrb 0 0 0 0>
      outlineSettings {
       bUseBrushTransparency=true
       roundingType=FixedRadius
       width=1
       color {colorUseRule=UseColor_Specified, specifiedColor=<rgba 0.7243 0.7243 0.7243 1>}
       cornerRadii=<4 4 4 4>
      }
      tintColor {colorUseRule=UseColor_Specified, specifiedColor=<rgba 0.7243 0.7243 0.7243 1>}
      uVRegion {bIsValid=false, max=<0 0>, min=<0 0>}
     }
     hoveredForeground {colorUseRule=UseColor_Specified, specifiedColor=<rgba 0 1 1 1>}
     hoveredSlateSound {resourceObject=null}
     normal {
      bIsDynamicallyLoaded=false
      drawAs=NoDrawType
      imageType=NoImage
      mirroring=NoMirror
      resourceName=None
      resourceObject=null
      tiling=NoTile
      imageSize=<32 32>
      margin=<ltrb 0 0 0 0>
      outlineSettings {
       bUseBrushTransparency=true
       roundingType=FixedRadius
       width=1
       color {colorUseRule=UseColor_Specified, specifiedColor=<rgba 0.6951 0.6951 0.6951 1>}
       cornerRadii=<4 4 4 4>
      }
      tintColor {colorUseRule=UseColor_Specified, specifiedColor=<rgba 0.4955 0.4955 0.4955 1>}
      uVRegion {bIsValid=false, max=<0 0>, min=<0 0>}
     }
     normalForeground {colorUseRule=UseColor_Specified, specifiedColor=<rgba 0 1 1 1>}
     normalPadding=<ltrb 0 0 0 0>
     pressed {
      bIsDynamicallyLoaded=false
      drawAs=NoDrawType
      imageType=NoImage
      mirroring=NoMirror
      resourceName=None
      resourceObject=null
      tiling=NoTile
      imageSize=<32 32>
      margin=<ltrb 0 0 0 0>
      outlineSettings {
       bUseBrushTransparency=true
       roundingType=FixedRadius
       width=1
       color {colorUseRule=UseColor_Specified, specifiedColor=<rgba 0.7243 0.7243 0.7243 1>}
       cornerRadii=<4 4 4 4>
      }
      tintColor {colorUseRule=UseColor_Specified, specifiedColor=<rgba 0.3843 0.3843 0.3843 1>}
      uVRegion {bIsValid=false, max=<0 0>, min=<0 0>}
     }
     pressedForeground {colorUseRule=UseColor_Specified, specifiedColor=<rgba 0 1 1 1>}
     pressedPadding=<ltrb 0 0 0 0>
     pressedSlateSound {resourceObject=null}
    }
   }
   slot {class=VerticalBoxSlot}
  }
  Overlay_2 {
   widgetClassPath=&Overlay
   parent=Button_Details_Hover
   widget=Overlay_2
   slot {class=ButtonSlot, properties {horizontalAlignment=HAlign_Fill, verticalAlignment=VAlign_Fill, padding=<ltrb 0 0 0 0>}}
  }
  Spacer_378 {widgetClassPath=&Spacer, parent=Overlay_2, widget=Spacer_378, properties {size=<1200 1>}, slot {class=OverlaySlot}}
  Spacer {widgetClassPath=&Spacer, parent=Overlay_2, widget=Spacer, properties {size=<1 100>}, slot {class=OverlaySlot}}
  VerticalBox_125 {
   widgetClassPath=&VerticalBox
   parent=Overlay_2
   widget=VerticalBox_125
   slot {class=OverlaySlot, properties {horizontalAlignment=HAlign_Fill, verticalAlignment=VAlign_Fill}}
  }
  Text_ActivityDescription {
   widgetClassPath=&CommonTextBlock
   bIsVariable=true
   parent=VerticalBox_125
   widget=Text_ActivityDescription
   properties {
    applyLineHeightToBottomLine=false
    autoWrapText=true
    mobileFontSizeMultiplier=1.2
    text="[Game Mode Description.  Wrapping is set to per character, text widget clips to bounds on itself. Lorem ipsum dolor sit amet, consectetur adipiscing elit."
    wrapTextAt=920
    wrappingPolicy=AllowPerCharacterWrapping
    colorAndOpacity {colorUseRule=UseColor_Specified, specifiedColor=<rgba 0.5 0.5 0.5 1>}
    font {
     bForceMonospaced=false
     bMaterialIsStencil=false
     fontMaterial=null
     fontObject=&HeadingNow
     letterSpacing=0
     monospacedWidth=1
     size=13.5
     skewAmount=0
     typefaceFontName=74 Regular
     outlineSettings {
      bApplyOutlineToDropShadows=false
      bMiteredCorners=false
      bSeparateFillAlpha=false
      outlineMaterial=null
      outlineSize=0
      outlineColor=<rgba 0 0 0 1>
     }
    }
    shadowOffset=<0 0>
    shapedTextOptions {bOverride_TextFlowDirection=true, bOverride_TextShapingMethod=false, textFlowDirection=Culture, textShapingMethod=Auto}
    strikeBrush {
     bIsDynamicallyLoaded=false
     drawAs=NoDrawType
     imageType=NoImage
     mirroring=NoMirror
     resourceName=None
     resourceObject=null
     tiling=NoTile
     imageSize=<32 32>
     margin=<ltrb 0 0 0 0>
     outlineSettings {
      bUseBrushTransparency=false
      roundingType=HalfHeightRadius
      width=0
      color {colorUseRule=UseColor_Specified, specifiedColor=<rgba 0 0 0 0>}
      cornerRadii=<0 0 0 0>
     }
     tintColor {colorUseRule=UseColor_Specified, specifiedColor=<rgba 1 1 1 1>}
     uVRegion {bIsValid=false, max=<0 0>, min=<0 0>}
    }
   }
   slot {class=VerticalBoxSlot, properties {verticalAlignment=VAlign_Top}}
  }
  Description_InGamePurchase_Divider {
   widgetClassPath=&Image
   bIsVariable=true
   parent=VerticalBox_125
   widget=Description_InGamePurchase_Divider
   properties {
    pixelSnapping=SnapToPixel
    brush {
     bIsDynamicallyLoaded=false
     drawAs=Image
     imageType=NoImage
     mirroring=NoMirror
     resourceName=None
     resourceObject=null
     tiling=NoTile
     imageSize=<32 2>
     margin=<ltrb 0 0 0 0>
     outlineSettings {
      bUseBrushTransparency=false
      roundingType=HalfHeightRadius
      width=0
      color {colorUseRule=UseColor_Specified, specifiedColor=<rgba 0 0 0 0>}
      cornerRadii=<0.1 0.1 0.1 0.1>
     }
     tintColor {colorUseRule=UseColor_Specified, specifiedColor=<rgba 0.5 0.5 0.5 1>}
     uVRegion {bIsValid=false, max=<0 0>, min=<0 0>}
    }
    colorAndOpacity=<rgba 0.5 0.5 0.5 1>
   }
   slot {class=VerticalBoxSlot, properties {verticalAlignment=VAlign_Center, padding=<ltrb 0 30 0 30>}}
  }
  InGamePurchaseContainer {
   widgetClassPath=&HorizontalBox
   bIsVariable=true
   parent=VerticalBox_125
   widget=InGamePurchaseContainer
   slot {class=VerticalBoxSlot, properties {horizontalAlignment=HAlign_Left, verticalAlignment=VAlign_Bottom}}
  }
  Image_InGamePurchase {
   widgetClassPath=&Image
   bIsVariable=true
   parent=InGamePurchaseContainer
   widget=Image_InGamePurchase
   properties {
    pixelSnapping=SnapToPixel
    brush {
     bIsDynamicallyLoaded=false
     drawAs=Image
     imageType=FullColor
     mirroring=NoMirror
     resourceName=None
     resourceObject=/FortUILibrary/Textures/MaterialInstances/MI_UIKit_Icon_Shop
     tiling=NoTile
     imageSize=<26 26>
     margin=<ltrb 0 0 0 0>
     outlineSettings {
      bUseBrushTransparency=false
      roundingType=HalfHeightRadius
      width=0
      color {colorUseRule=UseColor_Specified, specifiedColor=<rgba 0 0 0 0>}
      cornerRadii=<0.1 0.1 0.1 0.1>
     }
     tintColor {colorUseRule=UseColor_Specified, specifiedColor=<rgba 1 1 1 1>}
     uVRegion {bIsValid=false, max=<0 0>, min=<0 0>}
    }
   }
   slot {class=HorizontalBoxSlot, properties {horizontalAlignment=HAlign_Left, padding=<ltrb 0 0 10 0>, size {sizeRule=Fill, value=1}}}
  }
  Text_ActivityDescription_InGamePurchases {
   widgetClassPath=&CommonTextBlock
   bIsVariable=true
   parent=InGamePurchaseContainer
   widget=Text_ActivityDescription_InGamePurchases
   properties {
    applyLineHeightToBottomLine=false
    autoWrapText=true
    minDesiredWidth=500
    mobileFontSizeMultiplier=1.2
    text=Contains In-Island Transactions
    wrapTextAt=920
    wrappingPolicy=AllowPerCharacterWrapping
    colorAndOpacity {colorUseRule=UseColor_Specified, specifiedColor=<rgba 0.5 0.5 0.5 1>}
    font {
     bForceMonospaced=false
     bMaterialIsStencil=false
     fontMaterial=null
     fontObject=&HeadingNow
     letterSpacing=0
     monospacedWidth=1
     size=13.5
     skewAmount=0
     typefaceFontName=74 Regular
     outlineSettings {
      bApplyOutlineToDropShadows=false
      bMiteredCorners=false
      bSeparateFillAlpha=false
      outlineMaterial=null
      outlineSize=0
      outlineColor=<rgba 0 0 0 1>
     }
    }
    renderTransformPivot=<0 0.5>
    shadowOffset=<0 0>
    shapedTextOptions {bOverride_TextFlowDirection=true, bOverride_TextShapingMethod=false, textFlowDirection=Culture, textShapingMethod=Auto}
    strikeBrush {
     bIsDynamicallyLoaded=false
     drawAs=NoDrawType
     imageType=NoImage
     mirroring=NoMirror
     resourceName=None
     resourceObject=null
     tiling=NoTile
     imageSize=<32 32>
     margin=<ltrb 0 0 0 0>
     outlineSettings {
      bUseBrushTransparency=false
      roundingType=HalfHeightRadius
      width=0
      color {colorUseRule=UseColor_Specified, specifiedColor=<rgba 0 0 0 0>}
      cornerRadii=<0 0 0 0>
     }
     tintColor {colorUseRule=UseColor_Specified, specifiedColor=<rgba 1 1 1 1>}
     uVRegion {bIsValid=false, max=<0 0>, min=<0 0>}
    }
   }
   slot {class=HorizontalBoxSlot, properties {horizontalAlignment=HAlign_Left, verticalAlignment=VAlign_Center}}
  }
  WBP_UIKit_Backplate_Description {
   widgetClassPath=&WBP_UIKit_Backplate
   bIsVariable=true
   parent=Overlay_2
   widget=WBP_UIKit_Backplate_Description
   properties {brightness=Medium, colorAndOpacity=<rgba 1 1 1 0.25>}
   slot {class=OverlaySlot, properties {horizontalAlignment=HAlign_Fill, verticalAlignment=VAlign_Fill, padding=<ltrb -16 -16 -16 -16>}}
  }
  WBP_Outline {
   widgetClassPath=/FortUILibrary/UIKit/Buttons/Blocks/WBP_UIKit_Block_Outline
   bIsVariable=true
   parent=Overlay_2
   widget=WBP_Outline
   properties {
    bIsFocusable=true
    bottomLeftCornerRadius=20
    bottomRightCornerRadius=20
    navigation=&self:WidgetTree.WBP_Outline.WidgetNavigation_0
    priority=100
    radiusInPixels=true
    topLeftCornerRadius=20
    topRightCornerRadius=20
    visibility=Visible
    transitionTiming {
     deselectingFocused=0
     deselectingIdle=0
     disabling=0
     enabling=0
     focusing=1
     hovering=0
     pressing=0
     releasing=0
     selecting=0
     unfocusing=0
     unhovering=0
    }
   }
   slot {class=OverlaySlot, properties {horizontalAlignment=HAlign_Fill, verticalAlignment=VAlign_Fill, padding=<ltrb -8 -8 -8 -8>}}
  }
  DarkOverlay {
   widgetClassPath=&Image
   bIsVariable=true
   parent=ModalContent
   widget=DarkOverlay
   properties {
    visibility=HitTestInvisible
    brush {
     bIsDynamicallyLoaded=false
     drawAs=Image
     imageType=NoImage
     mirroring=NoMirror
     resourceName=None
     resourceObject=/FortniteCoreUI/UI/DiscoveryShared/Art/MaterialInstances/MI_ActivityDetails_Overlay
     tiling=NoTile
     imageSize=<32 112>
     margin=<ltrb 0 0 0 0>
     outlineSettings {
      bUseBrushTransparency=false
      roundingType=HalfHeightRadius
      width=0
      color {colorUseRule=UseColor_Specified, specifiedColor=<rgba 0 0 0 0>}
      cornerRadii=<0 0 0 0>
     }
     tintColor {colorUseRule=UseColor_Specified, specifiedColor=<rgba 1 1 1 1>}
     uVRegion {bIsValid=false, max=<0 0>, min=<0 0>}
    }
   }
   slot {class=OverlaySlot, properties {horizontalAlignment=HAlign_Fill, verticalAlignment=VAlign_Bottom}}
  }
  SafeZone_Prompts {
   widgetClassPath=&SafeZone
   bIsVariable=true
   parent=ModalContent
   widget=SafeZone_Prompts
   properties {padTop=false}
   slot {class=OverlaySlot, properties {horizontalAlignment=HAlign_Right, verticalAlignment=VAlign_Bottom, padding=<ltrb 80 0 80 32>}}
  }
  WBP_ActionBarContainer {
   widgetClassPath=/Game/UI/Frontend/BottomBar/WBP_ActionBarContainer
   bIsVariable=true
   parent=SafeZone_Prompts
   widget=WBP_ActionBarContainer
   slot {class=SafeZoneSlot}
  }
  HorizontalBox_1 {
   widgetClassPath=&HorizontalBox
   namedSlotHost=WBP_ActionBarContainer
   widget=HorizontalBox_1
   properties {flowDirectionPreference=Culture}
  }
  VBox_CreatorPage {widgetClassPath=&VerticalBox, bIsVariable=true, parent=HorizontalBox_1, widget=VBox_CreatorPage, slot {class=HorizontalBoxSlot}}
  FortVisualAttachment_0 {
   widgetClassPath=&FortVisualAttachment
   parent=VBox_CreatorPage
   widget=FortVisualAttachment_0
   properties {visibility=Collapsed, contentAnchor=<0.5 1>}
   slot {class=VerticalBoxSlot, properties {horizontalAlignment=HAlign_Center, verticalAlignment=VAlign_Bottom}}
  }
  ActivityDetailsModal_CreatorPageTooltip {
   widgetClassPath=/FortniteCoreUI/UI/DiscoveryShared/ActivityDetailsModal_CreatorPageTooltip
   bInherited=true
   bIsVariable=true
   parent=FortVisualAttachment_0
   widget=ActivityDetailsModal_CreatorPageTooltip
   properties {visibility=Collapsed}
   slot {class=SizeBoxSlot, properties {padding=<ltrb 0 0 0 35>}}
  }
  Attachment_Coachmark {
   widgetClassPath=&FortVisualAttachment
   parent=VBox_CreatorPage
   widget=Attachment_Coachmark
   properties {visibility=Visible, contentAnchor=<0.6 1.1>}
   slot {class=VerticalBoxSlot, properties {verticalAlignment=VAlign_Bottom}}
  }
  WBP_ActivityDetails_FollowCreator_CoachMark {
   widgetClassPath=&/WBP_ActivityDetails_FollowCreator_CoachMark
   bIsVariable=true
   parent=Attachment_Coachmark
   widget=WBP_ActivityDetails_FollowCreator_CoachMark
   properties {bIsBackHandler=false, bIsModal=false}
   slot {class=SizeBoxSlot}
  }
  Button_CreatorPage {
   widgetClassPath=/FortUILibrary/UIKit/Buttons/WBP_UIKit_Button_Quiet
   bInherited=true
   bIsVariable=true
   parent=VBox_CreatorPage
   widget=Button_CreatorPage
   properties {text=View Creator, triggeringInputAction {dataTable=&ActivityBrowserInputActionTable, rowName=ShowCreatorPage}}
   slot {class=VerticalBoxSlot, properties {horizontalAlignment=HAlign_Right, verticalAlignment=VAlign_Bottom, padding=<ltrb 0 0 24 0>}}
  }
  Button_Back_Unbound {
   widgetClassPath=&WBP_UIKit_Button_Quiet
   bIsVariable=true
   parent=HorizontalBox_1
   widget=Button_Back_Unbound
   properties {text=Back, triggeringInputAction {dataTable=&InputActionDataTable, rowName=GenericBack}}
   slot {class=HorizontalBoxSlot, properties {horizontalAlignment=HAlign_Right, verticalAlignment=VAlign_Bottom}}
  }
  VB_Version {
   widgetClassPath=&VerticalBox
   bIsVariable=true
   parent=Overlay_6
   widget=VB_Version
   properties {visibility=Collapsed}
   slot {class=OverlaySlot, properties {verticalAlignment=VAlign_Bottom, padding=<ltrb 8 0 0 8>}}
  }
  Text_LinkCodeVersion {
   widgetClassPath=&CommonTextBlock
   bInherited=true
   bIsVariable=true
   parent=VB_Version
   widget=Text_LinkCodeVersion
   properties {
    applyLineHeightToBottomLine=false
    text=v50 1234-5678-1234
    font {
     bForceMonospaced=false
     bMaterialIsStencil=false
     fontMaterial=null
     fontObject=&BurbankSmall
     letterSpacing=0
     monospacedWidth=1
     size=8
     skewAmount=0
     typefaceFontName=Bold
     outlineSettings {
      bApplyOutlineToDropShadows=false
      bMiteredCorners=false
      bSeparateFillAlpha=false
      outlineMaterial=null
      outlineSize=0
      outlineColor=<rgba 0 0 0 1>
     }
    }
    shadowOffset=<0 0>
   }
   slot {class=VerticalBoxSlot}
  }
  Spacer_64 {widgetClassPath=&Spacer, parent=VB_Version, widget=Spacer_64, properties {size=<1 2>}, slot {class=VerticalBoxSlot}}
  Text_VersionSubmode {
   widgetClassPath=&CommonTextBlock
   bInherited=true
   bIsVariable=true
   parent=VB_Version
   widget=Text_VersionSubmode
   properties {
    applyLineHeightToBottomLine=false
    text=vo5 Solo
    font {
     bForceMonospaced=false
     bMaterialIsStencil=false
     fontMaterial=null
     fontObject=&BurbankSmall
     letterSpacing=0
     monospacedWidth=1
     size=8
     skewAmount=0
     typefaceFontName=Bold
     outlineSettings {
      bApplyOutlineToDropShadows=false
      bMiteredCorners=false
      bSeparateFillAlpha=false
      outlineMaterial=null
      outlineSize=0
      outlineColor=<rgba 0 0 0 1>
     }
    }
    shadowOffset=<0 0>
   }
   slot {class=VerticalBoxSlot}
  }
  TODOMustberemovedoncewehaveVM {widgetClassPath=&Overlay, parent=Overlay_6, widget=TODOMustberemovedoncewehaveVM, slot {class=OverlaySlot}}
  DetailsDisplay_SelectedActivity {
   widgetClassPath=/FortniteCoreUI/UI/DiscoveryShared/ActivityDetailsDisplay
   bInherited=true
   bIsVariable=true
   parent=TODOMustberemovedoncewehaveVM
   widget=DetailsDisplay_SelectedActivity
   properties {
    bIsFocusable=true
    showAcceptButton=true
    showDetailsButton=true
    showFavoriteButton=true
    showLinkCode=true
    useRedesign=true
    visibility=Collapsed
   }
   slot {class=OverlaySlot}
  }
  O_Modal {
   widgetClassPath=&Overlay
   bIsVariable=true
   parent=Overlay_6
   widget=O_Modal
   slot {class=OverlaySlot, properties {horizontalAlignment=HAlign_Fill, verticalAlignment=VAlign_Fill}}
  }
  Button_CloseTouch {widgetClassPath=&CommonButtonBase, bInherited=true, widget=null}
  Button_Back {widgetClassPath=&CommonButtonBase, bInherited=true, widget=null}
 ]
}
