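# Translucent UI material for the Discovery game details background, compositing a media thumbnail with a nebulous animated-style backdrop.
# The thumbnail scales to fill with zoom and pivot, blending through vertical, rotatable horizontal and circular vignettes plus an eased opacity transition.
# Background colors are either manual parameters or derived through hue shift and rotation.
# Static switches add an error-texture fallback, fixed media size (a hack matching image to video) and mirroring for localization.
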Material /FortniteCoreUI/UI/DiscoveryShared/GameDetails/Materials/M_UI_GameDetails_Bg
graph {
 properties {
  bUsedWithStaticMesh=false
  blendMode=BLEND_Translucent
  materialDomain=MD_UI
  pixelDepthOffsetMode=PDOM_Legacy
  usedShadingModels=Default Lit
 }
 nodes [
  n1 = VectorParameter() {group=5- Colors, parameterName=Primary Background Color, sortPriority=1, defaultValue=<rgba 0 0.3278 0.9823 1>}
  n2 = NamedRerouteDeclaration(None=@n1.RGB) {name=Main Color, nodeColor=<rgba 0 0 0 1>}
  n3 = VectorParameter() {group=5- Colors, parameterName=Background Color, sortPriority=3, defaultValue=<rgba 0.0009 0.0009 0.0097 1>}
  n4 = StaticSwitchParameter(True=@n3.RGB, False=NamedRerouteUsage()) {defaultValue=true, group=5- Colors, parameterName=Manual Colors, sortPriority=0}
  MaterialFunctionCall!(Primary Color=@n2, Background Color=@n4, Is Flipped Horizontally) {materialFunction=/FortUILibrary/Materials/Background/Nebulous/MF_UI_Background_Immersive}
  n5 = TextureObjectParameter() {group=1 - Thumbnail, parameterName=Thumbnail, sortPriority=1, texture=/Game/UI/DiscoveryShared/Art/T_UI_Discovery_DefaultEmpty}
  NamedRerouteDeclaration(None=@n5) {name=Texture, nodeColor=<rgba 0 0 0 1>}
  NamedRerouteDeclaration(None=TextureSample(UVs=NamedRerouteUsage(), Tex=NamedRerouteUsage()).RGB) {name=Thumbnail Color, nodeColor=<rgba 0 0 0 1>}
  NamedRerouteDeclaration(None=MaterialFunctionCall() {materialFunction=&GetUserInterfaceUV}.Normalized UV) {name=Normalized UV, nodeColor=<rgba 0 0 0 1>}
  n6 = MaterialFunctionCall(UV=NamedRerouteUsage(), TextureSize=NamedRerouteUsage()) {materialFunction=&MF_UI_UVScaling}
  NamedRerouteDeclaration(None=@n6.Scaled to Fit) {name=Scale to Fit UV, nodeColor=<rgba 0 0 0 1>}
  NamedRerouteDeclaration(None=@n6.ScaleY) {name=Scale Y UV, nodeColor=<rgba 0 0 0 1>}
  NamedRerouteDeclaration(None=MaterialFunctionCall!(ScreenUV, TextureSize, ScalePivotPoint) {materialFunction=/FortUILibrary/MaterialFunctions/Basics/MF_UI_ScreenSpaceUVScaling}.Scaled to Fit) {name=Screenspace UV, nodeColor=<rgba 0 0 0 1>}
  n7 = NamedRerouteUsage()
  n8 = NamedRerouteUsage()
  n9 = AppendVector(A=OneMinus(None=ComponentMask(None=@n8) {r=true}), B=ComponentMask(None=@n8) {g=true})
  n10 = MaterialFunctionCall!(
   UV=Multiply(A=StaticSwitchParameter(True=@n9, False=@n8) {desc=Used for localization, group=0 - Global, parameterName=Inverse Order, sortPriority=0}) {constB=1.2, desc=Triangles scale}
   Color A
   Color B
  ) {materialFunction=/FortniteCoreUI/UI/DiscoveryShared/Art/MaterialFunctions/MF_UI_Discovery_TileBacking}
  n11 = MaterialFunctionCall!(Base=@n7, Add=Multiply(A=@n10.Masked Pattern) {constB=2}) {materialFunction=/Engine/Functions/Engine_MaterialFunctions02/Math/MultiplyAdd}
  n12 = NamedRerouteDeclaration(None=MaterialFunctionCall(UV=NamedRerouteUsage()) {materialFunction=&MF_UI_RadialGradient}.Smooth) {name=Radial Gradient, nodeColor=<rgba 0 0 0 1>}
  NamedRerouteDeclaration(None=LinearInterpolate(A=@n7, B=Saturate(None=@n11.Result), Alpha=Multiply(A=@n12) {constB=0.5})) {name=Thumbnail Backing, nodeColor=<rgba 0 0 0 1>}
  n13 = Max(A=NamedRerouteUsage()) {constB=0}
  n14 = Saturate(None=MaterialFunctionCall(Hue Shift Percentage=Constant() {r=0.072}, Texture=@n13) {materialFunction=&HueShift}.Result)
  n15 = Constant() {r=0.6534}
  n16 = MaterialFunctionCall!(Color=Max(A=Desaturation(None=@n14, Fraction=@n15)) {constB=0.001}, Contrast=Constant() {r=-0.0929}) {materialFunction=/Engine/Functions/Engine_MaterialFunctions01/ImageAdjustment/Contrast_Preserve_Color}
  NamedRerouteDeclaration(None=Multiply(A=Saturate(None=@n16.Result)) {constB=0.0116}) {name=Background Color, nodeColor=<rgba 0 0 0 1>}
  n17 = NamedRerouteUsage()
  n18 = Add(A=ScalarParameter() {group=1 - Thumbnail, parameterName=Thumbnail Zoom, sortPriority=1}, B=Constant() {r=1})
  n19 = AppendVector(A=Constant() {r=0.5}, B=ScalarParameter() {group=1 - Thumbnail, parameterName=Pivot Point Y, sliderMax=0.5, sliderMin=-0.5, sortPriority=2})
  NamedRerouteDeclaration(None=MaterialFunctionCall(
    UV=MaterialFunctionCall(UVs=@n17, Texture Scale=@n18) {materialFunction=&ScaleUVsByCenter}.UVs
    TextureSize=NamedRerouteUsage()
    ScalePivotPoint=NamedRerouteDeclaration(None=@n19) {name=Pivot Point, nodeColor=<rgba 0 0 0 1>}
   ) {materialFunction=&MF_UI_UVScaling}.Scaled To Fill) {name=Thumbnail UV, nodeColor=<rgba 0 0 0 1>}
  n20 = NamedRerouteUsage()
  n21 = AppendVector(
   A=Constant()
   B=ScalarParameter() {group=2 - Vertical Vignette, parameterName=Vertical Vignette Offset, sliderMax=1, sliderMin=-1, sortPriority=4}
  )
  n22 = SmoothStep(
   Min=Multiply(A=ScalarParameter() {defaultValue=0.944, group=2 - Vertical Vignette, parameterName=Vertical Vignette Smooth, sliderMax=1, sortPriority=2}) {constB=-1}
   Value=ComponentMask(None=Add(A=@n20, B=@n21)) {g=true}
  )
  n23 = ScalarParameter() {defaultValue=0.944, group=2 - Vertical Vignette, parameterName=Vertical Vignette Dispersion, sliderMax=1, sortPriority=1}
  n24 = Constant() {desc=Max, r=1}
  n25 = Power(
   Base=Divide(A=Subtract(A=@n22, B=@n23), B=Subtract(A=@n24, B=@n23))
   Exp=ScalarParameter() {defaultValue=1.25, group=2 - Vertical Vignette, parameterName=Vertical Vignette Decay Rate, sortPriority=3}
  ) {constExponent=1.25}
  NamedRerouteDeclaration(None=StaticSwitchParameter(True=Saturate(None=@n25), False=Constant()) {defaultValue=true, group=2 - Vertical Vignette, parameterName=Display Vertical Gradient, sortPriority=0}) {name=Thumbnail Vignette, nodeColor=<rgba 0 0 0 1>}
  n26 = ScalarParameter() {defaultValue=0.944, group=3 - Horizontal Vignette, parameterName=Vignette Smooth, sliderMax=1, sortPriority=1}
  n27 = ScalarParameter() {group=3 - Horizontal Vignette, parameterName=Vignette Rotation, sortPriority=0}
  n28 = MaterialFunctionCall(
   UVs=NamedRerouteUsage()
   "Rotation Angle (0-1)"=StaticSwitchParameter(True=OneMinus(None=@n27), False=@n27) {desc=Used for localization, group=0 - Global, parameterName=Inverse Order, sortPriority=0}
  ) {materialFunction=&CustomRotator}
  n29 = StaticSwitchParameter(True=OneMinus(None=@n28.Rotated Values), False=@n28.Rotated Values) {desc=Used for localization, group=0 - Global, parameterName=Inverse Order, sortPriority=0}
  n30 = Add(A=@n29, B=ScalarParameter() {group=3 - Horizontal Vignette, parameterName=Vignette Right Offset, sliderMax=1, sliderMin=-1}) {constB=0.5}
  n31 = Constant()
  n32 = OneMinus(None=Subtract(A=@n29, B=ScalarParameter() {group=3 - Horizontal Vignette, parameterName=Vignette Left Offset, sliderMax=1, sliderMin=-1, sortPriority=31}) {constB=0.5})
  n33 = AppendVector(
   A=StaticSwitchParameter(True=ComponentMask(None=@n30) {r=true}, False=@n31) {defaultValue=true, group=3 - Horizontal Vignette, parameterName=Display Right Gradient, sortPriority=0}
   B=StaticSwitchParameter(True=ComponentMask(None=@n32) {r=true}, False=@n31) {defaultValue=true, group=3 - Horizontal Vignette, parameterName=Display Left Gradient, sortPriority=0}
  )
  n34 = ScalarParameter() {defaultValue=0.944, group=3 - Horizontal Vignette, parameterName=Vignette Dispersion, sliderMax=1, sortPriority=1}
  n35 = Subtract(A=SmoothStep(Min=Multiply(A=@n26) {constB=-1}, Value=Saturate(None=@n33)), B=@n34)
  n36 = Subtract(A=Constant() {desc=Max, r=1}, B=@n34)
  n37 = Saturate(None=Power(
    Base=Divide(A=@n35, B=@n36)
    Exp=ScalarParameter() {defaultValue=1.25, group=3 - Horizontal Vignette, parameterName=Vignette Decay Rate, sortPriority=30}
   ) {constExponent=1.25})
  n38 = ComponentMask(None=@n37) {r=true}
  n39 = Constant()
  n40 = ComponentMask(None=@n37) {g=true}
  NamedRerouteDeclaration(None=LinearInterpolate(
    A=StaticSwitchParameter(True=@n38, False=@n39) {defaultValue=true, group=3 - Horizontal Vignette, parameterName=Display Right Gradient, sortPriority=0}
    Alpha=StaticSwitchParameter(True=@n40, False=@n39) {defaultValue=true, group=3 - Horizontal Vignette, parameterName=Display Left Gradient, sortPriority=0}
   )) {name=Horizontal Gradient Mask, nodeColor=<rgba 0 0 0 1>}
  NamedRerouteDeclaration(None=@n6.ScaleX) {name=Scale X UV, nodeColor=<rgba 0 0 0 1>}
  NamedRerouteDeclaration(None=ComponentMask(None=@n10.Masked Pattern) {r=true}) {name=Triangle Pattern, nodeColor=<rgba 0 0 0 1>}
  n41 = MaterialFunctionCall(
   UVs=NamedRerouteUsage()
   CenterPosition=AppendVector(A=ScalarParameter() {defaultValue=1, parameterName=Position_X}, B=ScalarParameter() {defaultValue=-0.3, parameterName=Position_Y})
   Radius=ScalarParameter() {defaultValue=1.44, parameterName=Radius}
   Density=ScalarParameter() {defaultValue=5, parameterName=Density}
  ) {materialFunction=&RadialGradientExponential}
  n42 = OneMinus(None=@n41.RadialGradientExponential)
  n43 = ScalarParameter() {defaultValue=1, parameterName=InvertDensity, sliderMax=1}
  NamedRerouteDeclaration(None=Saturate(None=LinearInterpolate(A=@n41.RadialGradientExponential, B=@n42, Alpha=@n43))) {name=Circular Vignette, nodeColor=<rgba 0 0 0 1>}
  n44 = TextureObjectParameter() {parameterName=Error Fallback Texture}
  NamedRerouteDeclaration(None=TextureSample(UVs=NamedRerouteUsage(), Tex=NamedRerouteDeclaration(None=@n44) {name=Error Texture, nodeColor=<rgba 0 0 0 1>}).RGB) {name=FallBack RGB, nodeColor=<rgba 0 0 0 1>}
  NamedRerouteDeclaration(None=@n6.Scaled To Fill) {name=Scale to Fill UV, nodeColor=<rgba 0 0 0 1>}
  n45 = Constant3Vector() {constant=<rgba 1 1 1 0>}
  n46 = MaterialFunctionCall(
   Base=@n13
   Add=RotateAboutAxis(NormalizedRotationAxis=Normalize(VectorInput=@n45), RotationAngle=Constant() {r=0.08}, PivotPoint=Constant(), Position=@n13)
  ) {materialFunction=&MultiplyAdd}
  NamedRerouteDeclaration(None=LinearInterpolate(B=Saturate(None=@n46.Result)) {constAlpha=0.9912}) {name=Secondary Color, nodeColor=<rgba 0 0 0 1>}
  n47 = ScalarParameter() {defaultValue=2048, group=1 - Thumbnail, parameterName=Media Width}
  n48 = ScalarParameter() {defaultValue=1024, group=1 - Thumbnail, parameterName=Media Height, sortPriority=33}
  n49 = TextureProperty(Tex=NamedRerouteUsage())
  n50 = LinearInterpolate(A=@n49, B=TextureProperty(Tex=NamedRerouteUsage()), Alpha=ScalarParameter() {parameterName=IsError, sliderMax=1})
  NamedRerouteDeclaration(None=StaticSwitchParameter(
    True=AppendVector(A=@n47, B=@n48) {desc=HACK - Used to send the texture size manually to have both image and video to be the same size}
    False=StaticSwitchParameter(True=@n50, False=@n49) {group=1 - Thumbnail, parameterName=Support Error Fallback, sortPriority=0}
   ) {defaultValue=true, group=1 - Thumbnail, parameterName=Use Fixed Size, sortPriority=31}) {name=Texture Size, nodeColor=<rgba 0 0 0 1>}
  n51 = VectorParameter() {group=5- Colors, parameterName=Secondary Background Color, sortPriority=2, defaultValue=<rgba 0 0.2423 0.9734 1>}
  n52 = NamedRerouteUsage()
  NamedRerouteDeclaration(None=MaterialFunctionCall!(
    Primary Color=@n2
    Secondary Color=StaticSwitchParameter(True=@n51.RGB, False=@n52) {defaultValue=true, group=5- Colors, parameterName=Manual Colors, sortPriority=0}
    Background Color=@n4
    Wind Opacity=ScalarParameter() {defaultValue=0.5, parameterName=Wind Opacity, sliderMax=1}
    Rotation Multiplier
    Triangle Opacity=ScalarParameter() {defaultValue=1, parameterName=Triangle Opacity, sliderMax=1}
    Offset Multiplier
   ) {materialFunction=/FortUILibrary/Materials/Background/Nebulous/MF_UI_BackgroundNebulous}.Final Color) {name=Outside Background, nodeColor=<rgba 0 0 0 1>}
  NamedRerouteUsage()
  n53 = NamedRerouteUsage()
  n54 = NamedRerouteUsage()
  n55 = StaticSwitchParameter(True=LinearInterpolate(A=@n54, B=NamedRerouteUsage(), Alpha=ScalarParameter() {parameterName=IsError, sliderMax=1}), False=@n54) {group=1 - Thumbnail, parameterName=Support Error Fallback, sortPriority=0}
  n56 = ScalarParameter() {defaultValue=1, group=1 - Thumbnail, parameterName=Thumbnail Opacity, sliderMax=1, sortPriority=1}
  n57 = LinearInterpolate(A=NamedRerouteUsage(), B=SmoothStep(Min=NamedRerouteUsage(), Value=NamedRerouteUsage()), Alpha=NamedRerouteUsage())
  n58 = LinearInterpolate(
   A=@n56
   B=NamedRerouteDeclaration(None=MaterialFunctionCall("Value (0-1)"=@n56) {materialFunction=&MF_UI_EaseCurves}.EaseOutQuint) {name=Thumbnail Opacity Transition, nodeColor=<rgba 0 0 0 1>}
   Alpha=Saturate(None=Multiply(A=@n57, B=@n57) {constB=4})
  )
  n59 = NamedRerouteUsage()
  n60 = StaticSwitchParameter(True=LinearInterpolate(A=NamedRerouteUsage(), Alpha=@n59), False=@n59) {defaultValue=true, parameterName=Display Circular Vignette, sortPriority=0}
  n61 = NamedRerouteUsage()
  StaticSwitchParameter(
   True=LinearInterpolate(A=LinearInterpolate(A=@n53, B=@n55, Alpha=@n58), B=NamedRerouteUsage(), Alpha=LinearInterpolate(A=@n60, Alpha=@n61))
   False=NamedRerouteUsage()
  ) {defaultValue=true, group=1 - Thumbnail, parameterName=Display Thumbnail, sortPriority=0}
 ]
}
